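# Control Rig for the M_MED_Ensemble male body cosmetic that procedurally corrects its dyn_Skirt bones as legs and torso move.
# A two-branch sequence handles left (B/D/E) then right (G/H/J) chains, with slight crossover.
# Spherical pose readers on calves, thighs, thigh controls and spine_02 weight parent constraints and additional transform offsets.
# Each branch aims the opposite thigh control at its foot control, letting either leg influence both sides.
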ControlRigBlueprint /BRCosmetics/Characters/Player/Male/Medium/Bodies/M_MED_Ensemble/Meshes/M_MED_Ensemble_CtrlRig
graphs [
 ContainedGraph {nodes [
   n1 = RigVMFunctionEntryNode()
   SequenceAggregate = &RigVMFunction_Sequence(ExecuteContext=@n1.ExecuteContext)
   SequenceAggregate_1 = &RigVMFunction_Sequence(ExecuteContext=@SequenceAggregate.B)
   SequenceAggregate_2 = &RigVMFunction_Sequence(ExecuteContext=@SequenceAggregate_1.B)
   RigVMFunctionReturnNode(A=@SequenceAggregate.A, B=@SequenceAggregate_1.A, C=@SequenceAggregate_2.A, D=@SequenceAggregate_2.B)
  ]}
 RigVMFunctionLibrary {links [], nodes []}
 RigVMModel {nodes [
   RigUnit_BeginExecution = &RigUnit_BeginExecution()
   SequenceAggregate = RigVMAggregateNode(ExecuteContext=@RigUnit_BeginExecution.ExecutePin)
   n1 = &RigUnit_SphericalPoseReader(
    ActiveRegionScaleFactors=(PositiveWidth=2,NegativeWidth=0.2,PositiveHeight=1,NegativeHeight=2)
    ActiveRegionSize=0.1
    Debug=(bDrawDebug=false,DebugScale=25,DebugSegments=20,DebugThickness=0.25)
    DriverAxis=(X=-1,Y=0,Z=0)
    DriverItem=(Type=Bone,Name="calf_l")
    FalloffRegionScaleFactors=(PositiveWidth=1,NegativeWidth=0.9,PositiveHeight=3,NegativeHeight=1)
    FalloffSize=0.61
    FlipHeightScaling=false
    FlipWidthScaling=false
    OptionalParentItem=(Type=Bone,Name="thigh_l")
    RotationOffset=(X=-140,Y=-10,Z=30)
    ExecutePin=@SequenceAggregate.A
   )
   n2 = &RigVMFunction_MathQuaternionFromEuler(Euler=(X=20,Y=0,Z=0))
   > &RigUnit_ModifyTransforms(
    ItemToModify="((Item=(Type=Bone,Name=\"dyn_SkirtE03\"),Transform=(Rotation=(X=0,Y=0,Z=0,W=1),Translation=(X=0,Y=0,Z=0),Scale3D=(X=1,Y=1,Z=1))),(Item=(Type=Bone,Name=\"dyn_SkirtE05\"),Transform=(Rotation=(X=0,Y=0,Z=0,W=1),Translation=(X=0,Y=0,Z=0),Scale3D=(X=1,Y=1,Z=1))),(It ...(truncated)"
    ItemToModify.2.Transform.Rotation=&RigVMFunction_MathQuaternionFromEuler(Euler=(X=20,Y=0,Z=0)).Result
    ItemToModify.0.Transform.Rotation=&RigVMFunction_MathQuaternionFromEuler(Euler=(X=60,Y=0,Z=0)).Result
    ItemToModify.1.Transform.Rotation=&RigVMFunction_MathQuaternionFromEuler(Euler=(X=80,Y=0,Z=0)).Result
    ItemToModify.3.Transform.Rotation=@n2.Result
    ItemToModify.4.Transform.Rotation=@n2.Result
    Weight=@n1.OutputParam
   )
   > n3 = &RigUnit_SphericalPoseReader(
    ActiveRegionScaleFactors=(PositiveWidth=1,NegativeWidth=1,PositiveHeight=1,NegativeHeight=1)
    ActiveRegionSize=0.1
    Debug=(bDrawDebug=false,DebugScale=25,DebugSegments=20,DebugThickness=0.25)
    DriverAxis=(X=-1,Y=0,Z=0)
    DriverItem=(Type=Bone,Name="thigh_l")
    FalloffRegionScaleFactors=(PositiveWidth=1,NegativeWidth=1.5,PositiveHeight=1,NegativeHeight=2)
    FalloffSize=0.37
    FlipHeightScaling=false
    FlipWidthScaling=false
    OptionalParentItem=(Type=Bone,Name="pelvis")
    RotationOffset=(X=130,Y=40,Z=-20)
   )
   > &RigUnit_ParentConstraint(
    AdvancedSettings=(RotationOrderForFilter=XZY)
    Child=(Type=Bone,Name="dyn_SkirtE03")
    Filter=(TranslationFilter=(bX=true,bY=true,bZ=true),RotationFilter=(bX=true,bY=true,bZ=true),ScaleFilter=(bX=True,bY=True,bZ=True))
    Parents=((Item=(Type=Bone,Name="dyn_SkirtE03"),Weight=0),(Item=(Type=Bone,Name="pelvis"),Weight=1))
    Weight=@n3.OutputParam
   )
   > &RigUnit_ParentConstraint(
    AdvancedSettings=(RotationOrderForFilter=XZY)
    Child=(Type=Bone,Name="dyn_SkirtE04")
    Filter=(TranslationFilter=(bX=true,bY=true,bZ=true),RotationFilter=(bX=true,bY=true,bZ=true),ScaleFilter=(bX=True,bY=True,bZ=True))
    Parents=((Item=(Type=Bone,Name="thigh_twist_01_l"),Weight=0),(Item=(Type=Bone,Name="pelvis"),Weight=1))
    Weight=@n3.OutputParam
   )
   > &RigUnit_ParentConstraint(
    AdvancedSettings=(RotationOrderForFilter=XZY)
    Child=(Type=Bone,Name="dyn_SkirtE05")
    Filter=(TranslationFilter=(bX=true,bY=true,bZ=true),RotationFilter=(bX=true,bY=true,bZ=true),ScaleFilter=(bX=True,bY=True,bZ=True))
    Parents=((Item=(Type=Bone,Name="calf_l"),Weight=0),(Item=(Type=Bone,Name="pelvis"),Weight=1))
    Weight=@n3.OutputParam
   )
   > &RigUnit_ParentConstraint(
    AdvancedSettings=(RotationOrderForFilter=XZY)
    Child=(Type=Bone,Name="dyn_SkirtD04")
    Filter=(TranslationFilter=(bX=false,bY=false,bZ=false),RotationFilter=(bX=true,bY=true,bZ=true),ScaleFilter=(bX=True,bY=True,bZ=True))
    Parents=((Item=(Type=Bone,Name="dyn_SkirtD03"),Weight=0.7),(Item=(Type=Bone,Name="pelvis"),Weight=0.3))
    Weight=@n3.OutputParam
   )
   > &RigUnit_AimConstraintLocalSpaceOffset(
    AdvancedSettings=(DebugSettings=(bEnabled=false,Scale=10,WorldOffset=(Rotation=(X=0,Y=0,Z=0,W=1),Translation=(X=0,Y=0,Z=0),Scale3D=(X=1,Y=1,Z=1))),RotationOrderForFilter=XZY)
    Child=(Type=Control,Name="thigh_r_ctrl")
    Filter=(bX=True,bY=True,bZ=True)
    Parents=((Item=(Type=Control,Name="foot_r_ctrl"),Weight=1))
    Weight=1
    WorldUp=(Target=(X=0,Y=0,Z=1),Kind=Direction,Space=(Type=None,Name="None"))
    bMaintainOffset=true
   )
   > n4 = &RigUnit_SphericalPoseReader(
    ActiveRegionScaleFactors=(PositiveWidth=3,NegativeWidth=2,PositiveHeight=4,NegativeHeight=2)
    ActiveRegionSize=0.1
    Debug=(bDrawDebug=false,DebugScale=25,DebugSegments=20,DebugThickness=0.25)
    DriverAxis=(X=-1,Y=0,Z=0)
    DriverItem=(Type=Control,Name="thigh_r_ctrl")
    FalloffRegionScaleFactors=(PositiveWidth=0.3,NegativeWidth=2,PositiveHeight=1,NegativeHeight=1)
    FalloffSize=0.35
    FlipHeightScaling=false
    FlipWidthScaling=false
    OptionalParentItem=(Type=Bone,Name="pelvis")
    RotationOffset=(X=-30,Y=0,Z=0)
   )
   > &RigUnit_ParentConstraint(
    AdvancedSettings=(RotationOrderForFilter=XZY)
    Child=(Type=Bone,Name="dyn_SkirtD03")
    Filter=(TranslationFilter=(bX=false,bY=false,bZ=false),RotationFilter=(bX=true,bY=true,bZ=true),ScaleFilter=(bX=True,bY=True,bZ=True))
    Parents=((Item=(Type=Bone,Name="thigh_twist_01_l"),Weight=0),(Item=(Type=Bone,Name="pelvis"),Weight=1))
    Weight=@n4.OutputParam
   )
   > &RigUnit_ModifyTransforms(
    ItemToModify="((Item=(Type=Bone,Name=\"dyn_SkirtD04\"),Transform=(Rotation=(X=0.2146,Y=-0.1274,Z=0.0283,W=0.9679),Translation=(X=0,Y=0,Z=0),Scale3D=(X=1,Y=1,Z=1))),(Item=(Type=Bone,Name=\"dyn_SkirtE04\"),Transform=(Rotation=(X=0.1736,Y=0,Z=0,W=0.9848),Translation=(X=0,Y=0,Z ...(truncated)"
    Weight=@n4.OutputParam
   )
   > n5 = &RigUnit_SphericalPoseReader(
    ActiveRegionScaleFactors=(PositiveWidth=1,NegativeWidth=1,PositiveHeight=1,NegativeHeight=1)
    ActiveRegionSize=0.1
    Debug=(bDrawDebug=false,DebugScale=25,DebugSegments=20,DebugThickness=0.25)
    DriverAxis=(X=-1,Y=0,Z=0)
    DriverItem=(Type=Control,Name="thigh_l_ctrl")
    FalloffRegionScaleFactors=(PositiveWidth=1,NegativeWidth=1,PositiveHeight=1,NegativeHeight=0.8)
    FalloffSize=0.6
    FlipHeightScaling=false
    FlipWidthScaling=false
    OptionalParentItem=(Type=Bone,Name="pelvis")
    RotationOffset=(X=-110,Y=10,Z=20)
   )
   > &RigUnit_ParentConstraint(
    AdvancedSettings=(RotationOrderForFilter=XZY)
    Child=(Type=Bone,Name="dyn_SkirtB03")
    Filter=(TranslationFilter=(bX=true,bY=true,bZ=true),RotationFilter=(bX=True,bY=True,bZ=True),ScaleFilter=(bX=True,bY=True,bZ=True))
    Parents=((Item=(Type=Bone,Name="thigh_twist_01_l"),Weight=0.8),(Item=(Type=Bone,Name="pelvis"),Weight=0.2))
    Weight=@n5.OutputParam
   )
   > &RigUnit_ModifyTransforms(
    ItemToModify="((Item=(Type=Bone,Name=\"dyn_SkirtE03\"),Transform=(Rotation=(X=0.1305,Y=0,Z=0,W=0.9914),Translation=(X=0,Y=0,Z=0),Scale3D=(X=1,Y=1,Z=1))),(Item=(Type=Bone,Name=\"dyn_SkirtE05\"),Transform=(Rotation=(X=0.1305,Y=0,Z=0,W=0.9914),Translation=(X=0,Y=0,Z=0),Scale3D ...(truncated)"
    Weight=@n5.OutputParam
   )
   > n6 = &RigUnit_SphericalPoseReader(
    ActiveRegionScaleFactors=(PositiveWidth=3,NegativeWidth=2,PositiveHeight=2,NegativeHeight=2)
    ActiveRegionSize=0.1
    Debug=(bDrawDebug=false,DebugScale=25,DebugSegments=20,DebugThickness=0.25)
    DriverAxis=(X=1,Y=0,Z=0)
    DriverItem=(Type=Bone,Name="spine_02")
    FalloffRegionScaleFactors=(PositiveWidth=1,NegativeWidth=1,PositiveHeight=1,NegativeHeight=1)
    FalloffSize=0.45
    FlipWidthScaling=false
    OptionalParentItem=(Type=Bone,Name="collision")
    RotationOffset=(X=90,Y=0,Z=5)
   )
   > &RigUnit_ParentConstraint(
    AdvancedSettings=(RotationOrderForFilter=XZY)
    Child=(Type=Bone,Name="dyn_SkirtE04")
    Filter=(TranslationFilter=(bX=false,bY=false,bZ=false),RotationFilter=(bX=True,bY=True,bZ=True),ScaleFilter=(bX=True,bY=True,bZ=True))
    Parents=((Item=(Type=Bone,Name="dyn_SkirtE04"),Weight=0.5),(Item=(Type=Bone,Name="pelvis"),Weight=0.5))
    Weight=@n6.OutputParam
   )
   > &RigUnit_ModifyTransforms(
    ItemToModify="((Item=(Type=Bone,Name=\"dyn_SkirtD04\"),Transform=(Rotation=(X=0,Y=0.1736,Z=0,W=0.9848),Translation=(X=0,Y=0,Z=0),Scale3D=(X=1,Y=1,Z=1))),(Item=(Type=Bone,Name=\"dyn_SkirtE04\"),Transform=(Rotation=(X=0,Y=0,Z=0,W=1),Translation=(X=0,Y=0,Z=0),Scale3D=(X=1,Y=1, ...(truncated)"
    ItemToModify.1.Transform.Rotation=&RigVMFunction_MathQuaternionFromEuler(Euler=(X=0,Y=0,Z=-10)).Result
    Weight=@n6.OutputParam
   )
   n7 = &RigUnit_SphericalPoseReader(
    ActiveRegionScaleFactors=(PositiveWidth=2,NegativeWidth=0.2,PositiveHeight=1,NegativeHeight=2)
    ActiveRegionSize=0.1
    Debug=(bDrawDebug=false,DebugScale=25,DebugSegments=20,DebugThickness=0.25)
    DriverAxis=(X=1,Y=0,Z=0)
    DriverItem=(Type=Bone,Name="calf_r")
    FalloffRegionScaleFactors=(PositiveWidth=1,NegativeWidth=0.9,PositiveHeight=3,NegativeHeight=1)
    FalloffSize=0.61
    FlipHeightScaling=false
    FlipWidthScaling=true
    OptionalParentItem=(Type=Bone,Name="thigh_r")
    RotationOffset=(X=40,Y=-10,Z=30)
    ExecutePin=@SequenceAggregate.B
   )
   n8 = &RigVMFunction_MathQuaternionFromEuler(Euler=(X=15,Y=0,Z=0))
   > &RigUnit_ModifyTransforms(
    ItemToModify="((Item=(Type=Bone,Name=\"dyn_SkirtG03\"),Transform=(Rotation=(X=0,Y=0,Z=0,W=1),Translation=(X=0,Y=0,Z=0),Scale3D=(X=1,Y=1,Z=1))),(Item=(Type=Bone,Name=\"dyn_SkirtG05\"),Transform=(Rotation=(X=0,Y=0,Z=0,W=1),Translation=(X=0,Y=0,Z=0),Scale3D=(X=1,Y=1,Z=1))),(It ...(truncated)"
    ItemToModify.0.Transform.Rotation=&RigVMFunction_MathQuaternionFromEuler(Euler=(X=60,Y=0,Z=0)).Result
    ItemToModify.2.Transform.Rotation=&RigVMFunction_MathQuaternionFromEuler(Euler=(X=20,Y=0,Z=0)).Result
    ItemToModify.1.Transform.Rotation=&RigVMFunction_MathQuaternionFromEuler(Euler=(X=80,Y=30,Z=0)).Result
    ItemToModify.3.Transform.Rotation=@n8.Result
    ItemToModify.4.Transform.Rotation=@n8.Result
    Weight=@n7.OutputParam
   )
   > n9 = &RigUnit_SphericalPoseReader(
    ActiveRegionScaleFactors=(PositiveWidth=1,NegativeWidth=1,PositiveHeight=1,NegativeHeight=1)
    ActiveRegionSize=0.1
    Debug=(bDrawDebug=false,DebugScale=25,DebugSegments=20,DebugThickness=0.25)
    DriverAxis=(X=1,Y=0,Z=0)
    DriverItem=(Type=Bone,Name="thigh_r")
    FalloffRegionScaleFactors=(PositiveWidth=1,NegativeWidth=1.5,PositiveHeight=1,NegativeHeight=2)
    FalloffSize=0.37
    FlipHeightScaling=false
    FlipWidthScaling=true
    OptionalParentItem=(Type=Bone,Name="pelvis")
    RotationOffset=(X=-40,Y=40,Z=-20)
   )
   > &RigUnit_ParentConstraint(
    AdvancedSettings=(RotationOrderForFilter=XZY)
    Child=(Type=Bone,Name="dyn_SkirtG03")
    Filter=(TranslationFilter=(bX=true,bY=true,bZ=true),RotationFilter=(bX=true,bY=true,bZ=true),ScaleFilter=(bX=True,bY=True,bZ=True))
    Parents=((Item=(Type=Bone,Name="thigh_twist_01_r"),Weight=0),(Item=(Type=Bone,Name="pelvis"),Weight=1))
    Weight=@n9.OutputParam
   )
   > &RigUnit_ParentConstraint(
    AdvancedSettings=(RotationOrderForFilter=XZY)
    Child=(Type=Bone,Name="dyn_SkirtG04")
    Filter=(TranslationFilter=(bX=true,bY=true,bZ=true),RotationFilter=(bX=true,bY=true,bZ=true),ScaleFilter=(bX=True,bY=True,bZ=True))
    Parents=((Item=(Type=Bone,Name="thigh_twist_01_r"),Weight=0),(Item=(Type=Bone,Name="pelvis"),Weight=1))
    Weight=@n9.OutputParam
   )
   > &RigUnit_ParentConstraint(
    AdvancedSettings=(RotationOrderForFilter=XZY)
    Child=(Type=Bone,Name="dyn_SkirtG05")
    Filter=(TranslationFilter=(bX=true,bY=true,bZ=true),RotationFilter=(bX=true,bY=true,bZ=true),ScaleFilter=(bX=True,bY=True,bZ=True))
    Parents=((Item=(Type=Bone,Name="calf_r"),Weight=0),(Item=(Type=Bone,Name="pelvis"),Weight=1))
    Weight=@n9.OutputParam
   )
   > &RigUnit_ParentConstraint(
    AdvancedSettings=(RotationOrderForFilter=XZY)
    Child=(Type=Bone,Name="dyn_SkirtH04")
    Filter=(TranslationFilter=(bX=false,bY=false,bZ=false),RotationFilter=(bX=true,bY=true,bZ=true),ScaleFilter=(bX=True,bY=True,bZ=True))
    Parents=((Item=(Type=Bone,Name="dyn_SkirtH03"),Weight=0.5),(Item=(Type=Bone,Name="pelvis"),Weight=0.5))
    Weight=@n9.OutputParam
   )
   > &RigUnit_AimConstraintLocalSpaceOffset(
    AdvancedSettings=(DebugSettings=(bEnabled=false,Scale=10,WorldOffset=(Rotation=(X=0,Y=0,Z=0,W=1),Translation=(X=0,Y=0,Z=0),Scale3D=(X=1,Y=1,Z=1))),RotationOrderForFilter=XZY)
    Child=(Type=Control,Name="thigh_l_ctrl")
    Filter=(bX=True,bY=True,bZ=True)
    Parents=((Item=(Type=Control,Name="foot_l_ctrl"),Weight=1))
    Weight=1
    WorldUp=(Target=(X=0,Y=0,Z=1),Kind=Direction,Space=(Type=None,Name="None"))
    bMaintainOffset=true
   )
   > n10 = &RigUnit_SphericalPoseReader(
    ActiveRegionScaleFactors=(PositiveWidth=3,NegativeWidth=2,PositiveHeight=4,NegativeHeight=2)
    ActiveRegionSize=0.1
    Debug=(bDrawDebug=false,DebugScale=25,DebugSegments=20,DebugThickness=0.25)
    DriverAxis=(X=1,Y=0,Z=0)
    DriverItem=(Type=Control,Name="thigh_l_ctrl")
    FalloffRegionScaleFactors=(PositiveWidth=0.3,NegativeWidth=2,PositiveHeight=1,NegativeHeight=1)
    FalloffSize=0.35
    FlipHeightScaling=false
    FlipWidthScaling=true
    OptionalParentItem=(Type=Bone,Name="pelvis")
    RotationOffset=(X=-220,Y=0,Z=-5)
   )
   > &RigUnit_ParentConstraint(
    AdvancedSettings=(RotationOrderForFilter=XZY)
    Child=(Type=Bone,Name="dyn_SkirtH03")
    Filter=(TranslationFilter=(bX=false,bY=false,bZ=false),RotationFilter=(bX=true,bY=true,bZ=true),ScaleFilter=(bX=True,bY=True,bZ=True))
    Parents=((Item=(Type=Bone,Name="thigh_twist_01_r"),Weight=0),(Item=(Type=Bone,Name="pelvis"),Weight=1))
    Weight=@n10.OutputParam
   )
   > &RigUnit_ParentConstraint(
    AdvancedSettings=(RotationOrderForFilter=XZY)
    Child=(Type=Bone,Name="dyn_SkirtE05")
    Filter=(TranslationFilter=(bX=false,bY=false,bZ=false),RotationFilter=(bX=true,bY=true,bZ=true),ScaleFilter=(bX=True,bY=True,bZ=True))
    Parents=((Item=(Type=Bone,Name="dyn_SkirtE05"),Weight=0.5),(Item=(Type=Bone,Name="pelvis"),Weight=0.5))
    Weight=@n10.OutputParam
   )
   > &RigUnit_ModifyTransforms(
    ItemToModify="((Item=(Type=Bone,Name=\"dyn_SkirtH03\"),Transform=(Rotation=(X=0.3769,Y=0.1604,Z=-0.0665,W=0.9098),Translation=(X=0,Y=0,Z=0),Scale3D=(X=1,Y=1,Z=1))),(Item=(Type=Bone,Name=\"dyn_SkirtG03\"),Transform=(Rotation=(X=0.1387,Y=-0.1981,Z=-0.1387,W=0.9603),Translatio ...(truncated)"
    Weight=@n10.OutputParam
   )
   > n11 = &RigUnit_SphericalPoseReader(
    ActiveRegionScaleFactors=(PositiveWidth=1,NegativeWidth=0.5,PositiveHeight=1,NegativeHeight=0.8)
    ActiveRegionSize=0.1
    Debug=(bDrawDebug=false,DebugScale=25,DebugSegments=20,DebugThickness=0.25)
    DriverAxis=(X=1,Y=0,Z=0)
    DriverItem=(Type=Control,Name="thigh_r_ctrl")
    FalloffRegionScaleFactors=(PositiveWidth=1,NegativeWidth=2,PositiveHeight=1,NegativeHeight=2)
    FalloffSize=0.35
    FlipHeightScaling=false
    FlipWidthScaling=true
    OptionalParentItem=(Type=Bone,Name="pelvis")
    RotationOffset=(X=40,Y=10,Z=45)
   )
   > &RigUnit_ParentConstraint(
    AdvancedSettings=(RotationOrderForFilter=XZY)
    Child=(Type=Bone,Name="dyn_SkirtJ03")
    Filter=(TranslationFilter=(bX=true,bY=true,bZ=true),RotationFilter=(bX=True,bY=True,bZ=True),ScaleFilter=(bX=True,bY=True,bZ=True))
    Parents=((Item=(Type=Bone,Name="thigh_twist_01_r"),Weight=0.8),(Item=(Type=Bone,Name="pelvis"),Weight=0.2))
    Weight=@n11.OutputParam
   )
   > &RigUnit_ModifyTransforms(
    ItemToModify="((Item=(Type=Bone,Name=\"dyn_SkirtG03\"),Transform=(Rotation=(X=0.0868,Y=-0.0868,Z=0.0076,W=0.9924),Translation=(X=0,Y=0,Z=0),Scale3D=(X=1,Y=1,Z=1))),(Item=(Type=Bone,Name=\"dyn_SkirtH03\"),Transform=(Rotation=(X=0.0868,Y=-0.0868,Z=0.0076,W=0.9924),Translation ...(truncated)"
    Weight=@n11.OutputParam
   )
   > n12 = &RigUnit_SphericalPoseReader(
    ActiveRegionScaleFactors=(PositiveWidth=3,NegativeWidth=2,PositiveHeight=2,NegativeHeight=2)
    ActiveRegionSize=0.1
    Debug=(bDrawDebug=false,DebugScale=25,DebugSegments=20,DebugThickness=0.25)
    DriverAxis=(X=1,Y=0,Z=0)
    DriverItem=(Type=Bone,Name="spine_02")
    FalloffRegionScaleFactors=(PositiveWidth=1,NegativeWidth=1,PositiveHeight=1,NegativeHeight=1)
    FalloffSize=0.45
    FlipWidthScaling=false
    OptionalParentItem=(Type=Bone,Name="collision")
    RotationOffset=(X=90,Y=0,Z=5)
   )
   > &RigUnit_ParentConstraint(
    AdvancedSettings=(RotationOrderForFilter=XZY)
    Child=(Type=Bone,Name="dyn_SkirtG04")
    Filter=(TranslationFilter=(bX=false,bY=false,bZ=false),RotationFilter=(bX=True,bY=True,bZ=True),ScaleFilter=(bX=True,bY=True,bZ=True))
    Parents=((Item=(Type=Bone,Name="dyn_SkirtG04"),Weight=0.5),(Item=(Type=Bone,Name="pelvis"),Weight=0.5))
    Weight=@n12.OutputParam
   )
   > &RigUnit_ModifyTransforms(
    ItemToModify="((Item=(Type=Bone,Name=\"dyn_SkirtH04\"),Transform=(Rotation=(X=0,Y=-0.1736,Z=0,W=0.9848),Translation=(X=0,Y=0,Z=0),Scale3D=(X=1,Y=1,Z=1))),(Item=(Type=Bone,Name=\"dyn_SkirtG04\"),Transform=(Rotation=(X=0,Y=0,Z=0,W=1),Translation=(X=0,Y=0,Z=0),Scale3D=(X=1,Y=1 ...(truncated)"
    ItemToModify.1.Transform.Rotation=&RigVMFunction_MathQuaternionFromEuler(Euler=(X=0,Y=0,Z=-10)).Result
    Weight=@n12.OutputParam
   )
  ]}
]
hierarchy [
 root {loc=<0 0 0>, rot=<0 0 0 1>}
 pelvis {parent=root, loc=<0 -0.7071 93.1191>, rot=<0.0127 0.707 -0.0127 -0.707>}
 spine_01 {parent=pelvis, loc=<6.7494 -0.6388 0>, rot=<0 0 0.034 -0.9994>}
 spine_02 {parent=spine_01, loc=<7.2518 -0.035 0>, rot=<0 0 0.0238 -0.9997>}
 spine_03 {parent=spine_02, loc=<6.0104 0 0>, rot=<0 0 0.0149 -0.9999>}
 spine_04 {parent=spine_03, loc=<8.1024 0 0>, rot=<0 0 -0.0132 -0.9999>}
 spine_05 {parent=spine_04, loc=<10.0146 0 0>, rot=<0 0 -0.0015 -1>}
 neck_01 {parent=spine_05, loc=<11.3138 0.1529 0>, rot=<0 0 -0.2003 -0.9797>}
 neck_02 {parent=neck_01, loc=<4.5627 0 0>, rot=<0 0 0.0504 -0.9987>}
 head {parent=neck_02, loc=<5.3914 0 0>, rot=<0 0 0.115 -0.9934>}
 clavicle_l {parent=spine_05, loc=<4.3575 -0.1466 -2.3223>, rot=<0.0465 -0.7289 -0.009 -0.6829>}
 upperarm_l {parent=clavicle_l, loc=<15.5111 0 0>, rot=<0.0226 -0.3234 -0.0088 -0.9459>}
 lowerarm_l {parent=upperarm_l, loc=<26.6514 0 0>, rot=<0.0391 -0.0099 -0.2453 -0.9686>}
 hand_l {parent=lowerarm_l, loc=<22.3491 0 0>, rot=<-0.7081 -0.0507 0.0975 -0.6975>}
 thumb_01_l {parent=hand_l, loc=<1.3118 -0.3191 -2.459>, rot=<0.6546 -0.4042 0.0664 -0.6354>}
 thumb_02_l {parent=thumb_01_l, loc=<5.9336 -0.0448 0.0401>, rot=<-0.0037 0.0213 0.298 -0.9543>}
 thumb_03_l {parent=thumb_02_l, loc=<2.2223 0.3682 -0.0907>, rot=<0.0118 -0.013 0.0066 -0.9998>}
 index_metacarpal_l {parent=hand_l, loc=<1.6535 0.0183 -2.96>, rot=<0.0158 -0.0124 -0.0173 -0.9996>}
 index_01_l {parent=index_metacarpal_l, loc=<8.3008 0.0357 -0.0524>, rot=<-0.0044 0.0485 0.0924 -0.9945>}
 index_02_l {parent=index_01_l, loc=<3.3587 -0.3004 0.0452>, rot=<-0.002 0.0027 0.3225 -0.9466>}
 index_03_l {parent=index_02_l, loc=<2.5578 -0.123 -0.0252>, rot=<-0.002 0.0067 0.2558 -0.9667>}
 middle_metacarpal_l {parent=hand_l, loc=<2.1856 -0.1725 -0.5984>, rot=<0.006 0.0168 -0.0329 -0.9993>}
 middle_01_l {parent=middle_metacarpal_l, loc=<7.3275 -0.2447 0.0129>, rot=<-0.0208 0.0042 0.0858 -0.9961>}
 middle_02_l {parent=middle_01_l, loc=<4.0123 0.1401 0.1175>, rot=<-0.0218 0.0258 0.3512 -0.9357>}
 middle_03_l {parent=middle_02_l, loc=<2.6077 0.0534 0.1373>, rot=<0.0082 0.0188 0.1708 -0.9851>}
 ring_metacarpal_l {parent=hand_l, loc=<2.2074 -0.3625 1.3638>, rot=<-0.0284 0.0722 -0.0458 -0.9959>}
 ring_01_l {parent=ring_metacarpal_l, loc=<7.0295 -0.6197 -0.1446>, rot=<-0.0489 -0.0252 0.1892 -0.9804>}
 ring_02_l {parent=ring_01_l, loc=<3.6174 0.4647 -0.1027>, rot=<0 0 0.2806 -0.9598>}
 ring_03_l {parent=ring_02_l, loc=<2.4152 0.114 0.0199>, rot=<0 0 0.1189 -0.9929>}
 pinky_metacarpal_l {parent=hand_l, loc=<1.7616 -0.9753 3.3814>, rot=<-0.0397 0.0814 -0.069 -0.9935>}
 pinky_01_l {parent=pinky_metacarpal_l, loc=<7.1709 -0.4831 -0.1094>, rot=<-0.0654 -0.0035 0.2242 -0.9723>}
 pinky_02_l {parent=pinky_01_l, loc=<2.8274 0.0893 -0.1043>, rot=<0.0037 0.0011 0.2854 -0.9584>}
 pinky_03_l {parent=pinky_02_l, loc=<2.2046 0.0307 -0.0078>, rot=<-0.0002 0.0098 0.1445 -0.9895>}
 weapon_l {parent=hand_l, loc=<-0.1974 -3.4111 0.3811>, rot=<0 0 0 1>}
 lowerarm_twist_02_l {parent=lowerarm_l, loc=<9.6199 0 0>, rot=<0 0 0 -1>}
 lowerarm_twist_01_l {parent=lowerarm_l, loc=<17.8549 0 0>, rot=<0 0 0 -1>}
 lowerarm_l_front {parent=lowerarm_l, loc=<0 0.5246 0>, rot=<0 0 0 -1>}
 lowerarm_l_back {parent=lowerarm_l, loc=<0 -0.4766 0>, rot=<0 0 0 -1>}
 scale_lowerarm_l {parent=lowerarm_l, loc=<0 0 0>, rot=<0 0 0 1>}
 hand_l_Sleeves1 {parent=lowerarm_l, loc=<16.439 0 0>, rot=<0 0 0 1>}
 hand_l_Sleeves_01 {parent=hand_l_Sleeves1, loc=<6.1125 2.9938 0.0239>, rot=<0 0 0 1>}
 hand_l_Sleeves_02 {parent=hand_l_Sleeves1, loc=<-3.1998 0.052 -4.1704>, rot=<0 0 0 1>}
 hand_l_Sleeves_02_01 {parent=hand_l_Sleeves_02, loc=<0.0533 0.8912 -2.6501>, rot=<0 0 0 1>}
 hand_l_Sleeves_03 {parent=hand_l_Sleeves1, loc=<5.4627 -3.9744 -0.8673>, rot=<0 0 0 1>}
 hand_l_Sleeves_04 {parent=hand_l_Sleeves1, loc=<6.1997 -0.3873 2.7024>, rot=<0 0 0 1>}
 hand_l_Sleeves2 {parent=lowerarm_l, loc=<9.6199 0 0>, rot=<0 0 0 1>}
 upperarm_twist_01_l {parent=upperarm_l, loc=<0.0041 0 0>, rot=<0 0 0 1>}
 up_uppelalm_l {parent=upperarm_twist_01_l, loc=<-0.0044 -0.0004 0.5003>, rot=<0 0 0 1>}
 flont_uppelalm_l {parent=upperarm_twist_01_l, loc=<-0.0042 0.4996 0.0001>, rot=<0 0 0 1>}
 upperarm_twist_02_l {parent=upperarm_l, loc=<13.7941 0 0>, rot=<0 0 0 1>}
 scale_upperarm_l {parent=upperarm_l, loc=<0 0 0>, rot=<0 0 0 1>}
 clavicle_r {parent=spine_05, loc=<4.3575 -0.1466 2.3223>, rot=<0.7289 0.0465 -0.6829 0.009>}
 upperarm_r {parent=clavicle_r, loc=<-15.5111 0 0>, rot=<0.0226 -0.3234 -0.0088 -0.9459>}
 lowerarm_r {parent=upperarm_r, loc=<-26.6514 0 0>, rot=<0.0391 -0.0099 -0.2453 -0.9686>}
 hand_r {parent=lowerarm_r, loc=<-22.3491 0 0>, rot=<-0.7081 -0.0507 0.0975 -0.6975>}
 weapon_r {parent=hand_r, loc=<0.1974 3.4111 -0.3811>, rot=<0 0 0 1>}
 thumb_01_r {parent=hand_r, loc=<-1.3119 0.3192 2.459>, rot=<0.6546 -0.4042 0.0664 -0.6354>}
 thumb_02_r {parent=thumb_01_r, loc=<-5.9336 0.0448 -0.0401>, rot=<-0.0037 0.0213 0.298 -0.9543>}
 thumb_03_r {parent=thumb_02_r, loc=<-2.2223 -0.3682 0.0907>, rot=<0.0118 -0.013 0.0066 -0.9998>}
 index_metacarpal_r {parent=hand_r, loc=<-1.6534 -0.0181 2.9601>, rot=<0.0158 -0.0124 -0.0173 -0.9996>}
 index_01_r {parent=index_metacarpal_r, loc=<-8.3008 -0.0357 0.0524>, rot=<-0.0044 0.0485 0.0924 -0.9945>}
 index_02_r {parent=index_01_r, loc=<-3.3589 0.3003 -0.0453>, rot=<-0.002 0.0027 0.3225 -0.9466>}
 index_03_r {parent=index_02_r, loc=<-2.5578 0.123 0.0252>, rot=<-0.002 0.0067 0.2558 -0.9667>}
 middle_metacarpal_r {parent=hand_r, loc=<-2.1856 0.1726 0.5984>, rot=<0.006 0.0168 -0.0329 -0.9993>}
 middle_01_r {parent=middle_metacarpal_r, loc=<-7.3273 0.2447 -0.0129>, rot=<-0.0208 0.0042 0.0858 -0.9961>}
 middle_02_r {parent=middle_01_r, loc=<-4.0142 -0.1411 -0.07>, rot=<-0.0217 0.0259 0.3512 -0.9357>}
 middle_03_r {parent=middle_02_r, loc=<-2.6074 -0.0534 -0.1312>, rot=<0.0083 0.0188 0.1708 -0.9851>}
 ring_metacarpal_r {parent=hand_r, loc=<-2.2073 0.3627 -1.3638>, rot=<-0.0284 0.0722 -0.0458 -0.9959>}
 ring_01_r {parent=ring_metacarpal_r, loc=<-7.0297 0.6197 0.1446>, rot=<-0.0489 -0.0252 0.1892 -0.9804>}
 ring_02_r {parent=ring_01_r, loc=<-3.6171 -0.4648 0.1027>, rot=<0 0 0.2806 -0.9598>}
 ring_03_r {parent=ring_02_r, loc=<-2.4152 -0.114 -0.0199>, rot=<0 0 0.1189 -0.9929>}
 pinky_metacarpal_r {parent=hand_r, loc=<-1.7617 0.9753 -3.3814>, rot=<-0.0397 0.0814 -0.069 -0.9935>}
 pinky_01_r {parent=pinky_metacarpal_r, loc=<-7.1707 0.4832 0.1094>, rot=<-0.0654 -0.0035 0.2242 -0.9723>}
 pinky_02_r {parent=pinky_01_r, loc=<-2.8274 -0.0893 0.1043>, rot=<0.0037 0.0011 0.2854 -0.9584>}
 pinky_03_r {parent=pinky_02_r, loc=<-2.2045 -0.031 0.0078>, rot=<-0.0002 0.0098 0.1445 -0.9895>}
 lowerarm_twist_02_r {parent=lowerarm_r, loc=<-9.6199 0 0>, rot=<0 0 0 -1>}
 lowerarm_twist_01_r {parent=lowerarm_r, loc=<-17.8549 0 0>, rot=<0 0 0 -1>}
 lowerarm_r_front {parent=lowerarm_r, loc=<0 -0.4766 0>, rot=<0 0 0 1>}
 lowerarm_r_back {parent=lowerarm_r, loc=<0 0.5246 0>, rot=<0 0 0 1>}
 scale_lowerarm_r {parent=lowerarm_r, loc=<0 0 0>, rot=<0 0 0 1>}
 hand_r_Sleeves1 {parent=lowerarm_r, loc=<-17.855 0 0>, rot=<0 0 0 1>}
 hand_r_Sreeves_01 {parent=hand_r_Sleeves1, loc=<-4.6965 -2.9938 -0.0238>, rot=<0 0 0 1>}
 hand_r_Sreeves_02 {parent=hand_r_Sleeves1, loc=<4.6158 -0.052 4.1705>, rot=<0 0 0 1>}
 hand_r_Sreeves_02_01 {parent=hand_r_Sreeves_02, loc=<-0.0533 -0.8912 2.6501>, rot=<0 0 0 1>}
 hand_r_Sreeves_03 {parent=hand_r_Sleeves1, loc=<-4.0465 3.9743 0.8671>, rot=<0 0 0 1>}
 hand_r_Sreeves_04 {parent=hand_r_Sleeves1, loc=<-4.7835 0.3872 -2.7025>, rot=<0 0 0 1>}
 hand_r_Sleeves2 {parent=lowerarm_r, loc=<-9.6199 0 0>, rot=<0 0 0 1>}
 upperarm_twist_01_r {parent=upperarm_r, loc=<-0.0041 0 0>, rot=<0 0 0 -1>}
 up_upperarm_r {parent=upperarm_twist_01_r, loc=<0.0041 0.0004 -0.5>, rot=<0 0 0 1>}
 front_upperarm_r {parent=upperarm_twist_01_r, loc=<0.0041 -0.4996 0>, rot=<0 0 0 1>}
 upperarm_twist_02_r {parent=upperarm_r, loc=<-13.7941 0 0>, rot=<0 0 0 -1>}
 scale_upperarm_r {parent=upperarm_r, loc=<0 0 0>, rot=<0 0 0 1>}
 dyn_collarB_01 {parent=spine_05, loc=<17.8062 -7.5282 0.2589>, rot=<0.021 0.006 -0.0702 -0.9973>}
 dyn_collarB_02 {parent=dyn_collarB_01, loc=<0.808 0.1455 0>, rot=<0 0 0.0504 -0.9987>}
 dyn_collarC_01 {parent=spine_05, loc=<12.7614 3.9035 11.7671>, rot=<0.6615 0.2807 -0.0135 -0.6953>}
 dyn_collarC_02 {parent=dyn_collarC_01, loc=<0.9476 0.1455 0>, rot=<0 0 0.0504 -0.9987>}
 dyn_collarA_01 {parent=spine_05, loc=<12.3945 3.7671 -11.6806>, rot=<-0.6571 -0.3145 0.0079 -0.685>}
 dyn_collarA_02 {parent=dyn_collarA_01, loc=<0.9476 0.1455 0>, rot=<0 0 0.0504 -0.9987>}
 dyn_collarD_01 {parent=spine_05, loc=<1.9721 13.1416 8.0918>, rot=<0.0284 -0.7065 -0.0284 -0.7065>}
 dyn_collarD_02 {parent=dyn_collarD_01, loc=<0 0 1.0898>, rot=<0 0 0 1>}
 dyn_collarE_01 {parent=spine_05, loc=<1.9721 13.1416 -7.8032>, rot=<0.0284 -0.7065 -0.0284 -0.7065>}
 dyn_collarE_02 {parent=dyn_collarE_01, loc=<0 0 1.0949>, rot=<0 0 0 1>}
 dyn_SkirtJ01 {parent=spine_03, loc=<2.9738 8.8516 12.5153>, rot=<-0.3171 -0.6493 -0.3169 -0.6144>}
 dyn_SkirtJ02 {parent=dyn_SkirtJ01, loc=<0 0 -9.6258>, rot=<0 0 0 1>}
 dyn_SkirtJ03 {parent=dyn_SkirtJ02, loc=<0 0 -9.7266>, rot=<0.0326 0 0 -0.9995>}
 dyn_SkirtJ04 {parent=dyn_SkirtJ03, loc=<0 0 -12.1444>, rot=<0 0 0 1>}
 dyn_SkirtJ05 {parent=dyn_SkirtJ04, loc=<0 0 -16.3764>, rot=<-0.0177 0 0 -0.9998>}
 dyn_SkirtJ06 {parent=dyn_SkirtJ05, loc=<-1.9182 0 -14.7554>, rot=<0 0 0 1>}
 dyn_SkirtJ07 {parent=dyn_SkirtJ06, loc=<-1.6394 0.0478 -15.103>, rot=<0 0 0 1>}
 dyn_SkirtH01 {parent=spine_03, loc=<2.5223 -1.1516 14.864>, rot=<0.4898 0.487 0.5677 0.4479>}
 dyn_SkirtH02 {parent=dyn_SkirtH01, loc=<0 1.1723 -9.6258>, rot=<0 0 0 1>}
 dyn_SkirtH03 {parent=dyn_SkirtH02, loc=<0 1.6157 -9.7266>, rot=<0 0 0 1>}
 dyn_SkirtH04 {parent=dyn_SkirtH03, loc=<0 1.1215 -12.1444>, rot=<0 0 0 1>}
 dyn_SkirtH05 {parent=dyn_SkirtH04, loc=<0 2.1403 -16.3764>, rot=<0 0 0 1>}
 dyn_SkirtH06 {parent=dyn_SkirtH05, loc=<0 3.2818 -14.7554>, rot=<0 0 0 1>}
 dyn_SkirtH07 {parent=dyn_SkirtH06, loc=<0 -0.5572 -12.1241>, rot=<0 0 0 1>}
 dyn_SkirtG01 {parent=spine_03, loc=<4.0426 -10.4402 9.3862>, rot=<0.6945 0.1302 0.699 0.1098>}
 dyn_SkirtG02 {parent=dyn_SkirtG01, loc=<0 1.1723 -9.6258>, rot=<0 0 0 1>}
 dyn_SkirtG03 {parent=dyn_SkirtG02, loc=<0 1.6157 -9.7266>, rot=<0 0 0 1>}
 dyn_SkirtG04 {parent=dyn_SkirtG03, loc=<0 1.1215 -12.1444>, rot=<0 0 0 1>}
 dyn_SkirtG05 {parent=dyn_SkirtG04, loc=<0 2.1403 -16.3764>, rot=<0 0 0 1>}
 dyn_SkirtG06 {parent=dyn_SkirtG05, loc=<0 2.2001 -15.063>, rot=<-0.0597 0 0 -0.9982>}
 dyn_SkirtG07 {parent=dyn_SkirtG06, loc=<0 -0.5572 -12.1241>, rot=<0 0 0 1>}
 dyn_SkirtF01 {parent=spine_03, loc=<4.1109 -12.5804 0.2307>, rot=<0.7066 -0.0019 0.7073 -0.0193>}
 dyn_SkirtF02 {parent=dyn_SkirtF01, loc=<0 1.1723 -9.6258>, rot=<0 0 0 1>}
 dyn_SkirtF03 {parent=dyn_SkirtF02, loc=<0 1.6157 -9.7266>, rot=<0 0 0 1>}
 dyn_SkirtF04 {parent=dyn_SkirtF03, loc=<0 1.1215 -12.1444>, rot=<0 0 0 1>}
 dyn_SkirtF05 {parent=dyn_SkirtF04, loc=<0 2.1403 -16.3764>, rot=<0 0 0 1>}
 dyn_SkirtF06 {parent=dyn_SkirtF05, loc=<0 2.2001 -15.063>, rot=<-0.0597 0 0 -0.9982>}
 dyn_SkirtF07 {parent=dyn_SkirtF06, loc=<0 -0.5572 -12.1241>, rot=<0 0 0 1>}
 dyn_SkirtE01 {parent=spine_03, loc=<4.0426 -10.4402 -9.3862>, rot=<0.699 -0.1098 0.6945 -0.1302>}
 dyn_SkirtE02 {parent=dyn_SkirtE01, loc=<0 1.1723 -9.6258>, rot=<0 0 0 1>}
 dyn_SkirtE03 {parent=dyn_SkirtE02, loc=<0 1.6157 -9.7266>, rot=<0 0 0 1>}
 dyn_SkirtE04 {parent=dyn_SkirtE03, loc=<0 1.1215 -12.1444>, rot=<0 0 0 1>}
 dyn_SkirtE05 {parent=dyn_SkirtE04, loc=<0 2.1403 -16.3764>, rot=<0 0 0 1>}
 dyn_SkirtE06 {parent=dyn_SkirtE05, loc=<0 2.2001 -15.063>, rot=<-0.0597 0 0 -0.9982>}
 dyn_SkirtE07 {parent=dyn_SkirtE06, loc=<0 -0.5572 -12.1241>, rot=<0 0 0 1>}
 dyn_SkirtD01 {parent=spine_03, loc=<2.5223 -1.1516 -14.864>, rot=<0.5677 -0.4479 0.4898 -0.487>}
 dyn_SkirtD02 {parent=dyn_SkirtD01, loc=<0 1.1723 -9.6258>, rot=<0 0 0 1>}
 dyn_SkirtD03 {parent=dyn_SkirtD02, loc=<0 1.6157 -9.7266>, rot=<0 0 0 1>}
 dyn_SkirtD04 {parent=dyn_SkirtD03, loc=<0 1.1215 -12.1444>, rot=<0 0 0 1>}
 dyn_SkirtD05 {parent=dyn_SkirtD04, loc=<0 2.1403 -16.3764>, rot=<0 0 0 1>}
 dyn_SkirtD06 {parent=dyn_SkirtD05, loc=<0 3.2818 -14.7554>, rot=<0 0 0 1>}
 dyn_SkirtD07 {parent=dyn_SkirtD06, loc=<0 -0.5572 -12.1241>, rot=<0 0 0 1>}
 dyn_SkirtB01 {parent=spine_03, loc=<2.9738 8.8516 -12.5153>, rot=<0.3169 -0.6144 0.3171 -0.6493>}
 dyn_SkirtB02 {parent=dyn_SkirtB01, loc=<0 0 -9.6258>, rot=<0 0 0 1>}
 dyn_SkirtB03 {parent=dyn_SkirtB02, loc=<0 0 -9.7266>, rot=<0 0 0 1>}
 dyn_SkirtB04 {parent=dyn_SkirtB03, loc=<0 0 -12.1444>, rot=<0.0374 0 0 -0.9993>}
 dyn_SkirtB05 {parent=dyn_SkirtB04, loc=<0 0 -16.3764>, rot=<0.0074 0 0 -1>}
 dyn_SkirtB06 {parent=dyn_SkirtB05, loc=<1.2862 0 -14.7554>, rot=<0 0 0 1>}
 dyn_SkirtB07 {parent=dyn_SkirtB06, loc=<2.1112 1.0311 -15.103>, rot=<0 0 0 1>}
 dyn_SkirtK01 {parent=spine_03, loc=<1.4908 11.2458 0.2307>, rot=<0.0408 -0.7042 -0.0619 -0.7061>}
 dyn_SkirtK02 {parent=dyn_SkirtK01, loc=<0 1.1723 -9.6258>, rot=<0 0 0 1>}
 dyn_SkirtK03 {parent=dyn_SkirtK02, loc=<0 1.6157 -9.7266>, rot=<0 0 0 1>}
 dyn_SkirtK04 {parent=dyn_SkirtK03, loc=<0 1.1215 -12.1444>, rot=<0 0 0 1>}
 dyn_SkirtK05 {parent=dyn_SkirtK04, loc=<0 1.445 -16.4554>, rot=<-0.0565 0 0 -0.9984>}
 dyn_SkirtK06 {parent=dyn_SkirtK05, loc=<0 -0.5745 -14.3335>, rot=<0.0394 0 0 -0.9992>}
 dyn_SkirtK07 {parent=dyn_SkirtK06, loc=<0 -0.5572 -15.103>, rot=<0 0 0 1>}
 thigh_l {parent=pelvis, loc=<-2.5693 -0.0109 -7.7501>, rot=<0.0012 0.0572 0.0155 -0.9982>}
 calf_l {parent=thigh_l, loc=<-39.0484 0 0>, rot=<0.0006 -0.0161 -0.0391 -0.9991>}
 foot_l {parent=calf_l, loc=<-42.561 0.1077 0.4292>, rot=<0.0418 -0.0455 0.0343 -0.9975>}
 ball_l {parent=foot_l, loc=<-6.1599 13.8264 0.6806>, rot=<0.0001 -0.0001 -0.7063 -0.7079>}
 foot_l_ctrl {controlType=EULER_TRANSFORM, parent=foot_l, type=Control, loc=<0 0 0>, rot=<0 0 0 1>}
 calf_twist_02_l {parent=calf_l, loc=<-18.8064 -0.0967 0.0397>, rot=<0.0237 0.0095 -0.0035 -0.9997>}
 calf_twist_01_l {parent=calf_l, loc=<-35.813 -0.0967 0.0397>, rot=<0.0237 0.0095 -0.0035 -0.9997>}
 calf_l_front {parent=calf_l, loc=<0 1 0>, rot=<0 0 0 1>}
 calf_l_back {parent=calf_l, loc=<0 -1 0>, rot=<0 0 0 1>}
 thigh_twist_01_l {parent=thigh_l, loc=<-21.6303 0 0>, rot=<0.0013 0 0 -1>}
 thigh_l_ctrl {controlType=EULER_TRANSFORM, parent=thigh_l, type=Control, loc=<0 0 0>, rot=<0 0 0 1>}
 thigh_volume_bk_l {parent=thigh_l, loc=<0 -1.5 0>, rot=<0 0 0 1>}
 thigh_volume_fn_l {parent=thigh_l, loc=<0 1.5 0>, rot=<0.9998 0 -0.0217 0>}
 thigh_volume_lt_l1 {parent=thigh_l, loc=<0 0 -1.5>, rot=<-0.7069 -0.0154 0.0154 -0.7069>}
 thigh_r {parent=pelvis, loc=<-2.5691 -0.0109 7.7501>, rot=<0.0572 -0.0012 0.9982 0.0155>}
 calf_r {parent=thigh_r, loc=<39.0487 0 0>, rot=<0.0006 -0.0161 -0.0391 -0.9991>}
 foot_r {parent=calf_r, loc=<42.5612 -0.1077 -0.4292>, rot=<0.0418 -0.0455 0.0343 -0.9975>}
 ball_r {parent=foot_r, loc=<6.1599 -13.8264 -0.6806>, rot=<0.0001 -0.0001 -0.7063 -0.7079>}
 foot_r_ctrl {controlType=EULER_TRANSFORM, parent=foot_r, type=Control, loc=<0 0 0>, rot=<0 0 0 1>}
 calf_twist_02_r {parent=calf_r, loc=<18.8064 0.0967 -0.0397>, rot=<0.0237 0.0095 -0.0035 -0.9997>}
 calf_twist_01_r {parent=calf_r, loc=<35.8132 0.0967 -0.0397>, rot=<0.0237 0.0095 -0.0035 -0.9997>}
 calf_r_front {parent=calf_r, loc=<0 -1 0>, rot=<0 0 0 1>}
 calf_r_back {parent=calf_r, loc=<0 1 0>, rot=<0 0 0 1>}
 thigh_twist_01_r {parent=thigh_r, loc=<21.6303 0 0>, rot=<0.0013 0 0 -1>}
 thigh_r_ctrl {controlType=EULER_TRANSFORM, parent=thigh_r, type=Control, loc=<0 0 0>, rot=<0 0 0 1>}
 thigh_volume_bk_r {parent=thigh_r, loc=<0 1.5 0>, rot=<1 0 0 0>}
 thigh_volume_fn_r {parent=thigh_r, loc=<0 -1.5 0>, rot=<0 0 0 1>}
 thigh_volume_rt_r {parent=thigh_r, loc=<0 0 1.5>, rot=<0.7071 0 0 -0.7071>}
 attach {parent=root, loc=<0 0 0>, rot=<0 0 0 1>}
 ik_hand_root {parent=root, loc=<0 0 0>, rot=<0 0 0 1>}
 ik_hand_gun {parent=ik_hand_root, loc=<-51.9727 4.0883 102.6935>, rot=<0.6084 0.1793 0.4613 -0.6205>}
 ik_hand_r {parent=ik_hand_gun, loc=<0 0 0>, rot=<0 0 0 1>}
 ik_hand_l {parent=ik_hand_gun, loc=<53.0285 82.1781 35.2051>, rot=<-0.5102 -0.7906 -0.3387 0>}
 ik_foot_root {parent=root, loc=<0 0 0>, rot=<0 0 0 1>}
 ik_foot_l {parent=ik_foot_root, loc=<15.2608 -4.4176 9.4432>, rot=<0.0357 0.7045 0.0284 -0.7082>}
 ik_foot_r {parent=ik_foot_root, loc=<-15.2608 -4.4176 9.4432>, rot=<0.7082 0.0284 -0.7045 0.0357>}
 collision {parent=root, loc=<0 0 0>, rot=<0 0 0 1>}
 collision_01 {parent=root, loc=<0 0 0>, rot=<0 0 0 1>}
 hand_l_Sreeves_gravity {parent=root, loc=<0 0 0>, rot=<0 0 0 1>}
 hand_r_Sreeves_gravity {parent=root, loc=<0 0 0>, rot=<0 0 0 1>}
]
