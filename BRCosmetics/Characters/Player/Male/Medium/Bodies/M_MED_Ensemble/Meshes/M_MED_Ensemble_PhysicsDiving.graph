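# Physics asset for the M_MED_Ensemble male cosmetic body, named for diving, that defines secondary physics for its long skirt and sleeves.
# Seven skirt chains (B, D–H, J) hang from spine_01, with cross-constraints between neighboring columns forming a cloth-like lattice, and two locked joints in chain F.
# Sleeve chains attach to forearm twist bodies, which link to spine_01 with free rotation.
# Pelvis, spine, thigh and calf capsules supply body shapes.
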PhysicsAsset /BRCosmetics/Characters/Player/Male/Medium/Bodies/M_MED_Ensemble/Meshes/M_MED_Ensemble_PhysicsDiving
bodies {
 calf_l [{shapeType=Capsule, length=45.6252, radius=10.8103, shapeName=calf_l_capsule, center=<-25.2542 -0.0002 -0.1341>, rotation=<rot -90 0 0>}]
 calf_r [{shapeType=Capsule, length=45.6252, radius=10.8103, shapeName=calf_l_capsule, center=<25.1541 -0.0002 -1.8062>, rotation=<rot -90 0 0>}]
 dyn_SkirtB01 [{
   shapeType=Capsule
   length=0.5
   radius=0.5
   shapeName=dyn_SkirtJ01_capsule
   center=<0.0023 0.2489 -4.373>
   rotation=<rot -0.6229 150.9391 -179.654>
  }]
 dyn_SkirtB02 [{shapeType=Capsule, length=8.75, radius=3.25, shapeName=dyn_SkirtH02_capsule, center=<0.0001 1.1859 -4.8124>, rotation=<rot 0 0 -7.0095>}]
 dyn_SkirtB03 [{
   shapeType=Capsule
   length=8.1287
   radius=4.341
   shapeName=dyn_SkirtB03_capsule
   center=<-0.6475 1.4413 -3.3255>
   rotation=<rot -85.5373 0.0012 32.1037>
  }]
 dyn_SkirtB04 [{
   shapeType=Capsule
   length=18.7285
   radius=11.3147
   shapeName=dyn_SkirtB04_capsule
   center=<-2.401 6.3782 -9.4916>
   rotation=<rot -0.2657 -72.4065 -90.0841>
  }]
 dyn_SkirtB05 [{
   shapeType=Capsule
   length=19.8573
   radius=10.1858
   shapeName=dyn_SkirtB05_capsule
   center=<-2.401 6.3782 -9.4916>
   rotation=<rot -0.2658 -72.4065 -90.0841>
  }]
 dyn_SkirtB06 [{
   shapeType=Capsule
   length=17.7626
   radius=10.5114
   shapeName=dyn_SkirtB06_capsule
   center=<-4.0653 6.3661 -10.7332>
   rotation=<rot -0.2657 -72.4065 -90.0841>
  }]
 dyn_SkirtB07 [{
   shapeType=Capsule
   length=19.5757
   radius=11.1526
   shapeName=dyn_SkirtB06_capsule
   center=<-2.4211 3.7453 -7.2192>
   rotation=<rot -0.2657 -72.4065 -90.0841>
  }]
 dyn_SkirtD01 [{shapeType=Capsule, length=9, radius=3, shapeName=dyn_SkirtD01_capsule, center=<0 0 -4.9582>}]
 dyn_SkirtD02 [{shapeType=Capsule, length=13.4554, radius=4.3793, shapeName=dyn_SkirtD02_capsule, center=<0 0.993 0>, rotation=<rot 90 0 0>}]
 dyn_SkirtD03 [{
   shapeType=Capsule
   length=21.2998
   radius=8.7433
   shapeName=dyn_SkirtD03_capsule
   center=<0.0419 6.1891 -5.9384>
   rotation=<rot -0.2785 -90.3896 -89.9979>
  }]
 dyn_SkirtD04 [{
   shapeType=Capsule
   length=21.2998
   radius=8.7433
   shapeName=dyn_SkirtD04_capsule
   center=<0.0419 6.1891 -5.9384>
   rotation=<rot -0.2785 -90.3896 -89.9979>
  }]
 dyn_SkirtD05 [{
   shapeType=Capsule
   length=21.2998
   radius=8.7433
   shapeName=dyn_SkirtD05_capsule
   center=<0.0419 6.1891 -5.9384>
   rotation=<rot -0.2785 -90.3896 -89.9979>
  }]
 dyn_SkirtD06 [{
   shapeType=Capsule
   length=21.2998
   radius=8.7433
   shapeName=dyn_SkirtD06_capsule
   center=<0.0419 6.1891 -5.9384>
   rotation=<rot -0.2785 -90.3896 -89.9979>
  }]
 dyn_SkirtD07 [{
   shapeType=Capsule
   length=21.2998
   radius=8.7433
   shapeName=dyn_SkirtD07_capsule
   center=<0.0419 6.1891 -5.9384>
   rotation=<rot -0.2785 -90.3896 -89.9979>
  }]
 dyn_SkirtE01 [{shapeType=Capsule, length=8.5, radius=3.5, shapeName=dyn_SkirtE01_capsule, center=<0 0 -4.9582>, rotation=<rot 0 0 -6.106>}]
 dyn_SkirtE02 [{
   shapeType=Capsule
   length=3.8716
   radius=8.1284
   shapeName=dyn_SkirtE02_capsule
   center=<0 4.4233 -4.485>
   rotation=<rot 83.066 -118.404 -118.2284>
  }]
 dyn_SkirtE03 [{
   shapeType=Capsule
   length=22.6447
   radius=7.3985
   shapeName=dyn_SkirtE03_capsule
   center=<0.0332 4.9154 -5.9446>
   rotation=<rot -0.2784 -90.3896 -89.998>
  }]
 dyn_SkirtE04 [{
   shapeType=Capsule
   length=23.5
   radius=6.5431
   shapeName=dyn_SkirtE03_capsule
   center=<0.0353 5.2177 -5.9432>
   rotation=<rot -0.2784 -90.3896 -89.998>
  }]
 dyn_SkirtE05 [{
   shapeType=Capsule
   length=21.9702
   radius=8.0729
   shapeName=dyn_SkirtE05_capsule
   center=<0.0369 5.4625 -5.942>
   rotation=<rot -0.2784 -90.3896 -89.998>
  }]
 dyn_SkirtE06 [{
   shapeType=Capsule
   length=20.2254
   radius=9.8178
   shapeName=dyn_SkirtE06_capsule
   center=<0.0398 5.8833 -5.9399>
   rotation=<rot -0.2784 -90.3897 -89.998>
  }]
 dyn_SkirtE07 [{
   shapeType=Capsule
   length=20.2254
   radius=9.8178
   shapeName=dyn_SkirtE07_capsule
   center=<0.0398 5.8833 -5.9399>
   rotation=<rot -0.2784 -90.3897 -89.998>
  }]
 dyn_SkirtF01 [{shapeType=Capsule, length=8.5, radius=3.5, shapeName=dyn_SkirtF01_capsule, center=<0 0 -4.9582>, rotation=<rot 0 0 -6.1057>}]
 dyn_SkirtF02 [{
   shapeType=Capsule
   length=3.5343
   radius=8.4657
   shapeName=dyn_SkirtF02_capsule
   center=<0 4.4233 -4.4851>
   rotation=<rot 83.7383 -77.1354 -77.2096>
  }]
 dyn_SkirtF03 [{
   shapeType=Capsule
   length=22.6447
   radius=7.3985
   shapeName=dyn_SkirtF03_capsule
   center=<0.0332 4.9154 -5.9446>
   rotation=<rot -0.2784 -90.3896 -89.9981>
  }]
 dyn_SkirtF04 [{
   shapeType=Capsule
   length=23.5
   radius=6.5431
   shapeName=dyn_SkirtF04_capsule
   center=<0.0353 5.2176 -5.9432>
   rotation=<rot -0.2784 -90.3896 -89.9981>
  }]
 dyn_SkirtF05 [{
   shapeType=Capsule
   length=21.9702
   radius=8.0729
   shapeName=dyn_SkirtF05_capsule
   center=<0.037 5.4625 -5.942>
   rotation=<rot -0.2784 -90.3896 -89.9981>
  }]
 dyn_SkirtF06 [{
   shapeType=Capsule
   length=20.2254
   radius=9.8178
   shapeName=dyn_SkirtF06_capsule
   center=<0.0398 5.8833 -5.9399>
   rotation=<rot -0.2784 -90.3896 -89.9981>
  }]
 dyn_SkirtF07 [{
   shapeType=Capsule
   length=18.0434
   radius=11.9997
   shapeName=dyn_SkirtF07_capsule
   center=<0.0398 5.8903 -7.3742>
   rotation=<rot -0.2784 -90.3896 -89.9981>
  }]
 dyn_SkirtG01 [{shapeType=Capsule, length=8.5, radius=3.5, shapeName=dyn_SkirtG01_capsule, center=<0 0 -4.9582>, rotation=<rot 0 0 -6.1057>}]
 dyn_SkirtG02 [{
   shapeType=Capsule
   length=3.8716
   radius=8.1284
   shapeName=dyn_SkirtG02_capsule
   center=<0 4.4233 -4.4851>
   rotation=<rot 83.8498 -83.075 -83.1145>
  }]
 dyn_SkirtG03 [{
   shapeType=Capsule
   length=22.6447
   radius=7.3985
   shapeName=dyn_SkirtG03_capsule
   center=<0.0332 4.9154 -5.9446>
   rotation=<rot -0.2784 -90.3896 -89.9982>
  }]
 dyn_SkirtG04 [{
   shapeType=Capsule
   length=23.5
   radius=6.5431
   shapeName=dyn_SkirtG04_capsule
   center=<0.0353 5.2177 -5.9432>
   rotation=<rot -0.2784 -90.3896 -89.9982>
  }]
 dyn_SkirtG05 [{
   shapeType=Capsule
   length=21.9702
   radius=8.0729
   shapeName=dyn_SkirtG05_capsule
   center=<0.037 5.4625 -5.942>
   rotation=<rot -0.2784 -90.3896 -89.9982>
  }]
 dyn_SkirtG06 [{
   shapeType=Capsule
   length=20.2254
   radius=9.8178
   shapeName=dyn_SkirtG06_capsule
   center=<0.0398 5.8833 -5.9399>
   rotation=<rot -0.2784 -90.3896 -89.9982>
  }]
 dyn_SkirtG07 [{
   shapeType=Capsule
   length=20.2254
   radius=9.8178
   shapeName=dyn_SkirtG07_capsule
   center=<0.0398 5.8833 -5.9399>
   rotation=<rot -0.2784 -90.3896 -89.9982>
  }]
 dyn_SkirtH01 [{shapeType=Capsule, length=8.75, radius=3.25, shapeName=dyn_SkirtH01_capsule, center=<0 0 -4.9582>, rotation=<rot 0 0 -9.9728>}]
 dyn_SkirtH02 [{shapeType=Capsule, length=8.75, radius=3.25, shapeName=dyn_SkirtH02_capsule, center=<0.0001 1.1859 -4.8124>, rotation=<rot 0 0 -7.0095>}]
 dyn_SkirtH03 [{
   shapeType=Capsule
   length=21.2998
   radius=8.7433
   shapeName=dyn_SkirtH03_capsule
   center=<0.0419 6.1891 -5.9384>
   rotation=<rot -0.2785 -90.3897 -89.9983>
  }]
 dyn_SkirtH04 [{
   shapeType=Capsule
   length=21.2998
   radius=8.7433
   shapeName=dyn_SkirtH04_capsule
   center=<0.0419 6.1891 -5.9384>
   rotation=<rot -0.2785 -90.3897 -89.9983>
  }]
 dyn_SkirtH05 [{
   shapeType=Capsule
   length=21.2998
   radius=8.7433
   shapeName=dyn_SkirtH05_capsule
   center=<0.0419 6.1891 -5.9384>
   rotation=<rot -0.2785 -90.3897 -89.9983>
  }]
 dyn_SkirtH06 [{
   shapeType=Capsule
   length=21.2998
   radius=8.7433
   shapeName=dyn_SkirtH06_capsule
   center=<0.0419 6.1891 -5.9384>
   rotation=<rot -0.2785 -90.3897 -89.9983>
  }]
 dyn_SkirtH07 [{
   shapeType=Capsule
   length=21.2998
   radius=8.7433
   shapeName=dyn_SkirtH07_capsule
   center=<0.0419 6.1891 -5.9384>
   rotation=<rot -0.2785 -90.3897 -89.9983>
  }]
 dyn_SkirtJ01 [{shapeType=Capsule, length=5.5973, radius=0.5, shapeName=dyn_SkirtJ01_capsule, center=<0 0 -4.9582>, rotation=<rot -66.195 -91.0856 91.1865>}]
 dyn_SkirtJ02 [{
   shapeType=Capsule
   length=9.75
   radius=0.5
   shapeName=dyn_SkirtJ02_capsule
   center=<-0.1527 0 -9.6032>
   rotation=<rot -89.0893 179.9948 143.6047>
  }]
 dyn_SkirtJ03 [{
   shapeType=Capsule
   length=10.517
   radius=5.1156
   shapeName=dyn_SkirtJ03_capsule
   center=<0.0507 -0.173 1.4791>
   rotation=<rot 0.1934 -46.1744 -3.2224>
  }]
 dyn_SkirtJ04 [{
   shapeType=Capsule
   length=18.905
   radius=11.1382
   shapeName=dyn_SkirtJ04_capsule
   center=<2.4784 6.4809 -9.4908>
   rotation=<rot -0.2205 -130.8216 -89.82>
  }]
 dyn_SkirtJ05 [{
   shapeType=Capsule
   length=19.0798
   radius=10.9633
   shapeName=dyn_SkirtJ05_capsule
   center=<2.4784 6.4809 -9.4908>
   rotation=<rot -0.2205 -130.8213 -89.82>
  }]
 dyn_SkirtJ06 [{
   shapeType=Capsule
   length=18.2626
   radius=10.0114
   shapeName=dyn_SkirtJ06_capsule
   center=<5.7698 5.7552 -10.7356>
   rotation=<rot -0.2204 -130.8213 -89.82>
  }]
 dyn_SkirtJ07 [{
   shapeType=Capsule
   length=20.0757
   radius=10.6526
   shapeName=dyn_SkirtJ06_capsule
   center=<1.428 4.2017 -7.2171>
   rotation=<rot -0.2204 -130.8213 -89.82>
  }]
 hand_l_Sleeves_02 [{
   shapeType=Capsule
   length=11.5297
   radius=7.5919
   shapeName=hand_l_Sleeves_02_capsule
   center=<-0.0986 0.556 -8.2518>
   rotation=<rot 0.0079 73.8949 -179.7854>
  }]
 hand_l_Sleeves_02_01 [{
   shapeType=Capsule
   length=4.7478
   radius=8.7969
   shapeName=hand_l_Sleeves_02_01_capsule
   center=<4.0618 0.1113 -6.4299>
   rotation=<rot 0 12.6445 -10.2901>
  }]
 hand_r_Sreeves_02 [{
   shapeType=Capsule
   length=11.6264
   radius=7.5941
   shapeName=hand_r_Sreeves_02_capsule
   center=<-1.237 -0.2631 8.2177>
   rotation=<rot 0.0079 73.035 -178.8798>
  }]
 hand_r_Sreeves_02_01 [{
   shapeType=Capsule
   length=4.8309
   radius=8.4856
   shapeName=hand_r_Sreeves_02_01_capsule
   center=<-3.0133 1.017 4.711>
   rotation=<rot 0 -0.4685 -4.011>
  }]
 lowerarm_twist_01_l [{
   shapeType=Capsule
   length=3
   radius=5.269
   shapeName=lowerarm_twist_01_l_capsule
   center=<7.6154 -0.3216 -6.7821>
   rotation=<rot 0 -92.9357 -68.4756>
  }]
 lowerarm_twist_01_r [{
   shapeType=Capsule
   length=3
   radius=5.3686
   shapeName=lowerarm_twist_01_r_capsule
   center=<-7.6442 0.4483 6.9188>
   rotation=<rot 0 -92.4645 -69.4683>
  }]
 pelvis [{shapeType=Capsule, length=11.9555, radius=14.8257, shapeName=pelvis_capsule, center=<-12.3584 2.4859 0>}]
 spine_01 [{shapeType=Capsule, length=13.1099, radius=13.2682, shapeName=spine_01_capsule, center=<-6.4214 2.4374 0>}]
 thigh_twist_01_l [{
   shapeType=Capsule
   length=29.7949
   radius=10.4551
   shapeName=thigh_r_capsule
   center=<3.9047 -0.0001 0.6354>
   rotation=<rot -89.1701 -179.9991 179.9991>
  }]
 thigh_twist_01_r [{
   shapeType=Capsule
   length=30.6096
   radius=9.6404
   shapeName=thigh_r_capsule
   center=<-3.4223 0 0.0496>
   rotation=<rot -89.1701 -179.9991 179.9991>
  }]
}
constraints [
 {bone1Name=dyn_SkirtJ01, bone2Name=spine_01, swing1LimitDegrees=30, swing2LimitDegrees=30, twistLimitDegrees=60}
 {bone1Name=dyn_SkirtJ02, bone2Name=dyn_SkirtJ01, swing1LimitDegrees=30, swing2LimitDegrees=30, twistLimitDegrees=60}
 {bone1Name=dyn_SkirtJ03, bone2Name=dyn_SkirtJ02, swing1LimitDegrees=30, swing2LimitDegrees=30, twistLimitDegrees=60}
 {bone1Name=dyn_SkirtJ04, bone2Name=dyn_SkirtJ03, swing1LimitDegrees=10, swing2LimitDegrees=15, twistLimitDegrees=60}
 {bone1Name=dyn_SkirtJ05, bone2Name=dyn_SkirtJ04, swing1LimitDegrees=10, swing2LimitDegrees=15, twistLimitDegrees=60}
 {bone1Name=dyn_SkirtJ06, bone2Name=dyn_SkirtJ05, swing1LimitDegrees=10, swing2LimitDegrees=15, twistLimitDegrees=65}
 {bone1Name=dyn_SkirtH01, bone2Name=spine_01, swing1LimitDegrees=20, swing2LimitDegrees=20, twistLimitDegrees=60}
 {bone1Name=dyn_SkirtH02, bone2Name=dyn_SkirtH01, swing1LimitDegrees=20, swing2LimitDegrees=20, twistLimitDegrees=60}
 {bone1Name=dyn_SkirtH03, bone2Name=dyn_SkirtH02, swing1LimitDegrees=20, swing2LimitDegrees=20, twistLimitDegrees=60}
 {bone1Name=dyn_SkirtH04, bone2Name=dyn_SkirtH03, swing1LimitDegrees=20, swing2LimitDegrees=20, twistLimitDegrees=60}
 {bone1Name=dyn_SkirtH05, bone2Name=dyn_SkirtH04, swing1LimitDegrees=20, swing2LimitDegrees=20, twistLimitDegrees=60}
 {bone1Name=dyn_SkirtH06, bone2Name=dyn_SkirtH05, swing1LimitDegrees=20, swing2LimitDegrees=20, twistLimitDegrees=60}
 {bone1Name=dyn_SkirtG01, bone2Name=spine_01, swing1LimitDegrees=20, swing2LimitDegrees=20, twistLimitDegrees=60}
 {bone1Name=dyn_SkirtG02, bone2Name=dyn_SkirtG01, swing1LimitDegrees=10, swing2LimitDegrees=10, twistLimitDegrees=10}
 {bone1Name=dyn_SkirtG03, bone2Name=dyn_SkirtG02, swing1LimitDegrees=10, swing2LimitDegrees=10, twistLimitDegrees=30}
 {bone1Name=dyn_SkirtG04, bone2Name=dyn_SkirtG03, swing1LimitDegrees=20, swing2LimitDegrees=20, twistLimitDegrees=35}
 {bone1Name=dyn_SkirtG05, bone2Name=dyn_SkirtG04, swing1LimitDegrees=20, swing2LimitDegrees=20, twistLimitDegrees=35}
 {bone1Name=dyn_SkirtG06, bone2Name=dyn_SkirtG05, swing1LimitDegrees=20, swing2LimitDegrees=20, twistLimitDegrees=30}
 {
  bone1Name=dyn_SkirtF01
  bone2Name=spine_01
  swing1LimitDegrees=60
  swing1Motion=Locked
  swing2LimitDegrees=60
  swing2Motion=Locked
  twistLimitDegrees=60
  twistMotion=Locked
 }
 {bone1Name=dyn_SkirtF02, bone2Name=dyn_SkirtF01, swing1LimitDegrees=10, swing2LimitDegrees=10, twistLimitDegrees=10}
 {
  bone1Name=dyn_SkirtF03
  bone2Name=dyn_SkirtF02
  swing1LimitDegrees=60
  swing1Motion=Locked
  swing2LimitDegrees=60
  swing2Motion=Locked
  twistLimitDegrees=60
  twistMotion=Locked
 }
 {bone1Name=dyn_SkirtF04, bone2Name=dyn_SkirtF03, swing1LimitDegrees=10, swing2LimitDegrees=10, twistLimitDegrees=35}
 {bone1Name=dyn_SkirtF05, bone2Name=dyn_SkirtF04, swing1LimitDegrees=10, swing2LimitDegrees=10, twistLimitDegrees=35}
 {bone1Name=dyn_SkirtF06, bone2Name=dyn_SkirtF05, swing1LimitDegrees=10, swing2LimitDegrees=10, twistLimitDegrees=30}
 {bone1Name=dyn_SkirtE01, bone2Name=spine_01, swing1LimitDegrees=20, swing2LimitDegrees=20, twistLimitDegrees=60}
 {bone1Name=dyn_SkirtE02, bone2Name=dyn_SkirtE01, swing1LimitDegrees=10, swing2LimitDegrees=10, twistLimitDegrees=10}
 {bone1Name=dyn_SkirtE03, bone2Name=dyn_SkirtE02, swing1LimitDegrees=10, swing2LimitDegrees=10, twistLimitDegrees=30}
 {bone1Name=dyn_SkirtE04, bone2Name=dyn_SkirtE03, swing1LimitDegrees=10, swing2LimitDegrees=10, twistLimitDegrees=35}
 {bone1Name=dyn_SkirtE05, bone2Name=dyn_SkirtE04, swing1LimitDegrees=10, swing2LimitDegrees=10, twistLimitDegrees=35}
 {bone1Name=dyn_SkirtE06, bone2Name=dyn_SkirtE05, swing1LimitDegrees=10, swing2LimitDegrees=10, twistLimitDegrees=30}
 {bone1Name=dyn_SkirtD01, bone2Name=spine_01, swing1LimitDegrees=20, swing2LimitDegrees=20, twistLimitDegrees=60}
 {bone1Name=dyn_SkirtD02, bone2Name=dyn_SkirtD01, swing1LimitDegrees=20, swing2LimitDegrees=20, twistLimitDegrees=60}
 {bone1Name=dyn_SkirtD03, bone2Name=dyn_SkirtD02, swing1LimitDegrees=20, swing2LimitDegrees=20, twistLimitDegrees=60}
 {bone1Name=dyn_SkirtD04, bone2Name=dyn_SkirtD03, swing1LimitDegrees=20, swing2LimitDegrees=20, twistLimitDegrees=60}
 {bone1Name=dyn_SkirtD05, bone2Name=dyn_SkirtD04, swing1LimitDegrees=20, swing2LimitDegrees=20, twistLimitDegrees=60}
 {bone1Name=dyn_SkirtD06, bone2Name=dyn_SkirtD05, swing1LimitDegrees=20, swing2LimitDegrees=20, twistLimitDegrees=60}
 {bone1Name=dyn_SkirtB01, bone2Name=spine_01, swing1LimitDegrees=30, swing2LimitDegrees=30, twistLimitDegrees=60}
 {bone1Name=dyn_SkirtB02, bone2Name=dyn_SkirtB01, swing1LimitDegrees=30, swing2LimitDegrees=30, twistLimitDegrees=60}
 {bone1Name=dyn_SkirtB03, bone2Name=dyn_SkirtB02, swing1LimitDegrees=30, swing2LimitDegrees=30, twistLimitDegrees=60}
 {bone1Name=dyn_SkirtB04, bone2Name=dyn_SkirtB03, swing1LimitDegrees=25, swing2LimitDegrees=25, twistLimitDegrees=60}
 {bone1Name=dyn_SkirtB05, bone2Name=dyn_SkirtB04, swing1LimitDegrees=25, swing2LimitDegrees=25, twistLimitDegrees=65}
 {bone1Name=dyn_SkirtB06, bone2Name=dyn_SkirtB05, swing1LimitDegrees=25, swing2LimitDegrees=25, twistLimitDegrees=65}
 {bone1Name=dyn_SkirtH02, bone2Name=dyn_SkirtJ02, swing1LimitDegrees=45, swing2LimitDegrees=45, twistLimitDegrees=45}
 {bone1Name=dyn_SkirtH07, bone2Name=dyn_SkirtH06, swing1LimitDegrees=1, swing2LimitDegrees=10, twistLimitDegrees=45}
 {bone1Name=dyn_SkirtG07, bone2Name=dyn_SkirtG06, swing1LimitDegrees=1, swing2LimitDegrees=10, twistLimitDegrees=15}
 {bone1Name=dyn_SkirtF07, bone2Name=dyn_SkirtF06, swing1LimitDegrees=1, swing2LimitDegrees=10, twistLimitDegrees=15}
 {bone1Name=dyn_SkirtE07, bone2Name=dyn_SkirtE06, swing1LimitDegrees=10, swing2LimitDegrees=10, twistLimitDegrees=15}
 {bone1Name=dyn_SkirtD07, bone2Name=dyn_SkirtD06, swing1LimitDegrees=1, swing2LimitDegrees=10, twistLimitDegrees=45}
 {bone1Name=dyn_SkirtG03, bone2Name=dyn_SkirtF03, swing1LimitDegrees=20, swing2LimitDegrees=20, twistLimitDegrees=20}
 {bone1Name=dyn_SkirtG04, bone2Name=dyn_SkirtF04, swing1LimitDegrees=20, swing2LimitDegrees=20, twistLimitDegrees=20}
 {bone1Name=dyn_SkirtG05, bone2Name=dyn_SkirtF05, swing1LimitDegrees=20, swing2LimitDegrees=20, twistLimitDegrees=20}
 {bone1Name=dyn_SkirtG06, bone2Name=dyn_SkirtF06, swing1LimitDegrees=20, swing2LimitDegrees=20, twistLimitDegrees=20}
 {bone1Name=dyn_SkirtG07, bone2Name=dyn_SkirtF07, swing1LimitDegrees=20, swing2LimitDegrees=20, twistLimitDegrees=20}
 {
  bone1Name=lowerarm_twist_01_l
  bone2Name=spine_01
  swing1LimitDegrees=45
  swing1Motion=Free
  swing2LimitDegrees=45
  swing2Motion=Free
  twistLimitDegrees=45
  twistMotion=Free
 }
 {bone1Name=hand_l_Sleeves_02, bone2Name=lowerarm_twist_01_l, swing1LimitDegrees=1, swing2LimitDegrees=60, twistLimitDegrees=45}
 {
  bone1Name=lowerarm_twist_01_r
  bone2Name=spine_01
  swing1LimitDegrees=45
  swing1Motion=Free
  swing2LimitDegrees=45
  swing2Motion=Free
  twistLimitDegrees=45
  twistMotion=Free
 }
 {bone1Name=hand_r_Sreeves_02, bone2Name=lowerarm_twist_01_r, swing1LimitDegrees=1, swing2LimitDegrees=60, twistLimitDegrees=45}
 {bone1Name=dyn_SkirtB07, bone2Name=dyn_SkirtB06, swing1LimitDegrees=12, swing2LimitDegrees=12, twistLimitDegrees=65}
 {bone1Name=dyn_SkirtJ07, bone2Name=dyn_SkirtJ06, swing1LimitDegrees=10, swing2LimitDegrees=15, twistLimitDegrees=65}
 {bone1Name=dyn_SkirtJ07, bone2Name=dyn_SkirtH07, swing1LimitDegrees=45, swing2LimitDegrees=45, twistLimitDegrees=20}
 {
  bone1Name=dyn_SkirtJ06
  bone2Name=dyn_SkirtH06
  swing1LimitDegrees=45
  swing1Motion=Free
  swing2LimitDegrees=45
  swing2Motion=Free
  twistLimitDegrees=45
  twistMotion=Free
 }
 {
  bone1Name=dyn_SkirtJ05
  bone2Name=dyn_SkirtH05
  swing1LimitDegrees=45
  swing1Motion=Free
  swing2LimitDegrees=45
  swing2Motion=Free
  twistLimitDegrees=45
  twistMotion=Free
 }
 {
  bone1Name=dyn_SkirtJ04
  bone2Name=dyn_SkirtH04
  swing1LimitDegrees=45
  swing1Motion=Free
  swing2LimitDegrees=45
  swing2Motion=Free
  twistLimitDegrees=45
  twistMotion=Free
 }
 {
  bone1Name=dyn_SkirtB07
  bone2Name=dyn_SkirtD07
  swing1LimitDegrees=35
  swing1Motion=Free
  swing2LimitDegrees=30
  swing2Motion=Free
  twistLimitDegrees=20
 }
 {bone1Name=dyn_SkirtB06, bone2Name=dyn_SkirtD06, swing1LimitDegrees=35, swing2LimitDegrees=75, twistLimitDegrees=75}
 {bone1Name=dyn_SkirtB03, bone2Name=dyn_SkirtD03, swing1LimitDegrees=45, swing2LimitDegrees=45, twistLimitDegrees=45}
 {bone1Name=dyn_SkirtD07, bone2Name=dyn_SkirtE07, swing1LimitDegrees=20, swing2LimitDegrees=20, twistLimitDegrees=20}
 {bone1Name=dyn_SkirtD06, bone2Name=dyn_SkirtE06, swing1LimitDegrees=20, swing2LimitDegrees=20, twistLimitDegrees=20}
 {bone1Name=dyn_SkirtD05, bone2Name=dyn_SkirtE05, swing1LimitDegrees=20, swing2LimitDegrees=20, twistLimitDegrees=20}
 {bone1Name=dyn_SkirtD04, bone2Name=dyn_SkirtE04, swing1LimitDegrees=20, swing2LimitDegrees=20, twistLimitDegrees=20}
 {bone1Name=dyn_SkirtD03, bone2Name=dyn_SkirtE03, swing1LimitDegrees=20, swing2LimitDegrees=20, twistLimitDegrees=20}
 {bone1Name=dyn_SkirtE07, bone2Name=dyn_SkirtF07, swing1LimitDegrees=20, swing2LimitDegrees=20, twistLimitDegrees=20}
 {bone1Name=dyn_SkirtE06, bone2Name=dyn_SkirtF06, swing1LimitDegrees=20, swing2LimitDegrees=20, twistLimitDegrees=20}
 {bone1Name=dyn_SkirtE05, bone2Name=dyn_SkirtF05, swing1LimitDegrees=20, swing2LimitDegrees=20, twistLimitDegrees=20}
 {bone1Name=dyn_SkirtE04, bone2Name=dyn_SkirtF04, swing1LimitDegrees=20, swing2LimitDegrees=20, twistLimitDegrees=20}
 {
  bone1Name=dyn_SkirtH07
  bone2Name=dyn_SkirtG07
  swing1LimitDegrees=20
  swing1Motion=Free
  swing2LimitDegrees=20
  swing2Motion=Free
  twistLimitDegrees=20
 }
 {
  bone1Name=dyn_SkirtH06
  bone2Name=dyn_SkirtG06
  swing1LimitDegrees=20
  swing1Motion=Free
  swing2LimitDegrees=20
  swing2Motion=Free
  twistLimitDegrees=20
 }
 {
  bone1Name=dyn_SkirtH04
  bone2Name=dyn_SkirtG04
  swing1LimitDegrees=20
  swing1Motion=Free
  swing2LimitDegrees=20
  swing2Motion=Free
  twistLimitDegrees=20
 }
 {
  bone1Name=dyn_SkirtH05
  bone2Name=dyn_SkirtG05
  swing1LimitDegrees=20
  swing1Motion=Free
  swing2LimitDegrees=20
  swing2Motion=Free
  twistLimitDegrees=20
 }
 {bone1Name=dyn_SkirtE03, bone2Name=dyn_SkirtF03, swing1LimitDegrees=20, swing2LimitDegrees=20, twistLimitDegrees=20}
 {bone1Name=dyn_SkirtB05, bone2Name=dyn_SkirtD05, swing1LimitDegrees=35, swing2LimitDegrees=75, twistLimitDegrees=75}
 {bone1Name=dyn_SkirtB04, bone2Name=dyn_SkirtD04, swing1LimitDegrees=35, swing2LimitDegrees=75, twistLimitDegrees=75}
 {bone1Name=dyn_SkirtH03, bone2Name=dyn_SkirtJ03, swing1LimitDegrees=45, swing2LimitDegrees=45, twistLimitDegrees=45}
 {bone1Name=hand_l_Sleeves_02_01, bone2Name=hand_l_Sleeves_02, swing1LimitDegrees=5, swing2LimitDegrees=5, twistLimitDegrees=5}
 {bone1Name=hand_r_Sreeves_02_01, bone2Name=hand_r_Sreeves_02, swing1LimitDegrees=5, swing2LimitDegrees=5, twistLimitDegrees=5}
]
