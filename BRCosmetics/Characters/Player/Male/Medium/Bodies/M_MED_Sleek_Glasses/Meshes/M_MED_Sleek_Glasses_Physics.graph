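# Physics asset for the male medium Sleek Glasses player body cosmetic, set up for a dangling necklace.
# Three necklace bones chain down from spine_05 through constraints with limited swing and locked twist.
# Capsules form each necklace link, and the first two links add thin, tilted boxes.
# Larger capsules on spine_04 and spine_05 provide torso volumes alongside the chain.
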PhysicsAsset /BRCosmetics/Characters/Player/Male/Medium/Bodies/M_MED_Sleek_Glasses/Meshes/M_MED_Sleek_Glasses_Physics
bodies {
 dyn_necklace_01 [
  {
   shapeType=Capsule
   length=3.7571
   radius=1.7429
   shapeName=dyn_necklace_03_capsule
   center=<-1.6099 1.8011 0.0261>
   rotation=<rot -89.1644 179.9993 143.0322>
  }
  {
   shapeType=Box
   extentX=4.6274
   extentY=0.6218
   extentZ=13.8777
   shapeName=dyn_necklace_01_box
   center=<-1.8192 1.0827 0>
   rotation=<rot 0 -30.7593 0>
  }
 ]
 dyn_necklace_02 [
  {
   shapeType=Capsule
   length=4.0273
   radius=1.7227
   shapeName=dyn_necklace_03_capsule
   center=<-2.2147 1.2048 0.0333>
   rotation=<rot -89.1649 179.9976 152.0352>
  }
  {
   shapeType=Box
   extentX=4.6274
   extentY=0.6218
   extentZ=13.8777
   shapeName=dyn_necklace_02_box
   center=<-1.8192 1.0827 0>
   rotation=<rot 0 -30.7593 0>
  }
 ]
 dyn_necklace_03 [{shapeType=Capsule, length=5.3523, radius=1.5432, shapeName=dyn_necklace_03_capsule, center=<-2.8839 0.5298 0>, rotation=<rot 90 90 100.4097>}]
 spine_04 [{shapeType=Capsule, length=16.1155, radius=11.1373, center=<3.2296 0.2909 0.1383>, rotation=<rot 0 7.0941 -176.3047>}]
 spine_05 [{shapeType=Capsule, length=24.1707, radius=10.3057, center=<-1.6772 -1.0783 -0.1078>, rotation=<rot -0.2276 -90 -1.8138>}]
}
constraints [
 {bone1Name=dyn_necklace_01, bone2Name=spine_05, swing1LimitDegrees=30, swing2LimitDegrees=20, twistLimitDegrees=0}
 {bone1Name=dyn_necklace_02, bone2Name=dyn_necklace_01, swing1LimitDegrees=30, swing2LimitDegrees=20, twistLimitDegrees=0}
 {bone1Name=dyn_necklace_03, bone2Name=dyn_necklace_02, swing1LimitDegrees=30, swing2LimitDegrees=20, twistLimitDegrees=0}
]
