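# Animation Blueprint for the male medium Ensemble outfit, adding secondary motion to its coat or skirt bones on a pose copied from the parent mesh.
# Arm helper bones are adjusted, and cached pose RbanCache feeds the cloth branches.
# bRigidBodyEnabled selects RigidBody simulations, switched by collision and diving/sliding flags, or an Anim Dynamics fallback.
# The fallback adds leg-following constraints and Control Rig outside skydiving, blends middle skirt bones, and applies direction-driven offsets.
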AnimBlueprint /BRCosmetics/Characters/Player/Male/Medium/Bodies/M_MED_Ensemble/Meshes/M_MED_Ensemble_AnimBP01
target_skeleton=&/M_MED_Ensemble_Skeleton
graphs [
 AnimGraph {nodes [
   Component To Local()
   n1 = RigidBody() {
    bClampLinearTranslationLimitToRefPose=true
    bOverrideWorldGravity=true
    overridePhysicsAsset=/BRCosmetics/Characters/Player/Male/Medium/Bodies/M_MED_Sleek_Glasses/Meshes/M_MED_Sleek_Glasses_Physics
    simulationSpace=BaseBoneSpace
    alphaBoolBlend {blendOption=Linear}
    baseBoneRef {boneName=spine_03}
    componentAppliedLinearAccClamp=<2000 2000 2000>
    componentLinearAccScale=<0.334 0.334 0.334>
    componentLinearVelScale=<0.334 0.334 0.334>
    overrideWorldGravity {z=-980}
   }
   "Constraint - dyn_SkirtF06 (PPP)"!(ComponentPose=@n1, Alpha=1, ConstraintWeights_0=0.2, ConstraintWeights_1:float=0.2, ConstraintWeights_2:float=0.6) -> (Pose) {
    alphaBoolBlend {blendOption=Linear}
    boneToModify {boneName=dyn_SkirtF06}
    constraintSetup [{targetBone {boneName=dyn_SkirtE06}}, {targetBone {boneName=dyn_SkirtG06}}, {targetBone {boneName=dyn_SkirtF06}}]
    constraintWeights [1,1,1]
   }
   "Constraint - collision (PP)"!(ComponentPose, Alpha=1, ConstraintWeights_0=1, ConstraintWeights_1:float=0.8) -> (Pose) {
    alphaBoolBlend {blendOption=Linear}
    boneToModify {boneName=collision}
    constraintSetup [{targetBone {boneName=pelvis}}, {targetBone {boneName=root}}]
    constraintWeights [1,1]
   }
   n2 = "Constraint - upperarmDn_l (PP)"!(ComponentPose, Alpha=1, ConstraintWeights_0=0.4, ConstraintWeights_1:float=0.6) -> (Pose) {
    alphaBoolBlend {blendOption=Linear}
    boneToModify {boneName=upperarmDn_l}
    constraintSetup [{targetBone {boneName=spine_04}}, {targetBone {boneName=clavicle_l}}]
    constraintWeights [1,1]
   }
   n3 = "Constraint - upperarmDn_r (PP)"!(ComponentPose=@n2, Alpha=1, ConstraintWeights_0=0.4, ConstraintWeights_1:float=0.6) -> (Pose) {
    alphaBoolBlend {blendOption=Linear}
    boneToModify {boneName=upperarmDn_r}
    constraintSetup [{targetBone {boneName=spine_04}}, {targetBone {boneName=clavicle_r}}]
    constraintWeights [1,1]
   }
   n4 = "Transform (Modify) Bone - Bone: upperarmDn_l"(ComponentPose=@n3, Alpha=Get Shoulder_rot_dn_l() -> (Shoulder_rot_dn_l:double), Translation="-8,0,0", Rotation="0,0,0", Scale="1,1,1") {
    rotationMode=BMM_Additive
    rotationSpace=BCS_BoneSpace
    scaleMode=BMM_Additive
    scaleSpace=BCS_BoneSpace
    translationMode=BMM_Additive
    translationSpace=BCS_BoneSpace
    alphaBoolBlend {blendOption=Linear}
    boneToModify {boneName=upperarmDn_l}
    translation {x=-8}
   }
   n5 = "Transform (Modify) Bone - Bone: upperarmDn_r"(ComponentPose=@n4, Alpha=Get Shoulder_rot_dn_r() -> (Shoulder_rot_dn_r:double), Translation="8,0,0", Rotation="0,0,0", Scale="1,1,1") {
    rotationMode=BMM_Additive
    rotationSpace=BCS_BoneSpace
    scaleMode=BMM_Additive
    scaleSpace=BCS_BoneSpace
    translationMode=BMM_Additive
    translationSpace=BCS_BoneSpace
    alphaBoolBlend {blendOption=Linear}
    boneToModify {boneName=upperarmDn_r}
    translation {x=8}
   }
   n6 = "Transform (Modify) Bone - Bone: upperarmDn_l"(ComponentPose=@n5, Alpha=Get Shoulder_rot_up_l() -> (Shoulder_rot_up_l:double), Translation="-5,0,12", Rotation="0,0,0", Scale="1,1,1") {
    rotationMode=BMM_Additive
    rotationSpace=BCS_BoneSpace
    scaleMode=BMM_Additive
    scaleSpace=BCS_BoneSpace
    translationMode=BMM_Additive
    translationSpace=BCS_BoneSpace
    alphaBoolBlend {blendOption=Linear}
    boneToModify {boneName=upperarmDn_l}
    translation {x=-5, z=12}
   }
   "Transform (Modify) Bone - Bone: upperarmDn_r"(ComponentPose=@n6, Alpha=Get Shoulder_rot_up_r() -> (Shoulder_rot_up_r:double), Translation="5,0,-12", Rotation="0,0,0", Scale="1,1,1") {
    rotationMode=BMM_Additive
    rotationSpace=BCS_BoneSpace
    scaleMode=BMM_Additive
    scaleSpace=BCS_BoneSpace
    translationMode=BMM_Additive
    translationSpace=BCS_BoneSpace
    alphaBoolBlend {blendOption=Linear}
    boneToModify {boneName=upperarmDn_r}
    translation {x=5, z=-12}
   }
   n7 = "Constraint - dyn_SkirtB04 (PP)"!(ComponentPose=Local To Component(LocalPose=Use cached pose 'RbanCache'()), Alpha=1, ConstraintWeights_0=0.5, ConstraintWeights_1:float=0.5) -> (Pose) {
    alphaBoolBlend {blendOption=Linear}
    boneToModify {boneName=dyn_SkirtB04}
    constraintSetup [{targetBone {boneName=calf_l}}, {targetBone {boneName=dyn_SkirtB03}}]
    constraintWeights [1,1]
   }
   n8 = "Constraint - dyn_SkirtB05 (PP)"!(ComponentPose=@n7, Alpha=1, ConstraintWeights_0=0, ConstraintWeights_1:float=1) -> (Pose) {
    alphaBoolBlend {blendOption=Linear}
    boneToModify {boneName=dyn_SkirtB05}
    constraintSetup [{targetBone {boneName=calf_twist_02_l}}, {targetBone {boneName=dyn_SkirtB04}}]
    constraintWeights [1,1]
   }
   n9 = "Constraint - dyn_SkirtB06 (PP)"!(ComponentPose=@n8, Alpha=1, ConstraintWeights_0=0, ConstraintWeights_1:float=1) -> (Pose) {
    alphaBoolBlend {blendOption=Linear}
    boneToModify {boneName=dyn_SkirtB06}
    constraintSetup [{targetBone {boneName=calf_twist_02_l}}, {targetBone {boneName=dyn_SkirtB05}}]
    constraintWeights [1,1]
   }
   n10 = "Constraint - dyn_SkirtD04 (PP)"!(ComponentPose=@n9, Alpha=1, ConstraintWeights_0=0.5, ConstraintWeights_1:float=0.5) -> (Pose) {
    alphaBoolBlend {blendOption=Linear}
    boneToModify {boneName=dyn_SkirtD04}
    constraintSetup [{targetBone {boneName=dyn_SkirtB04}}, {targetBone {boneName=dyn_SkirtD03}}]
    constraintWeights [1,1]
   }
   n11 = "Constraint - dyn_SkirtD05 (PP)"!(ComponentPose=@n10, Alpha=1, ConstraintWeights_0=0, ConstraintWeights_1:float=1) -> (Pose) {
    alphaBoolBlend {blendOption=Linear}
    boneToModify {boneName=dyn_SkirtD05}
    constraintSetup [{targetBone {boneName=calf_twist_02_l}}, {targetBone {boneName=dyn_SkirtD04}}]
    constraintWeights [1,1]
   }
   n12 = "Constraint - dyn_SkirtE05 (PP)"!(ComponentPose=@n11, Alpha=1, ConstraintWeights_0=0, ConstraintWeights_1:float=1) -> (Pose) {
    alphaBoolBlend {blendOption=Linear}
    boneToModify {boneName=dyn_SkirtE05}
    constraintSetup [{targetBone {boneName=calf_twist_02_l}}, {targetBone {boneName=dyn_SkirtE03}}]
    constraintWeights [1,1]
   }
   n13 = "Constraint - dyn_SkirtJ04 (PP)"!(ComponentPose=@n12, Alpha=1, ConstraintWeights_0=0.5, ConstraintWeights_1:float=0.5) -> (Pose) {
    alphaBoolBlend {blendOption=Linear}
    boneToModify {boneName=dyn_SkirtJ04}
    constraintSetup [{targetBone {boneName=calf_r}}, {targetBone {boneName=dyn_SkirtJ03}}]
    constraintWeights [1,1]
   }
   n14 = "Constraint - dyn_SkirtJ05 (PP)"!(ComponentPose=@n13, Alpha=1, ConstraintWeights_0=0, ConstraintWeights_1:float=1) -> (Pose) {
    alphaBoolBlend {blendOption=Linear}
    boneToModify {boneName=dyn_SkirtJ05}
    constraintSetup [{targetBone {boneName=calf_twist_02_r}}, {targetBone {boneName=dyn_SkirtJ04}}]
    constraintWeights [1,1]
   }
   n15 = "Constraint - dyn_SkirtJ06 (PP)"!(ComponentPose=@n14, Alpha=1, ConstraintWeights_0=0, ConstraintWeights_1:float=1) -> (Pose) {
    alphaBoolBlend {blendOption=Linear}
    boneToModify {boneName=dyn_SkirtJ06}
    constraintSetup [{targetBone {boneName=calf_twist_02_r}}, {targetBone {boneName=dyn_SkirtJ05}}]
    constraintWeights [1,1]
   }
   n16 = "Constraint - dyn_SkirtH04 (PP)"!(ComponentPose=@n15, Alpha=1, ConstraintWeights_0=0.5, ConstraintWeights_1:float=0.5) -> (Pose) {
    alphaBoolBlend {blendOption=Linear}
    boneToModify {boneName=dyn_SkirtH04}
    constraintSetup [{targetBone {boneName=dyn_SkirtJ04}}, {targetBone {boneName=dyn_SkirtH03}}]
    constraintWeights [1,1]
   }
   n17 = "Constraint - dyn_SkirtH05 (PP)"!(ComponentPose=@n16, Alpha=1, ConstraintWeights_0=0, ConstraintWeights_1:float=1) -> (Pose) {
    alphaBoolBlend {blendOption=Linear}
    boneToModify {boneName=dyn_SkirtH05}
    constraintSetup [{targetBone {boneName=calf_twist_02_r}}, {targetBone {boneName=dyn_SkirtH04}}]
    constraintWeights [1,1]
   }
   n18 = "Constraint - dyn_SkirtG05 (PP)"!(ComponentPose=@n17, Alpha=0, ConstraintWeights_0=0, ConstraintWeights_1:float=1) -> (Pose) {
    alphaBoolBlend {blendOption=Linear}
    boneToModify {boneName=dyn_SkirtG05}
    constraintSetup [{targetBone {boneName=calf_twist_02_r}}, {targetBone {boneName=dyn_SkirtG03}}]
    constraintWeights [1,1]
   }
   n19 = Control Rig(Source=Component To Local(ComponentPose=@n18)) {
    alphaBoolBlend {blendOption=Linear}
    controlRigAssetReference {blueprintRigClass=&/M_MED_Ensemble_CtrlRig.M_MED_Ensemble_CtrlRig_C}
   }
   n20 = Use cached pose 'RbanCache'()
   n21 = Get bIsSkyDiving()
   n22 = Anim Dynamics - Bone: dyn_SkirtB03(
    ComponentPose=Local To Component(LocalPose=Blend Poses by bool(BlendPose_0=@n20, BlendPose_1=@n19, bActiveValue=@n21) {blendType=Linear})
    Alpha=1
    ExternalForce="0,0,0"
   ) {
    angularDampingOverride=0.999
    bAngularSpring=true
    bGravityOverrideInSimSpace=true
    bOverrideAngularDamping=true
    bUseGravityOverride=true
    gravityScale=0.2
    simulationSpace=RootRelative
    alphaBoolBlend {blendOption=Linear}
    boundBone {boneName=dyn_SkirtB03}
    componentAppliedLinearAccClamp=<2000 2000 2000>
    componentLinearAccScale=<0.35 0.35 0.4>
    componentLinearVelScale=<0.4 0.4 0.4>
    gravityOverride {z=-980}
    physicsBodyDefinitions [{
      sphereCollisionRadius=0
      boundBone {boneName=dyn_SkirtB03}
      boxExtents=<15 15 5>
      constraintSetup {angularTargetAxis=AxisY, twistAxis=AxisY, angularLimitsMax=<10 0 0>, angularLimitsMin=<-1 0 0>, angularTarget=<0 0 0>}
      localJointOffset=<-0.1 -0.1 -7>
     }]
    relativeSpaceBone {boneName=thigh_twist_01_l}
   }
   n23 = Anim Dynamics - Bone: dyn_SkirtB04(ComponentPose=@n22, Alpha=1, ExternalForce="0,0,0") {
    angularDampingOverride=0.999
    bAngularSpring=true
    bGravityOverrideInSimSpace=true
    bOverrideAngularDamping=true
    bUseGravityOverride=true
    gravityScale=0.2
    simulationSpace=RootRelative
    alphaBoolBlend {blendOption=Linear}
    boundBone {boneName=dyn_SkirtB04}
    componentAppliedLinearAccClamp=<2000 2000 2000>
    componentLinearAccScale=<0.35 0.35 0.4>
    componentLinearVelScale=<0.4 0.4 0.4>
    gravityOverride {z=-980}
    physicsBodyDefinitions [{
      sphereCollisionRadius=0
      boundBone {boneName=dyn_SkirtB04}
      boxExtents=<15 15 5>
      constraintSetup {angularTargetAxis=AxisY, twistAxis=AxisY, angularLimitsMax=<10 0 0>, angularLimitsMin=<-1 0 0>, angularTarget=<0 0 0>}
      localJointOffset=<-0.1 -0.1 -7>
     }]
    relativeSpaceBone {boneName=thigh_twist_01_l}
   }
   n24 = Anim Dynamics - Bone: dyn_SkirtD03(ComponentPose=@n23, Alpha=1, ExternalForce="0,0,0") {
    angularDampingOverride=0.999
    bAngularSpring=true
    bGravityOverrideInSimSpace=true
    bOverrideAngularDamping=true
    bUseGravityOverride=true
    gravityScale=0.2
    simulationSpace=RootRelative
    alphaBoolBlend {blendOption=Linear}
    boundBone {boneName=dyn_SkirtD03}
    componentAppliedLinearAccClamp=<2000 2000 2000>
    componentLinearAccScale=<0.35 0.35 0.4>
    componentLinearVelScale=<0.4 0.4 0.4>
    gravityOverride {z=-980}
    physicsBodyDefinitions [{
      sphereCollisionRadius=0
      boundBone {boneName=dyn_SkirtD03}
      boxExtents=<15 15 5>
      constraintSetup {angularTargetAxis=AxisY, twistAxis=AxisY, angularLimitsMax=<10 0 0>, angularLimitsMin=<-1 0 0>, angularTarget=<0 0 0>}
      localJointOffset=<-0.1 -0.1 -7>
     }]
    relativeSpaceBone {boneName=thigh_twist_01_l}
   }
   n25 = Anim Dynamics - Bone: dyn_SkirtD04(ComponentPose=@n24, Alpha=1, ExternalForce="0,0,0") {
    angularDampingOverride=0.999
    bAngularSpring=true
    bGravityOverrideInSimSpace=true
    bOverrideAngularDamping=true
    bUseGravityOverride=true
    gravityScale=0.2
    simulationSpace=RootRelative
    alphaBoolBlend {blendOption=Linear}
    boundBone {boneName=dyn_SkirtD04}
    componentAppliedLinearAccClamp=<2000 2000 2000>
    componentLinearAccScale=<0.35 0.35 0.4>
    componentLinearVelScale=<0.4 0.4 0.4>
    gravityOverride {z=-980}
    physicsBodyDefinitions [{
      sphereCollisionRadius=0
      boundBone {boneName=dyn_SkirtD04}
      boxExtents=<15 15 5>
      constraintSetup {angularTargetAxis=AxisY, twistAxis=AxisY, angularLimitsMax=<10 0 0>, angularLimitsMin=<-1 0 0>, angularTarget=<0 0 0>}
      localJointOffset=<-0.1 -0.1 -7>
     }]
    relativeSpaceBone {boneName=thigh_twist_01_l}
   }
   n26 = Anim Dynamics - Bone: dyn_SkirtE03(ComponentPose=@n25, Alpha=1, ExternalForce="0,0,0") {
    angularDampingOverride=0.999
    bAngularSpring=true
    bGravityOverrideInSimSpace=true
    bOverrideAngularDamping=true
    bUseGravityOverride=true
    gravityScale=0.2
    simulationSpace=RootRelative
    alphaBoolBlend {blendOption=Linear}
    boundBone {boneName=dyn_SkirtE03}
    componentAppliedLinearAccClamp=<2000 2000 2000>
    componentLinearAccScale=<0.35 0.35 0.4>
    componentLinearVelScale=<0.4 0.4 0.4>
    gravityOverride {z=-980}
    physicsBodyDefinitions [{
      sphereCollisionRadius=0
      boundBone {boneName=dyn_SkirtE03}
      boxExtents=<15 15 5>
      constraintSetup {angularTargetAxis=AxisY, twistAxis=AxisY, angularLimitsMax=<7 0 0>, angularLimitsMin=<-1 0 0>, angularTarget=<0 0 0>}
      localJointOffset=<-0.1 -0.1 -5>
     }]
    relativeSpaceBone {boneName=thigh_twist_01_l}
   }
   n27 = Anim Dynamics - Bone: dyn_SkirtE04(ComponentPose=@n26, Alpha=1, ExternalForce="0,0,0") {
    angularDampingOverride=0.999
    bAngularSpring=true
    bGravityOverrideInSimSpace=true
    bOverrideAngularDamping=true
    bUseGravityOverride=true
    gravityScale=0.2
    simulationSpace=RootRelative
    alphaBoolBlend {blendOption=Linear}
    boundBone {boneName=dyn_SkirtE04}
    componentAppliedLinearAccClamp=<2000 2000 2000>
    componentLinearAccScale=<0.35 0.35 0.4>
    componentLinearVelScale=<0.4 0.4 0.4>
    gravityOverride {z=-980}
    physicsBodyDefinitions [{
      sphereCollisionRadius=0
      boundBone {boneName=dyn_SkirtE04}
      boxExtents=<15 15 5>
      constraintSetup {angularTargetAxis=AxisY, twistAxis=AxisY, angularLimitsMax=<7 0 0>, angularLimitsMin=<-1 0 0>, angularTarget=<0 0 0>}
      localJointOffset=<-0.1 -0.1 -5>
     }]
    relativeSpaceBone {boneName=thigh_twist_01_l}
   }
   n28 = Anim Dynamics - Bone: dyn_SkirtB05(ComponentPose=@n27, Alpha=1, ExternalForce="0,0,0") {
    angularDampingOverride=0.999
    bAngularSpring=true
    bGravityOverrideInSimSpace=true
    bOverrideAngularDamping=true
    bUseGravityOverride=true
    gravityScale=0.2
    simulationSpace=RootRelative
    alphaBoolBlend {blendOption=Linear}
    boundBone {boneName=dyn_SkirtB05}
    componentAppliedLinearAccClamp=<2000 2000 2000>
    componentLinearAccScale=<0.35 0.35 0.4>
    componentLinearVelScale=<0.4 0.4 0.4>
    gravityOverride {z=-980}
    physicsBodyDefinitions [{
      sphereCollisionRadius=0
      boundBone {boneName=dyn_SkirtB05}
      boxExtents=<15 15 5>
      constraintSetup {angularTargetAxis=AxisY, twistAxis=AxisY, angularLimitsMax=<10 0 0>, angularLimitsMin=<-1 0 0>, angularTarget=<0 0 0>}
      localJointOffset=<-0.1 -0.1 -7>
     }]
    relativeSpaceBone {boneName=calf_l}
   }
   n29 = Anim Dynamics - Bone: dyn_SkirtB06(ComponentPose=@n28, Alpha=1, ExternalForce="0,0,0") {
    angularDampingOverride=0.999
    bAngularSpring=true
    bGravityOverrideInSimSpace=true
    bOverrideAngularDamping=true
    bUseGravityOverride=true
    gravityScale=0.2
    simulationSpace=RootRelative
    alphaBoolBlend {blendOption=Linear}
    boundBone {boneName=dyn_SkirtB06}
    componentAppliedLinearAccClamp=<2000 2000 2000>
    componentLinearAccScale=<0.35 0.35 0.4>
    componentLinearVelScale=<0.4 0.4 0.4>
    gravityOverride {z=-980}
    physicsBodyDefinitions [{
      sphereCollisionRadius=0
      boundBone {boneName=dyn_SkirtB06}
      boxExtents=<15 15 5>
      constraintSetup {angularTargetAxis=AxisY, twistAxis=AxisY, angularLimitsMax=<10 0 0>, angularLimitsMin=<-1 0 0>, angularTarget=<0 0 0>}
      localJointOffset=<-0.1 -0.1 -7>
     }]
    relativeSpaceBone {boneName=calf_l}
   }
   n30 = Anim Dynamics - Bone: dyn_SkirtD05(ComponentPose=@n29, Alpha=1, ExternalForce="0,0,0") {
    angularDampingOverride=0.999
    bAngularSpring=true
    bGravityOverrideInSimSpace=true
    bOverrideAngularDamping=true
    bUseGravityOverride=true
    gravityScale=0.2
    simulationSpace=RootRelative
    alphaBoolBlend {blendOption=Linear}
    boundBone {boneName=dyn_SkirtD05}
    componentAppliedLinearAccClamp=<2000 2000 2000>
    componentLinearAccScale=<0.35 0.35 0.4>
    componentLinearVelScale=<0.4 0.4 0.4>
    gravityOverride {z=-980}
    physicsBodyDefinitions [{
      sphereCollisionRadius=0
      boundBone {boneName=dyn_SkirtD05}
      boxExtents=<15 15 5>
      constraintSetup {angularTargetAxis=AxisY, twistAxis=AxisY, angularLimitsMax=<10 0 0>, angularLimitsMin=<-1 0 0>, angularTarget=<0 0 0>}
      localJointOffset=<-0.1 -0.1 -7>
     }]
    relativeSpaceBone {boneName=calf_l}
   }
   n31 = Anim Dynamics - Bone: dyn_SkirtD06(ComponentPose=@n30, Alpha=1, ExternalForce="0,0,0") {
    angularDampingOverride=0.999
    bAngularSpring=true
    bGravityOverrideInSimSpace=true
    bOverrideAngularDamping=true
    bUseGravityOverride=true
    gravityScale=0.2
    simulationSpace=RootRelative
    alphaBoolBlend {blendOption=Linear}
    boundBone {boneName=dyn_SkirtD06}
    componentAppliedLinearAccClamp=<2000 2000 2000>
    componentLinearAccScale=<0.35 0.35 0.4>
    componentLinearVelScale=<0.4 0.4 0.4>
    gravityOverride {z=-980}
    physicsBodyDefinitions [{
      sphereCollisionRadius=0
      boundBone {boneName=dyn_SkirtD06}
      boxExtents=<15 15 5>
      constraintSetup {angularTargetAxis=AxisY, twistAxis=AxisY, angularLimitsMax=<10 0 0>, angularLimitsMin=<-1 0 0>, angularTarget=<0 0 0>}
      localJointOffset=<-0.1 -0.1 -7>
     }]
    relativeSpaceBone {boneName=calf_l}
   }
   n32 = Anim Dynamics - Bone: dyn_SkirtE05(ComponentPose=@n31, Alpha=1, ExternalForce="0,0,0") {
    angularDampingOverride=0.999
    bAngularSpring=true
    bGravityOverrideInSimSpace=true
    bOverrideAngularDamping=true
    bUseGravityOverride=true
    gravityScale=0.2
    simulationSpace=RootRelative
    alphaBoolBlend {blendOption=Linear}
    boundBone {boneName=dyn_SkirtE05}
    componentAppliedLinearAccClamp=<2000 2000 2000>
    componentLinearAccScale=<0.35 0.35 0.4>
    componentLinearVelScale=<0.4 0.4 0.4>
    gravityOverride {z=-980}
    physicsBodyDefinitions [{
      sphereCollisionRadius=0
      boundBone {boneName=dyn_SkirtE05}
      boxExtents=<15 15 5>
      constraintSetup {angularTargetAxis=AxisY, twistAxis=AxisY, angularLimitsMax=<7 0 0>, angularLimitsMin=<-1 0 0>, angularTarget=<0 0 0>}
      localJointOffset=<-0.1 -0.1 -5>
     }]
    relativeSpaceBone {boneName=calf_l}
   }
   n33 = Anim Dynamics - Bone: dyn_SkirtE06(ComponentPose=@n32, Alpha=1, ExternalForce="0,0,0") {
    angularDampingOverride=0.999
    bAngularSpring=true
    bGravityOverrideInSimSpace=true
    bOverrideAngularDamping=true
    bUseGravityOverride=true
    gravityScale=0.2
    simulationSpace=RootRelative
    alphaBoolBlend {blendOption=Linear}
    boundBone {boneName=dyn_SkirtE06}
    componentAppliedLinearAccClamp=<2000 2000 2000>
    componentLinearAccScale=<0.35 0.35 0.4>
    componentLinearVelScale=<0.4 0.4 0.4>
    gravityOverride {z=-980}
    physicsBodyDefinitions [{
      sphereCollisionRadius=0
      boundBone {boneName=dyn_SkirtE06}
      boxExtents=<15 15 5>
      constraintSetup {angularTargetAxis=AxisY, twistAxis=AxisY, angularLimitsMax=<7 0 0>, angularLimitsMin=<-1 0 0>, angularTarget=<0 0 0>}
      localJointOffset=<-0.1 -0.1 -5>
     }]
    relativeSpaceBone {boneName=calf_l}
   }
   n34 = Anim Dynamics - Bone: dyn_SkirtJ03(ComponentPose=@n33, Alpha=1, ExternalForce="0,0,0") {
    angularDampingOverride=0.999
    bAngularSpring=true
    bGravityOverrideInSimSpace=true
    bOverrideAngularDamping=true
    bUseGravityOverride=true
    gravityScale=0.2
    simulationSpace=RootRelative
    alphaBoolBlend {blendOption=Linear}
    boundBone {boneName=dyn_SkirtJ03}
    componentAppliedLinearAccClamp=<2000 2000 2000>
    componentLinearAccScale=<0.35 0.35 0.4>
    componentLinearVelScale=<0.4 0.4 0.4>
    gravityOverride {z=-980}
    physicsBodyDefinitions [{
      sphereCollisionRadius=0
      boundBone {boneName=dyn_SkirtJ03}
      boxExtents=<15 15 5>
      constraintSetup {angularTargetAxis=AxisY, twistAxis=AxisY, angularLimitsMax=<10 0 0>, angularLimitsMin=<-1 0 0>, angularTarget=<0 0 0>}
      localJointOffset=<-0.1 -0.1 -7>
     }]
    relativeSpaceBone {boneName=thigh_twist_01_r}
   }
   n35 = Anim Dynamics - Bone: dyn_SkirtJ04(ComponentPose=@n34, Alpha=1, ExternalForce="0,0,0") {
    angularDampingOverride=0.999
    bAngularSpring=true
    bGravityOverrideInSimSpace=true
    bOverrideAngularDamping=true
    bUseGravityOverride=true
    gravityScale=0.2
    simulationSpace=RootRelative
    alphaBoolBlend {blendOption=Linear}
    boundBone {boneName=dyn_SkirtJ04}
    componentAppliedLinearAccClamp=<2000 2000 2000>
    componentLinearAccScale=<0.35 0.35 0.4>
    componentLinearVelScale=<0.4 0.4 0.4>
    gravityOverride {z=-980}
    physicsBodyDefinitions [{
      sphereCollisionRadius=0
      boundBone {boneName=dyn_SkirtJ04}
      boxExtents=<15 15 5>
      constraintSetup {angularTargetAxis=AxisY, twistAxis=AxisY, angularLimitsMax=<10 0 0>, angularLimitsMin=<-1 0 0>, angularTarget=<0 0 0>}
      localJointOffset=<-0.1 -0.1 -7>
     }]
    relativeSpaceBone {boneName=thigh_twist_01_r}
   }
   n36 = Anim Dynamics - Bone: dyn_SkirtH03(ComponentPose=@n35, Alpha=1, ExternalForce="0,0,0") {
    angularDampingOverride=0.999
    bAngularSpring=true
    bGravityOverrideInSimSpace=true
    bOverrideAngularDamping=true
    bUseGravityOverride=true
    gravityScale=0.2
    simulationSpace=RootRelative
    alphaBoolBlend {blendOption=Linear}
    boundBone {boneName=dyn_SkirtH03}
    componentAppliedLinearAccClamp=<2000 2000 2000>
    componentLinearAccScale=<0.35 0.35 0.4>
    componentLinearVelScale=<0.4 0.4 0.4>
    gravityOverride {z=-980}
    physicsBodyDefinitions [{
      sphereCollisionRadius=0
      boundBone {boneName=dyn_SkirtH03}
      boxExtents=<15 15 5>
      constraintSetup {angularTargetAxis=AxisY, twistAxis=AxisY, angularLimitsMax=<10 0 0>, angularLimitsMin=<-1 0 0>, angularTarget=<0 0 0>}
      localJointOffset=<-0.1 -0.1 -7>
     }]
    relativeSpaceBone {boneName=thigh_twist_01_r}
   }
   n37 = Anim Dynamics - Bone: dyn_SkirtH04(ComponentPose=@n36, Alpha=1, ExternalForce="0,0,0") {
    angularDampingOverride=0.999
    bAngularSpring=true
    bGravityOverrideInSimSpace=true
    bOverrideAngularDamping=true
    bUseGravityOverride=true
    gravityScale=0.2
    simulationSpace=RootRelative
    alphaBoolBlend {blendOption=Linear}
    boundBone {boneName=dyn_SkirtH04}
    componentAppliedLinearAccClamp=<2000 2000 2000>
    componentLinearAccScale=<0.35 0.35 0.4>
    componentLinearVelScale=<0.4 0.4 0.4>
    gravityOverride {z=-980}
    physicsBodyDefinitions [{
      sphereCollisionRadius=0
      boundBone {boneName=dyn_SkirtH04}
      boxExtents=<15 15 5>
      constraintSetup {angularTargetAxis=AxisY, twistAxis=AxisY, angularLimitsMax=<10 0 0>, angularLimitsMin=<-1 0 0>, angularTarget=<0 0 0>}
      localJointOffset=<-0.1 -0.1 -7>
     }]
    relativeSpaceBone {boneName=thigh_twist_01_r}
   }
   n38 = Anim Dynamics - Bone: dyn_SkirtG03(ComponentPose=@n37, Alpha=1, ExternalForce="0,0,0") {
    angularDampingOverride=0.999
    bAngularSpring=true
    bGravityOverrideInSimSpace=true
    bOverrideAngularDamping=true
    bUseGravityOverride=true
    gravityScale=0.2
    simulationSpace=RootRelative
    alphaBoolBlend {blendOption=Linear}
    boundBone {boneName=dyn_SkirtG03}
    componentAppliedLinearAccClamp=<2000 2000 2000>
    componentLinearAccScale=<0.35 0.35 0.4>
    componentLinearVelScale=<0.4 0.4 0.4>
    gravityOverride {z=-980}
    physicsBodyDefinitions [{
      sphereCollisionRadius=0
      boundBone {boneName=dyn_SkirtG03}
      boxExtents=<15 15 5>
      constraintSetup {angularTargetAxis=AxisY, twistAxis=AxisY, angularLimitsMax=<5 0 0>, angularLimitsMin=<-1 0 0>, angularTarget=<0 0 0>}
      localJointOffset=<-0.1 -0.1 -5>
     }]
    relativeSpaceBone {boneName=thigh_twist_01_r}
   }
   n39 = Anim Dynamics - Bone: dyn_SkirtG04(ComponentPose=@n38, Alpha=1, ExternalForce="0,0,0") {
    angularDampingOverride=0.999
    bAngularSpring=true
    bGravityOverrideInSimSpace=true
    bOverrideAngularDamping=true
    bUseGravityOverride=true
    gravityScale=0.2
    simulationSpace=RootRelative
    alphaBoolBlend {blendOption=Linear}
    boundBone {boneName=dyn_SkirtG04}
    componentAppliedLinearAccClamp=<2000 2000 2000>
    componentLinearAccScale=<0.35 0.35 0.4>
    componentLinearVelScale=<0.4 0.4 0.4>
    gravityOverride {z=-980}
    physicsBodyDefinitions [{
      sphereCollisionRadius=0
      boundBone {boneName=dyn_SkirtG04}
      boxExtents=<15 15 5>
      constraintSetup {angularTargetAxis=AxisY, twistAxis=AxisY, angularLimitsMax=<5 0 0>, angularLimitsMin=<-1 0 0>, angularTarget=<0 0 0>}
      localJointOffset=<-0.1 -0.1 -5>
     }]
    relativeSpaceBone {boneName=thigh_twist_01_r}
   }
   n40 = Anim Dynamics - Bone: dyn_SkirtJ05(ComponentPose=@n39, Alpha=1, ExternalForce="0,0,0") {
    angularDampingOverride=0.999
    bAngularSpring=true
    bGravityOverrideInSimSpace=true
    bOverrideAngularDamping=true
    bUseGravityOverride=true
    gravityScale=0.2
    simulationSpace=RootRelative
    alphaBoolBlend {blendOption=Linear}
    boundBone {boneName=dyn_SkirtJ05}
    componentAppliedLinearAccClamp=<2000 2000 2000>
    componentLinearAccScale=<0.35 0.35 0.4>
    componentLinearVelScale=<0.4 0.4 0.4>
    gravityOverride {z=-980}
    physicsBodyDefinitions [{
      sphereCollisionRadius=0
      boundBone {boneName=dyn_SkirtJ05}
      boxExtents=<15 15 5>
      constraintSetup {angularTargetAxis=AxisY, twistAxis=AxisY, angularLimitsMax=<10 0 0>, angularLimitsMin=<-1 0 0>, angularTarget=<0 0 0>}
      localJointOffset=<-0.1 -0.1 -7>
     }]
    relativeSpaceBone {boneName=calf_r}
   }
   n41 = Anim Dynamics - Bone: dyn_SkirtJ06(ComponentPose=@n40, Alpha=1, ExternalForce="0,0,0") {
    angularDampingOverride=0.999
    bAngularSpring=true
    bGravityOverrideInSimSpace=true
    bOverrideAngularDamping=true
    bUseGravityOverride=true
    gravityScale=0.2
    simulationSpace=RootRelative
    alphaBoolBlend {blendOption=Linear}
    boundBone {boneName=dyn_SkirtJ06}
    componentAppliedLinearAccClamp=<2000 2000 2000>
    componentLinearAccScale=<0.35 0.35 0.4>
    componentLinearVelScale=<0.4 0.4 0.4>
    gravityOverride {z=-980}
    physicsBodyDefinitions [{
      sphereCollisionRadius=0
      boundBone {boneName=dyn_SkirtJ06}
      boxExtents=<15 15 5>
      constraintSetup {angularTargetAxis=AxisY, twistAxis=AxisY, angularLimitsMax=<10 0 0>, angularLimitsMin=<-1 0 0>, angularTarget=<0 0 0>}
      localJointOffset=<-0.1 -0.1 -7>
     }]
    relativeSpaceBone {boneName=calf_r}
   }
   n42 = Anim Dynamics - Bone: dyn_SkirtH05(ComponentPose=@n41, Alpha=1, ExternalForce="0,0,0") {
    angularDampingOverride=0.999
    bAngularSpring=true
    bGravityOverrideInSimSpace=true
    bOverrideAngularDamping=true
    bUseGravityOverride=true
    gravityScale=0.2
    simulationSpace=RootRelative
    alphaBoolBlend {blendOption=Linear}
    boundBone {boneName=dyn_SkirtH05}
    componentAppliedLinearAccClamp=<2000 2000 2000>
    componentLinearAccScale=<0.35 0.35 0.4>
    componentLinearVelScale=<0.4 0.4 0.4>
    gravityOverride {z=-980}
    physicsBodyDefinitions [{
      sphereCollisionRadius=0
      boundBone {boneName=dyn_SkirtH05}
      boxExtents=<15 15 5>
      constraintSetup {angularTargetAxis=AxisY, twistAxis=AxisY, angularLimitsMax=<10 0 0>, angularLimitsMin=<-1 0 0>, angularTarget=<0 0 0>}
      localJointOffset=<-0.1 -0.1 -7>
     }]
    relativeSpaceBone {boneName=calf_r}
   }
   n43 = Anim Dynamics - Bone: dyn_SkirtH06(ComponentPose=@n42, Alpha=1, ExternalForce="0,0,0") {
    angularDampingOverride=0.999
    bAngularSpring=true
    bGravityOverrideInSimSpace=true
    bOverrideAngularDamping=true
    bUseGravityOverride=true
    gravityScale=0.2
    simulationSpace=RootRelative
    alphaBoolBlend {blendOption=Linear}
    boundBone {boneName=dyn_SkirtH06}
    componentAppliedLinearAccClamp=<2000 2000 2000>
    componentLinearAccScale=<0.35 0.35 0.4>
    componentLinearVelScale=<0.4 0.4 0.4>
    gravityOverride {z=-980}
    physicsBodyDefinitions [{
      sphereCollisionRadius=0
      boundBone {boneName=dyn_SkirtH06}
      boxExtents=<15 15 5>
      constraintSetup {angularTargetAxis=AxisY, twistAxis=AxisY, angularLimitsMax=<10 0 0>, angularLimitsMin=<-1 0 0>, angularTarget=<0 0 0>}
      localJointOffset=<-0.1 -0.1 -7>
     }]
    relativeSpaceBone {boneName=calf_r}
   }
   n44 = Anim Dynamics - Bone: dyn_SkirtG05(ComponentPose=@n43, Alpha=1, ExternalForce="0,0,0") {
    angularDampingOverride=0.999
    bAngularSpring=true
    bGravityOverrideInSimSpace=true
    bOverrideAngularDamping=true
    bUseGravityOverride=true
    gravityScale=0.2
    simulationSpace=RootRelative
    alphaBoolBlend {blendOption=Linear}
    boundBone {boneName=dyn_SkirtG05}
    componentAppliedLinearAccClamp=<2000 2000 2000>
    componentLinearAccScale=<0.35 0.35 0.4>
    componentLinearVelScale=<0.4 0.4 0.4>
    gravityOverride {z=-980}
    physicsBodyDefinitions [{
      sphereCollisionRadius=0
      boundBone {boneName=dyn_SkirtG05}
      boxExtents=<15 15 5>
      constraintSetup {angularTargetAxis=AxisY, twistAxis=AxisY, angularLimitsMax=<5 0 0>, angularLimitsMin=<-1 0 0>, angularTarget=<0 0 0>}
      localJointOffset=<-0.1 -0.1 -5>
     }]
    relativeSpaceBone {boneName=calf_r}
   }
   n45 = Anim Dynamics - Bone: dyn_SkirtG06(ComponentPose=@n44, Alpha=1, ExternalForce="0,0,0") {
    angularDampingOverride=0.999
    bAngularSpring=true
    bGravityOverrideInSimSpace=true
    bOverrideAngularDamping=true
    bUseGravityOverride=true
    gravityScale=0.2
    simulationSpace=RootRelative
    alphaBoolBlend {blendOption=Linear}
    boundBone {boneName=dyn_SkirtG06}
    componentAppliedLinearAccClamp=<2000 2000 2000>
    componentLinearAccScale=<0.35 0.35 0.4>
    componentLinearVelScale=<0.4 0.4 0.4>
    gravityOverride {z=-980}
    physicsBodyDefinitions [{
      sphereCollisionRadius=0
      boundBone {boneName=dyn_SkirtG06}
      boxExtents=<15 15 5>
      constraintSetup {angularTargetAxis=AxisY, twistAxis=AxisY, angularLimitsMax=<5 0 0>, angularLimitsMin=<-1 0 0>, angularTarget=<0 0 0>}
      localJointOffset=<-0.1 -0.1 -5>
     }]
    relativeSpaceBone {boneName=calf_r}
   }
   n46 = "Constraint - dyn_SkirtF03 (PP)"!(ComponentPose=@n45, Alpha=1, ConstraintWeights_0=0.5, ConstraintWeights_1:float=0.5) -> (Pose) {
    alphaBoolBlend {blendOption=Linear}
    boneToModify {boneName=dyn_SkirtF03}
    constraintSetup [{targetBone {boneName=dyn_SkirtE03}}, {targetBone {boneName=dyn_SkirtG03}}]
    constraintWeights [1,1]
   }
   n47 = "Constraint - dyn_SkirtF04 (PP)"!(ComponentPose=@n46, Alpha=1, ConstraintWeights_0=0.5, ConstraintWeights_1:float=0.5) -> (Pose) {
    alphaBoolBlend {blendOption=Linear}
    boneToModify {boneName=dyn_SkirtF04}
    constraintSetup [{targetBone {boneName=dyn_SkirtE04}}, {targetBone {boneName=dyn_SkirtG04}}]
    constraintWeights [1,1]
   }
   n48 = "Constraint - dyn_SkirtF05 (PPP)"!(ComponentPose=@n47, Alpha=1, ConstraintWeights_0=0.2, ConstraintWeights_1:float=0.2, ConstraintWeights_2:float=0.6) -> (Pose) {
    alphaBoolBlend {blendOption=Linear}
    boneToModify {boneName=dyn_SkirtF05}
    constraintSetup [{targetBone {boneName=dyn_SkirtE05}}, {targetBone {boneName=dyn_SkirtG05}}, {targetBone {boneName=dyn_SkirtF05}}]
    constraintWeights [1,1,1]
   }
   n49 = "Constraint - dyn_SkirtF06 (PPP)"(ComponentPose=@n48, Alpha=1, ConstraintWeights_0=0.2, ConstraintWeights_1=0.2, ConstraintWeights_2=0.6) {
    alphaBoolBlend {blendOption=Linear}
    boneToModify {boneName=dyn_SkirtF06}
    constraintSetup [{targetBone {boneName=dyn_SkirtE06}}, {targetBone {boneName=dyn_SkirtG06}}, {targetBone {boneName=dyn_SkirtF06}}]
    constraintWeights [1,1,1]
   }
   n50 = Get character direction() -> (character direction:double)
   n51 = "Transform (Modify) Bone - Bone: dyn_SkirtF03"(ComponentPose=@n49, Alpha=@n50, Translation="0,0,0", Rotation="0,0,-15", Scale="1,1,0.8") {
    rotationMode=BMM_Additive
    rotationSpace=BCS_BoneSpace
    scaleMode=BMM_Additive
    scaleSpace=BCS_BoneSpace
    alphaBoolBlend {blendOption=Linear}
    alphaScaleBiasClamp {bInterpResult=true}
    boneToModify {boneName=dyn_SkirtF03}
    rotation {roll=-15}
    scale {z=0.8}
   }
   n52 = "Transform (Modify) Bone - Bone: dyn_SkirtG05"(ComponentPose=@n51, Alpha=@n50, Translation="0,0,0", Rotation="0,0,-10", Scale="1,1,0.8") {
    rotationMode=BMM_Additive
    rotationSpace=BCS_BoneSpace
    scaleMode=BMM_Additive
    scaleSpace=BCS_BoneSpace
    alphaBoolBlend {blendOption=Linear}
    alphaScaleBiasClamp {bInterpResult=true}
    boneToModify {boneName=dyn_SkirtG05}
    rotation {roll=-10}
    scale {z=0.8}
   }
   n53 = "Transform (Modify) Bone - Bone: dyn_SkirtE05"(ComponentPose=@n52, Alpha=@n50, Translation="0,0,0", Rotation="0,0,-12", Scale="1,1,0.8") {
    rotationMode=BMM_Additive
    rotationSpace=BCS_BoneSpace
    scaleMode=BMM_Additive
    scaleSpace=BCS_BoneSpace
    alphaBoolBlend {blendOption=Linear}
    alphaScaleBiasClamp {bInterpResult=true}
    boneToModify {boneName=dyn_SkirtE05}
    rotation {roll=-12}
    scale {z=0.8}
   }
   n54 = "Constraint - collision (TT)"!(ComponentPose=Local To Component(LocalPose=Use cached pose 'RbanCache'()), Alpha=1, ConstraintWeights_0=0.3, ConstraintWeights_1:float=0.7) -> (Pose) {
    alphaBoolBlend {blendOption=Linear}
    boneToModify {boneName=collision}
    constraintSetup [
     {transformType=Translation, perAxis {bX=false}, targetBone {boneName=pelvis}}
     {offsetOption=None, transformType=Translation, perAxis {bX=false}, targetBone {boneName=root}}
    ]
    constraintWeights [1,1]
   }
   n55 = "Constraint - collision (RR)"!(ComponentPose=@n54, Alpha=1, ConstraintWeights_0=0, ConstraintWeights_1:float=1) -> (Pose) {
    alphaBoolBlend {blendOption=Linear}
    boneToModify {boneName=collision}
    constraintSetup [{transformType=Rotation, targetBone {boneName=pelvis}}, {transformType=Rotation, targetBone {boneName=root}}]
    constraintWeights [1,1]
   }
   n56 = "Constraint - collision (PP)"(ComponentPose=@n55, Alpha=1, ConstraintWeights_0=0, ConstraintWeights_1=1) {
    alphaBoolBlend {blendOption=Linear}
    boneToModify {boneName=collision}
    constraintSetup [{targetBone {boneName=pelvis}}, {targetBone {boneName=root}}]
    constraintWeights [1,1]
   }
   n57 = RigidBody!(
    ComponentPose=@n56
    Alpha
    OverrideWorldGravity:Vector=Get GravityOverride01() -> (GravityOverride01:Vector)
    ExternalForce
    ComponentLinearAccScale:Vector="0.15,0.35,0.35"
    ComponentLinearVelScale:Vector="0.15,0.35,0.35"
   ) -> (Pose) {
    bClampLinearTranslationLimitToRefPose=true
    bOverrideWorldGravity=true
    overridePhysicsAsset=&/M_MED_Ensemble_Physics
    simulationSpace=BaseBoneSpace
    alphaBoolBlend {blendOption=Linear}
    baseBoneRef {boneName=collision}
    componentAppliedLinearAccClamp=<500 1200 1200>
    componentLinearAccScale=<0.334 0.334 0.334>
    componentLinearVelScale=<0.4 0.4 0.4>
    overrideWorldGravity {x=200, z=-980}
   }
   n58 = Component To Local(ComponentPose=@n57)
   n59 = Get Component_Scale() -> (Component_Scale:Vector)
   n60 = RigidBody(
    ComponentPose=Local To Component(LocalPose=Use cached pose 'RbanCache'())
    OverrideWorldGravity="0,0,-150"
    ComponentLinearAccScale=@n59
    ComponentLinearVelScale=@n59
   ) {
    bClampLinearTranslationLimitToRefPose=true
    bOverrideWorldGravity=true
    overridePhysicsAsset=&/M_MED_Ensemble_PhysicsDiving
    simulationSpace=BaseBoneSpace
    alphaBoolBlend {blendOption=Linear}
    baseBoneRef {boneName=spine_01}
    componentAppliedLinearAccClamp=<100 100 100>
    componentLinearAccScale=<0.334 0.334 0.334>
    componentLinearVelScale=<0.4 0.4 0.4>
    overrideWorldGravity {x=200, z=-980}
   }
   n61 = Component To Local(ComponentPose=@n60)
   n62 = Get Is Diving or Basball Sliding() -> (Is Diving or Basball Sliding:bool)
   n63 = RigidBody(
    ComponentPose=Local To Component(LocalPose=Use cached pose 'RbanCache'())
    OverrideWorldGravity=Get GravityOverride01()
    ComponentLinearAccScale="0.15,0.35,0.35"
    ComponentLinearVelScale="0.15,0.35,0.35"
   ) {
    bClampLinearTranslationLimitToRefPose=true
    bOverrideWorldGravity=true
    overridePhysicsAsset=&M_MED_Ensemble_Physics
    simulationSpace=BaseBoneSpace
    alphaBoolBlend {blendOption=Linear}
    baseBoneRef {boneName=pelvis}
    componentAppliedLinearAccClamp=<200 600 600>
    componentLinearAccScale=<0.334 0.334 0.334>
    componentLinearVelScale=<0.4 0.4 0.4>
    overrideWorldGravity {x=200, z=-980}
   }
   n64 = Local To Component(LocalPose=Blend Poses by bool(
     BlendPose_0=Component To Local(ComponentPose=@n63)
     BlendPose_1=Blend Poses by bool(BlendPose_0=@n61, BlendPose_1=@n58, bActiveValue=@n62) {blendType=Linear}
     bActiveValue=Get collision() -> (collision:bool)
    ) {blendType=Linear})
   Output Pose(Result=Blend Poses by bool(BlendPose_0=Component To Local(ComponentPose=@n64), BlendPose_1=Component To Local(ComponentPose=@n53), bActiveValue=Get bRigidBodyEnabled()) {blendType=Linear}) {layerGroup=None}
   n65 = "Constraint - hand_l_Sreeves_gravity (TT)"!(
    ComponentPose=Local To Component(LocalPose=Copy Pose From Mesh() {bUseAttachedParent=true})
    Alpha=1
    ConstraintWeights_0=0.7
    ConstraintWeights_1:float=0.3
   ) -> (Pose) {
    alphaBoolBlend {blendOption=Linear}
    boneToModify {boneName=hand_l_Sreeves_gravity}
    constraintSetup [{transformType=Translation, targetBone {boneName=lowerarm_twist_02_l}}, {transformType=Translation, targetBone {boneName=root}}]
    constraintWeights [1,1]
   }
   n66 = "Constraint - hand_r_Sreeves_gravity (TT)"!(ComponentPose=@n65, Alpha=1, ConstraintWeights_0=0.7, ConstraintWeights_1:float=0.3) -> (Pose) {
    alphaBoolBlend {blendOption=Linear}
    boneToModify {boneName=hand_r_Sreeves_gravity}
    constraintSetup [{transformType=Translation, targetBone {boneName=lowerarm_twist_02_r}}, {transformType=Translation, targetBone {boneName=root}}]
    constraintWeights [1,1]
   }
   n67 = "Transform (Modify) Bone - Bone: scale_upperarm_l"(ComponentPose=@n66, Alpha=1, Translation="0,0,-1", Rotation="0,0,0", Scale="1,0.6,0.6") {
    rotationMode=BMM_Additive
    rotationSpace=BCS_BoneSpace
    scaleMode=BMM_Additive
    scaleSpace=BCS_BoneSpace
    translationMode=BMM_Additive
    translationSpace=BCS_BoneSpace
    alphaBoolBlend {blendOption=Linear}
    boneToModify {boneName=scale_upperarm_l}
    scale {y=0.6, z=0.6}
    translation {z=-1}
   }
   n68 = "Transform (Modify) Bone - Bone: scale_lowerarm_l"(ComponentPose=@n67, Alpha=1, Translation="0,0,-2", Rotation="0,0,0", Scale="1,0.9,0.9") {
    rotationMode=BMM_Additive
    rotationSpace=BCS_BoneSpace
    scaleMode=BMM_Additive
    scaleSpace=BCS_BoneSpace
    translationMode=BMM_Additive
    translationSpace=BCS_BoneSpace
    alphaBoolBlend {blendOption=Linear}
    boneToModify {boneName=scale_lowerarm_l}
    scale {y=0.9, z=0.9}
    translation {z=-2}
   }
   n69 = "Transform (Modify) Bone - Bone: scale_upperarm_r"(ComponentPose=@n68, Alpha=1, Translation="0,0,1", Rotation="0,0,0", Scale="1,0.6,0.6") {
    rotationMode=BMM_Additive
    rotationSpace=BCS_BoneSpace
    scaleMode=BMM_Additive
    scaleSpace=BCS_BoneSpace
    translationMode=BMM_Additive
    translationSpace=BCS_BoneSpace
    alphaBoolBlend {blendOption=Linear}
    boneToModify {boneName=scale_upperarm_r}
    scale {y=0.6, z=0.6}
    translation {z=1}
   }
   n70 = "Transform (Modify) Bone - Bone: scale_lowerarm_r"(ComponentPose=@n69, Alpha=1, Translation="0,0,2", Rotation="0,0,0", Scale="1,0.9,0.9") {
    rotationMode=BMM_Additive
    rotationSpace=BCS_BoneSpace
    scaleMode=BMM_Additive
    scaleSpace=BCS_BoneSpace
    translationMode=BMM_Additive
    translationSpace=BCS_BoneSpace
    alphaBoolBlend {blendOption=Linear}
    boneToModify {boneName=scale_lowerarm_r}
    scale {y=0.9, z=0.9}
    translation {z=2}
   }
   n71 = Anim Dynamics - Bone: dyn_collarB_02(ComponentPose=@n70, Alpha=0.5, ExternalForce="0,0,0") {
    angularDampingOverride=0.999
    angularSpringConstant=800
    bAngularSpring=true
    bGravityOverrideInSimSpace=true
    bOverrideAngularDamping=true
    gravityScale=0.2
    simulationSpace=RootRelative
    alphaBoolBlend {blendOption=Linear}
    boundBone {boneName=dyn_collarB_02}
    componentAppliedLinearAccClamp=<10000 10000 10000>
    componentLinearAccScale=<0.1 0.1 0.1>
    componentLinearVelScale=<0.1 0.1 0.1>
    physicsBodyDefinitions [{
      sphereCollisionRadius=0
      boundBone {boneName=dyn_collarB_02}
      boxExtents=<5 10 10>
      constraintSetup {angularTargetAxis=AxisY, angularLimitsMax=<2 2 0>, angularLimitsMin=<-2 -2 -35>, angularTarget=<0 0 0>}
      localJointOffset=<-3 0.1 -0.1>
     }]
    relativeSpaceBone {boneName=spine_05}
   }
   n72 = Anim Dynamics - Bone: dyn_collarA_02(ComponentPose=@n71, Alpha=0.5, ExternalForce="0,0,0") {
    angularDampingOverride=0.999
    angularSpringConstant=800
    bAngularSpring=true
    bGravityOverrideInSimSpace=true
    bOverrideAngularDamping=true
    gravityScale=0.2
    simulationSpace=RootRelative
    alphaBoolBlend {blendOption=Linear}
    boundBone {boneName=dyn_collarA_02}
    componentAppliedLinearAccClamp=<10000 10000 10000>
    componentLinearAccScale=<0.1 0.1 0.1>
    componentLinearVelScale=<0.1 0.1 0.1>
    physicsBodyDefinitions [{
      sphereCollisionRadius=0
      boundBone {boneName=dyn_collarA_02}
      boxExtents=<5 10 10>
      constraintSetup {angularTargetAxis=AxisY, angularLimitsMax=<2 2 0>, angularLimitsMin=<-2 -2 -35>, angularTarget=<0 0 0>}
      localJointOffset=<-3 -4 -0.1>
     }]
    relativeSpaceBone {boneName=spine_05}
   }
   n73 = "Transform (Modify) Bone - Bone: dyn_collarC_01"(ComponentPose=@n72, Alpha=Get Shoulder_rot_up_r(), Translation="0,0,0", Rotation="0,70,0", Scale="1,1,1") {
    rotationMode=BMM_Additive
    rotationSpace=BCS_BoneSpace
    scaleMode=BMM_Additive
    scaleSpace=BCS_BoneSpace
    translationMode=BMM_Additive
    translationSpace=BCS_BoneSpace
    alphaBoolBlend {blendOption=Linear}
    boneToModify {boneName=dyn_collarC_01}
    rotation {yaw=70}
   }
   n74 = Anim Dynamics - Bone: dyn_collarC_02(ComponentPose=@n73, Alpha=0.5, ExternalForce="0,0,0") {
    angularDampingOverride=0.999
    angularSpringConstant=800
    bAngularSpring=true
    bGravityOverrideInSimSpace=true
    bOverrideAngularDamping=true
    gravityScale=0.2
    simulationSpace=RootRelative
    alphaBoolBlend {blendOption=Linear}
    boundBone {boneName=dyn_collarC_02}
    componentAppliedLinearAccClamp=<10000 10000 10000>
    componentLinearAccScale=<0.1 0.1 0.1>
    componentLinearVelScale=<0.1 0.1 0.1>
    physicsBodyDefinitions [{
      sphereCollisionRadius=0
      boundBone {boneName=dyn_collarC_02}
      boxExtents=<5 10 10>
      constraintSetup {angularTargetAxis=AxisY, angularLimitsMax=<2 2 5>, angularLimitsMin=<-2 -2 -35>, angularTarget=<0 0 0>}
      localJointOffset=<-3 -4 -0.1>
     }]
    relativeSpaceBone {boneName=spine_05}
   }
   n75 = "Transform (Modify) Bone - Bone: dyn_collarD_01"(ComponentPose=@n74, Alpha=Get Shoulder_rot_f_r() -> (Shoulder_rot_f_r:double), Translation="7,0,0", Rotation="-25,-10,-30", Scale="1,1,1") {
    rotationMode=BMM_Additive
    rotationSpace=BCS_BoneSpace
    scaleMode=BMM_Additive
    scaleSpace=BCS_BoneSpace
    translationMode=BMM_Additive
    translationSpace=BCS_BoneSpace
    alphaBoolBlend {blendOption=Linear}
    boneToModify {boneName=dyn_collarD_01}
    rotation=<rot -25 -10 -30>
    translation {x=7}
   }
   n76 = Get offset_weapon_0() -> (offset_weapon_0:double)
   n77 = "Transform (Modify) Bone - Bone: dyn_collarE_01"(ComponentPose=@n75, Alpha=@n76, Translation="0,0,0", Rotation="3,0,0", Scale="1,1,1") {
    rotationMode=BMM_Additive
    rotationSpace=BCS_BoneSpace
    scaleMode=BMM_Additive
    scaleSpace=BCS_BoneSpace
    translationMode=BMM_Additive
    translationSpace=BCS_BoneSpace
    alphaBoolBlend {blendOption=Linear}
    boneToModify {boneName=dyn_collarE_01}
    rotation {pitch=3}
   }
   n78 = "Transform (Modify) Bone - Bone: dyn_collarD_01"(ComponentPose=@n77, Alpha=@n76, Translation="2,0,0", Rotation="-15,0,0", Scale="1,1,1") {
    rotationMode=BMM_Additive
    rotationSpace=BCS_BoneSpace
    scaleMode=BMM_Additive
    scaleSpace=BCS_BoneSpace
    translationMode=BMM_Additive
    translationSpace=BCS_BoneSpace
    alphaBoolBlend {blendOption=Linear}
    boneToModify {boneName=dyn_collarD_01}
    rotation {pitch=-15}
    translation {x=2}
   }
   n79 = Anim Dynamics - Bone: dyn_collarD_02(ComponentPose=@n78, Alpha=0.5, ExternalForce="0,0,0") {
    angularDampingOverride=0.999
    angularSpringConstant=800
    bAngularSpring=true
    bGravityOverrideInSimSpace=true
    bOverrideAngularDamping=true
    gravityScale=0.2
    simulationSpace=RootRelative
    alphaBoolBlend {blendOption=Linear}
    boundBone {boneName=dyn_collarD_02}
    componentAppliedLinearAccClamp=<10000 10000 10000>
    componentLinearAccScale=<0.1 0.1 0.1>
    componentLinearVelScale=<0.1 0.1 0.1>
    physicsBodyDefinitions [{
      sphereCollisionRadius=0
      boundBone {boneName=dyn_collarD_02}
      boxExtents=<20 20 10>
      constraintSetup {angularTargetAxis=AxisY, twistAxis=AxisZ, angularLimitsMax=<2 8 2>, angularLimitsMin=<-2 -20 -2>, angularTarget=<0 0 0>}
      localJointOffset=<0.1 0 -2>
     }]
    relativeSpaceBone {boneName=spine_05}
   }
   n80 = Anim Dynamics - Bone: dyn_collarE_02(ComponentPose=@n79, Alpha=0.5, ExternalForce="0,0,0") {
    angularDampingOverride=0.999
    angularSpringConstant=800
    bAngularSpring=true
    bGravityOverrideInSimSpace=true
    bOverrideAngularDamping=true
    gravityScale=0.2
    simulationSpace=RootRelative
    alphaBoolBlend {blendOption=Linear}
    boundBone {boneName=dyn_collarE_02}
    componentAppliedLinearAccClamp=<10000 10000 10000>
    componentLinearAccScale=<0.1 0.1 0.1>
    componentLinearVelScale=<0.1 0.1 0.1>
    physicsBodyDefinitions [{
      sphereCollisionRadius=0
      boundBone {boneName=dyn_collarE_02}
      boxExtents=<20 20 10>
      constraintSetup {angularTargetAxis=AxisY, twistAxis=AxisZ, angularLimitsMax=<2 15 2>, angularLimitsMin=<-2 -15 -2>, angularTarget=<0 0 0>}
      localJointOffset=<-0.1 -0.1 -2>
     }]
    relativeSpaceBone {boneName=spine_05}
   }
   n81 = "Constraint - dyn_collarA_01 (PP)"!(ComponentPose=@n80, Alpha=1, ConstraintWeights_0=0, ConstraintWeights_1:float=1) -> (Pose) {
    alphaBoolBlend {blendOption=Linear}
    boneToModify {boneName=dyn_collarA_01}
    constraintSetup [{targetBone {boneName=spine_05}}, {targetBone {boneName=head}}]
    constraintWeights [1,1]
   }
   n82 = "Constraint - dyn_collarB_01 (PP)"!(ComponentPose=@n81, Alpha=1, ConstraintWeights_0=0, ConstraintWeights_1:float=1) -> (Pose) {
    alphaBoolBlend {blendOption=Linear}
    boneToModify {boneName=dyn_collarB_01}
    constraintSetup [{targetBone {boneName=spine_05}}, {targetBone {boneName=head}}]
    constraintWeights [1,1]
   }
   n83 = "Constraint - dyn_collarC_01 (PP)"!(ComponentPose=@n82, Alpha=1, ConstraintWeights_0=0, ConstraintWeights_1:float=1) -> (Pose) {
    alphaBoolBlend {blendOption=Linear}
    boneToModify {boneName=dyn_collarC_01}
    constraintSetup [{targetBone {boneName=spine_05}}, {targetBone {boneName=head}}]
    constraintWeights [1,1]
   }
   n84 = "Transform (Modify) Bone - Bone: dyn_collarB_01"(ComponentPose=@n83, Alpha=Get neck_rot_up() -> (neck_rot_up:double), Translation="0,-1,0", Rotation="0,-35,0", Scale="1,1,1") {
    rotationMode=BMM_Additive
    rotationSpace=BCS_BoneSpace
    scaleMode=BMM_Additive
    scaleSpace=BCS_BoneSpace
    translationMode=BMM_Additive
    translationSpace=BCS_BoneSpace
    alphaBoolBlend {blendOption=Linear}
    boneToModify {boneName=dyn_collarB_01}
    rotation {yaw=-35}
    translation {y=-1}
   }
   n85 = "Transform (Modify) Bone - Bone: dyn_collarC_02"(ComponentPose=@n84, Alpha=Get neck_rot_up(), Translation="0,0,0", Rotation="-10,-50,0", Scale="1,1,1") {
    rotationMode=BMM_Additive
    rotationSpace=BCS_BoneSpace
    scaleMode=BMM_Additive
    scaleSpace=BCS_BoneSpace
    translationMode=BMM_Additive
    translationSpace=BCS_BoneSpace
    alphaBoolBlend {blendOption=Linear}
    boneToModify {boneName=dyn_collarC_02}
    rotation {pitch=-10, yaw=-50}
   }
   n86 = "Constraint - hand_l_Sleeves1 (PP)"!(ComponentPose=@n85, Alpha=1, ConstraintWeights_0=1, ConstraintWeights_1:float=0) -> (Pose) {
    alphaBoolBlend {blendOption=Linear}
    boneToModify {boneName=hand_l_Sleeves1}
    constraintSetup [{targetBone {boneName=lowerarm_twist_01_l}}, {targetBone {boneName=None}}]
    constraintWeights [1,1]
   }
   n87 = "Constraint - hand_l_Sleeves2 (PP)"!(ComponentPose=@n86, Alpha=1, ConstraintWeights_0=1, ConstraintWeights_1:float=0) -> (Pose) {
    alphaBoolBlend {blendOption=Linear}
    boneToModify {boneName=hand_l_Sleeves2}
    constraintSetup [{targetBone {boneName=lowerarm_twist_02_l}}, {targetBone {boneName=None}}]
    constraintWeights [1,1]
   }
   n88 = "Constraint - hand_r_Sleeves1 (PP)"!(ComponentPose=@n87, Alpha=1, ConstraintWeights_0=1, ConstraintWeights_1:float=0) -> (Pose) {
    alphaBoolBlend {blendOption=Linear}
    boneToModify {boneName=hand_r_Sleeves1}
    constraintSetup [{targetBone {boneName=lowerarm_twist_01_r}}, {targetBone {boneName=None}}]
    constraintWeights [1,1]
   }
   n89 = "Constraint - hand_r_Sleeves2 (PP)"!(ComponentPose=@n88, Alpha=1, ConstraintWeights_0=1, ConstraintWeights_1:float=0) -> (Pose) {
    alphaBoolBlend {blendOption=Linear}
    boneToModify {boneName=hand_r_Sleeves2}
    constraintSetup [{targetBone {boneName=lowerarm_twist_02_r}}, {targetBone {boneName=None}}]
    constraintWeights [1,1]
   }
   n90 = "Transform (Modify) Bone - Bone: hand_r_Sreeves_01"(ComponentPose=@n89, Alpha=Get hand_shape_up_r() -> (hand_shape_up_r:double), Translation="12,0,0", Rotation="0,0,0", Scale="1,1,1") {
    rotationMode=BMM_Additive
    rotationSpace=BCS_BoneSpace
    scaleMode=BMM_Additive
    scaleSpace=BCS_BoneSpace
    translationMode=BMM_Additive
    translationSpace=BCS_BoneSpace
    alphaBoolBlend {blendOption=Linear}
    boneToModify {boneName=hand_r_Sreeves_01}
    translation {x=12}
   }
   n91 = "Transform (Modify) Bone - Bone: hand_r_Sreeves_03"(ComponentPose=@n90, Alpha=Get hand_shape_dn_r() -> (hand_shape_dn_r:double), Translation="7,0,0", Rotation="0,0,0", Scale="1,1,1") {
    rotationMode=BMM_Additive
    rotationSpace=BCS_BoneSpace
    scaleMode=BMM_Additive
    scaleSpace=BCS_BoneSpace
    translationMode=BMM_Additive
    translationSpace=BCS_BoneSpace
    alphaBoolBlend {blendOption=Linear}
    boneToModify {boneName=hand_r_Sreeves_03}
    translation {x=7}
   }
   n92 = "Transform (Modify) Bone - Bone: hand_r_Sreeves_04"(ComponentPose=@n91, Alpha=Get hand_shape_rt_r() -> (hand_shape_rt_r:double), Translation="12,0,1", Rotation="0,0,0", Scale="1,1,1") {
    rotationMode=BMM_Additive
    rotationSpace=BCS_BoneSpace
    scaleMode=BMM_Additive
    scaleSpace=BCS_BoneSpace
    translationMode=BMM_Additive
    translationSpace=BCS_BoneSpace
    alphaBoolBlend {blendOption=Linear}
    boneToModify {boneName=hand_r_Sreeves_04}
    translation {x=12, z=1}
   }
   n93 = "Transform (Modify) Bone - Bone: hand_l_Sleeves_01"(ComponentPose=@n92, Alpha=Get hand_shape_up_l() -> (hand_shape_up_l:double), Translation="-12,0,0", Rotation="0,0,0", Scale="1,1,1") {
    rotationMode=BMM_Additive
    rotationSpace=BCS_BoneSpace
    scaleMode=BMM_Additive
    scaleSpace=BCS_BoneSpace
    translationMode=BMM_Additive
    translationSpace=BCS_BoneSpace
    alphaBoolBlend {blendOption=Linear}
    boneToModify {boneName=hand_l_Sleeves_01}
    translation {x=-12}
   }
   n94 = "Transform (Modify) Bone - Bone: hand_l_Sleeves_03"(ComponentPose=@n93, Alpha=Get hand_shape_dn_l() -> (hand_shape_dn_l:double), Translation="-7,0,0", Rotation="0,0,0", Scale="1,1,1") {
    rotationMode=BMM_Additive
    rotationSpace=BCS_BoneSpace
    scaleMode=BMM_Additive
    scaleSpace=BCS_BoneSpace
    translationMode=BMM_Additive
    translationSpace=BCS_BoneSpace
    alphaBoolBlend {blendOption=Linear}
    boneToModify {boneName=hand_l_Sleeves_03}
    translation {x=-7}
   }
   n95 = "Transform (Modify) Bone - Bone: hand_l_Sleeves_04"(ComponentPose=@n94, Alpha=Get hand_shape_rt_l() -> (hand_shape_rt_l:double), Translation="-12,0,1", Rotation="0,0,0", Scale="1,1,1") {
    rotationMode=BMM_Additive
    rotationSpace=BCS_BoneSpace
    scaleMode=BMM_Additive
    scaleSpace=BCS_BoneSpace
    translationMode=BMM_Additive
    translationSpace=BCS_BoneSpace
    alphaBoolBlend {blendOption=Linear}
    boneToModify {boneName=hand_l_Sleeves_04}
    translation {x=-12, z=1}
   }
   n96 = Get offset_weapon() -> (offset_weapon:double)
   n97 = "Transform (Modify) Bone - Bone: hand_r_Sleeves1"(ComponentPose=@n95, Alpha=@n96, Translation="4,0,0", Rotation="0,0,50", Scale="1,1,1") {
    rotationMode=BMM_Additive
    rotationSpace=BCS_BoneSpace
    scaleMode=BMM_Additive
    scaleSpace=BCS_BoneSpace
    translationMode=BMM_Additive
    translationSpace=BCS_BoneSpace
    alphaBoolBlend {blendOption=Linear}
    boneToModify {boneName=hand_r_Sleeves1}
    rotation {roll=50}
    translation {x=4}
   }
   n98 = "Transform (Modify) Bone - Bone: hand_l_Sleeves1"(ComponentPose=@n97, Alpha=@n96, Translation="-4,0,0", Rotation="0,0,50", Scale="1,1,1") {
    rotationMode=BMM_Additive
    rotationSpace=BCS_BoneSpace
    scaleMode=BMM_Additive
    scaleSpace=BCS_BoneSpace
    translationMode=BMM_Additive
    translationSpace=BCS_BoneSpace
    alphaBoolBlend {blendOption=Linear}
    boneToModify {boneName=hand_l_Sleeves1}
    rotation {roll=50}
    translation {x=-4}
   }
   n99 = "Transform (Modify) Bone - Bone: hand_r_Sreeves_02"(ComponentPose=@n98, Alpha=@n96, Translation="7,2,0", Rotation="0,0,30", Scale="1,1,1") {
    rotationMode=BMM_Additive
    rotationSpace=BCS_BoneSpace
    scaleMode=BMM_Additive
    scaleSpace=BCS_BoneSpace
    translationMode=BMM_Additive
    translationSpace=BCS_BoneSpace
    alphaBoolBlend {blendOption=Linear}
    boneToModify {boneName=hand_r_Sreeves_02}
    rotation {roll=30}
    translation=<7 2>
   }
   n100 = "Transform (Modify) Bone - Bone: hand_l_Sleeves_02"(ComponentPose=@n99, Alpha=@n96, Translation="-7,-2,0", Rotation="0,0,30", Scale="1,0.8,1") {
    rotationMode=BMM_Additive
    rotationSpace=BCS_BoneSpace
    scaleMode=BMM_Additive
    scaleSpace=BCS_BoneSpace
    translationMode=BMM_Additive
    translationSpace=BCS_BoneSpace
    alphaBoolBlend {blendOption=Linear}
    boneToModify {boneName=hand_l_Sleeves_02}
    rotation {roll=30}
    scale {y=0.8}
    translation=<-7 -2>
   }
   n101 = Get IsFrontEnd() -> (IsFrontEnd:double)
   n102 = "Transform (Modify) Bone - Bone: hand_l_Sleeves2"(ComponentPose=@n100, Alpha=@n101, Translation="-2,0,0", Rotation="0,0,0", Scale="1,0.9,0.9") {
    rotationMode=BMM_Additive
    rotationSpace=BCS_BoneSpace
    scaleMode=BMM_Additive
    scaleSpace=BCS_BoneSpace
    translationMode=BMM_Additive
    translationSpace=BCS_BoneSpace
    alphaBoolBlend {blendOption=Linear}
    boneToModify {boneName=hand_l_Sleeves2}
    scale {y=0.9, z=0.9}
    translation {x=-2}
   }
   n103 = "Transform (Modify) Bone - Bone: hand_l_Sleeves1"(ComponentPose=@n102, Alpha=@n101, Translation="-4,-5,3", Rotation="0,7,0", Scale="1,0.9,0.8") {
    rotationMode=BMM_Additive
    rotationSpace=BCS_BoneSpace
    scaleMode=BMM_Additive
    scaleSpace=BCS_BoneSpace
    translationMode=BMM_Additive
    translationSpace=BCS_BoneSpace
    alphaBoolBlend {blendOption=Linear}
    boneToModify {boneName=hand_l_Sleeves1}
    rotation {yaw=7}
    scale {y=0.9, z=0.8}
    translation=<-4 -5 3>
   }
   n104 = "Transform (Modify) Bone - Bone: hand_l_Sleeves_02"(ComponentPose=@n103, Alpha=@n101, Translation="0,0,0", Rotation="0,0,0", Scale="1,1,1") {
    rotationMode=BMM_Additive
    rotationSpace=BCS_BoneSpace
    scaleMode=BMM_Additive
    scaleSpace=BCS_BoneSpace
    translationMode=BMM_Additive
    translationSpace=BCS_BoneSpace
    alphaBoolBlend {blendOption=Linear}
    boneToModify {boneName=hand_l_Sleeves_02}
   }
   n105 = "Transform (Modify) Bone - Bone: hand_l_Sleeves_02_01"(ComponentPose=@n104, Alpha=@n101, Translation="0,-5,2", Rotation="0,15,0", Scale="1,0.85,1") {
    rotationMode=BMM_Additive
    rotationSpace=BCS_BoneSpace
    scaleMode=BMM_Additive
    scaleSpace=BCS_BoneSpace
    translationMode=BMM_Additive
    translationSpace=BCS_BoneSpace
    alphaBoolBlend {blendOption=Linear}
    boneToModify {boneName=hand_l_Sleeves_02_01}
    rotation {yaw=15}
    scale {y=0.85}
    translation {y=-5, z=2}
   }
   n106 = "Transform (Modify) Bone - Bone: upperarm_l"(ComponentPose=@n105, Alpha=@n101, Translation="0,0,0", Rotation="0,0,0", Scale="1,0.9,0.9") {
    rotationMode=BMM_Additive
    rotationSpace=BCS_BoneSpace
    scaleMode=BMM_Additive
    scaleSpace=BCS_BoneSpace
    translationMode=BMM_Additive
    translationSpace=BCS_BoneSpace
    alphaBoolBlend {blendOption=Linear}
    boneToModify {boneName=upperarm_l}
    scale {y=0.9, z=0.9}
   }
   n107 = "Transform (Modify) Bone - Bone: lowerarm_l"(ComponentPose=@n106, Alpha=@n101, Translation="0,0,0", Rotation="0,0,0", Scale="1,0.9,0.9") {
    rotationMode=BMM_Additive
    rotationSpace=BCS_BoneSpace
    scaleMode=BMM_Additive
    scaleSpace=BCS_BoneSpace
    translationMode=BMM_Additive
    translationSpace=BCS_BoneSpace
    alphaBoolBlend {blendOption=Linear}
    boneToModify {boneName=lowerarm_l}
    scale {y=0.9, z=0.9}
   }
   n108 = "Transform (Modify) Bone - Bone: hand_l_Sleeves_01"(ComponentPose=@n107, Alpha=@n101, Translation="0,5,-3", Rotation="0,0,0", Scale="1,1,1") {
    rotationMode=BMM_Additive
    rotationSpace=BCS_BoneSpace
    scaleMode=BMM_Additive
    scaleSpace=BCS_BoneSpace
    translationMode=BMM_Additive
    translationSpace=BCS_BoneSpace
    alphaBoolBlend {blendOption=Linear}
    boneToModify {boneName=hand_l_Sleeves_01}
    translation {y=5, z=-3}
   }
   n109 = "Transform (Modify) Bone - Bone: hand_r_Sleeves2"(ComponentPose=@n108, Alpha=@n101, Translation="2,0,0", Rotation="0,0,0", Scale="1,1,1") {
    rotationMode=BMM_Additive
    rotationSpace=BCS_BoneSpace
    scaleMode=BMM_Additive
    scaleSpace=BCS_BoneSpace
    translationMode=BMM_Additive
    translationSpace=BCS_BoneSpace
    alphaBoolBlend {blendOption=Linear}
    boneToModify {boneName=hand_r_Sleeves2}
    translation {x=2}
   }
   n110 = "Transform (Modify) Bone - Bone: hand_r_Sleeves1"(ComponentPose=@n109, Alpha=@n101, Translation="4,0,-3", Rotation="10,0,0", Scale="1,0.9,0.9") {
    rotationMode=BMM_Additive
    rotationSpace=BCS_BoneSpace
    scaleMode=BMM_Additive
    scaleSpace=BCS_BoneSpace
    translationMode=BMM_Additive
    translationSpace=BCS_BoneSpace
    alphaBoolBlend {blendOption=Linear}
    boneToModify {boneName=hand_r_Sleeves1}
    rotation {pitch=10}
    scale {y=0.9, z=0.9}
    translation {x=4, z=-3}
   }
   n111 = "Transform (Modify) Bone - Bone: hand_r_Sreeves_02"(ComponentPose=@n110, Alpha=@n101, Translation="0,0,0", Rotation="0,0,0", Scale="1,1,1") {
    rotationMode=BMM_Additive
    rotationSpace=BCS_BoneSpace
    scaleMode=BMM_Additive
    scaleSpace=BCS_BoneSpace
    translationMode=BMM_Additive
    translationSpace=BCS_BoneSpace
    alphaBoolBlend {blendOption=Linear}
    boneToModify {boneName=hand_r_Sreeves_02}
   }
   n112 = "Transform (Modify) Bone - Bone: hand_r_Sreeves_02_01"(ComponentPose=@n111, Alpha=@n101, Translation="0,-5,-2", Rotation="0,-15,0", Scale="1,0.85,1") {
    rotationMode=BMM_Additive
    rotationSpace=BCS_BoneSpace
    scaleMode=BMM_Additive
    scaleSpace=BCS_BoneSpace
    translationMode=BMM_Additive
    translationSpace=BCS_BoneSpace
    alphaBoolBlend {blendOption=Linear}
    boneToModify {boneName=hand_r_Sreeves_02_01}
    rotation {yaw=-15}
    scale {y=0.85}
    translation {y=-5, z=-2}
   }
   n113 = "Transform (Modify) Bone - Bone: upperarm_r"(ComponentPose=@n112, Alpha=@n101, Translation="0,0,0", Rotation="0,0,0", Scale="1,0.9,0.9") {
    rotationMode=BMM_Additive
    rotationSpace=BCS_BoneSpace
    scaleMode=BMM_Additive
    scaleSpace=BCS_BoneSpace
    translationMode=BMM_Additive
    translationSpace=BCS_BoneSpace
    alphaBoolBlend {blendOption=Linear}
    boneToModify {boneName=upperarm_r}
    scale {y=0.9, z=0.9}
   }
   n114 = "Transform (Modify) Bone - Bone: lowerarm_r"(ComponentPose=@n113, Alpha=@n101, Translation="0,0,0", Rotation="0,0,0", Scale="1,0.9,0.9") {
    rotationMode=BMM_Additive
    rotationSpace=BCS_BoneSpace
    scaleMode=BMM_Additive
    scaleSpace=BCS_BoneSpace
    translationMode=BMM_Additive
    translationSpace=BCS_BoneSpace
    alphaBoolBlend {blendOption=Linear}
    boneToModify {boneName=lowerarm_r}
    scale {y=0.9, z=0.9}
   }
   n115 = Get IsFrontEnd_1() -> (IsFrontEnd_1:double)
   n116 = Anim Dynamics - Bone: hand_l_Sleeves_02_01(ComponentPose=@n114, Alpha=@n115, ExternalForce="0,0,0") {
    angularDampingOverride=0.98
    angularSpringConstant=800
    bAngularSpring=true
    bGravityOverrideInSimSpace=true
    bOverrideAngularDamping=true
    bUseGravityOverride=true
    simulationSpace=BoneRelative
    alphaBoolBlend {blendOption=Linear}
    boundBone {boneName=hand_l_Sleeves_02_01}
    componentAppliedLinearAccClamp=<3000 3000 3000>
    componentLinearAccScale=<0.2 0.2 0.25>
    componentLinearVelScale=<0.45 0.45 0.45>
    gravityOverride {y=-100, z=-800}
    physicsBodyDefinitions [{
      sphereCollisionRadius=0
      boundBone {boneName=hand_l_Sleeves_02_01}
      boxExtents=<20 20 20>
      constraintSetup {angularTargetAxis=AxisY, twistAxis=AxisZ, angularLimitsMax=<30 30 30>, angularLimitsMin=<-30 -30 -30>, angularTarget=<0 0 0>}
      localJointOffset=<0.01 -0.01 -2>
     }]
    relativeSpaceBone {boneName=hand_l_Sreeves_gravity}
   }
   n117 = Anim Dynamics - Bone: hand_r_Sreeves_02_01(ComponentPose=@n116, Alpha=@n115, ExternalForce="0,0,0") {
    angularDampingOverride=0.98
    angularSpringConstant=800
    bAngularSpring=true
    bGravityOverrideInSimSpace=true
    bOverrideAngularDamping=true
    bUseGravityOverride=true
    simulationSpace=BoneRelative
    alphaBoolBlend {blendOption=Linear}
    boundBone {boneName=hand_r_Sreeves_02_01}
    componentAppliedLinearAccClamp=<3000 3000 3000>
    componentLinearAccScale=<0.2 0.2 0.25>
    componentLinearVelScale=<0.45 0.45 0.45>
    gravityOverride {y=-100, z=-800}
    physicsBodyDefinitions [{
      sphereCollisionRadius=0
      boundBone {boneName=hand_r_Sreeves_02_01}
      boxExtents=<20 20 20>
      constraintSetup {angularTargetAxis=AxisY, twistAxis=AxisZ, angularLimitsMax=<30 30 30>, angularLimitsMin=<-30 -30 -30>, angularTarget=<0 0 0>}
      localJointOffset=<0.01 0.01 5>
     }]
    relativeSpaceBone {boneName=hand_r_Sreeves_gravity}
   }
   n118 = Anim Dynamics - Bone: hand_l_Sleeves_01(ComponentPose=@n117, Alpha=0.7, ExternalForce="0,0,0") {
    angularDampingOverride=0.98
    angularSpringConstant=800
    bAngularSpring=true
    bGravityOverrideInSimSpace=true
    bOverrideAngularDamping=true
    bUseGravityOverride=true
    simulationSpace=BoneRelative
    alphaBoolBlend {blendOption=Linear}
    boundBone {boneName=hand_l_Sleeves_01}
    componentAppliedLinearAccClamp=<3000 3000 3000>
    componentLinearAccScale=<0.2 0.2 0.25>
    componentLinearVelScale=<0.45 0.45 0.45>
    gravityOverride {y=-100, z=-800}
    physicsBodyDefinitions [{
      sphereCollisionRadius=0
      boundBone {boneName=hand_l_Sleeves_01}
      boxExtents=<20 20 15>
      constraintSetup {angularTargetAxis=AxisY, twistAxis=AxisZ, angularLimitsMax=<30 30 30>, angularLimitsMin=<-30 -30 -30>, angularTarget=<0 0 0>}
      localJointOffset=<0.01 -0.01 -2>
     }]
    relativeSpaceBone {boneName=hand_l_Sreeves_gravity}
   }
   n119 = Anim Dynamics - Bone: hand_l_Sleeves_03(ComponentPose=@n118, Alpha=0.7, ExternalForce="0,0,0") {
    angularDampingOverride=0.98
    angularSpringConstant=800
    bAngularSpring=true
    bGravityOverrideInSimSpace=true
    bOverrideAngularDamping=true
    bUseGravityOverride=true
    simulationSpace=BoneRelative
    alphaBoolBlend {blendOption=Linear}
    boundBone {boneName=hand_l_Sleeves_03}
    componentAppliedLinearAccClamp=<3000 3000 3000>
    componentLinearAccScale=<0.2 0.2 0.25>
    componentLinearVelScale=<0.45 0.45 0.45>
    gravityOverride {y=-100, z=-800}
    physicsBodyDefinitions [{
      sphereCollisionRadius=0
      boundBone {boneName=hand_l_Sleeves_03}
      boxExtents=<20 20 15>
      constraintSetup {angularTargetAxis=AxisY, twistAxis=AxisZ, angularLimitsMax=<30 30 30>, angularLimitsMin=<-30 -30 -30>, angularTarget=<0 0 0>}
      localJointOffset=<0.01 -0.01 2>
     }]
    relativeSpaceBone {boneName=hand_r_Sreeves_gravity}
   }
   n120 = Anim Dynamics - Bone: hand_r_Sreeves_01(ComponentPose=@n119, Alpha=0.7, ExternalForce="0,0,0") {
    angularDampingOverride=0.98
    angularSpringConstant=800
    bAngularSpring=true
    bGravityOverrideInSimSpace=true
    bOverrideAngularDamping=true
    bUseGravityOverride=true
    simulationSpace=BoneRelative
    alphaBoolBlend {blendOption=Linear}
    boundBone {boneName=hand_r_Sreeves_01}
    componentAppliedLinearAccClamp=<3000 3000 3000>
    componentLinearAccScale=<0.2 0.2 0.25>
    componentLinearVelScale=<0.45 0.45 0.45>
    gravityOverride {y=-100, z=-800}
    physicsBodyDefinitions [{
      sphereCollisionRadius=0
      boundBone {boneName=hand_r_Sreeves_01}
      boxExtents=<20 20 15>
      constraintSetup {angularTargetAxis=AxisY, twistAxis=AxisZ, angularLimitsMax=<30 30 30>, angularLimitsMin=<-30 -30 -30>, angularTarget=<0 0 0>}
      localJointOffset=<0.01 -0.01 -2>
     }]
    relativeSpaceBone {boneName=hand_l_Sreeves_gravity}
   }
   n121 = Anim Dynamics - Bone: hand_r_Sreeves_03(ComponentPose=@n120, Alpha=0.7, ExternalForce="0,0,0") {
    angularDampingOverride=0.98
    angularSpringConstant=800
    bAngularSpring=true
    bGravityOverrideInSimSpace=true
    bOverrideAngularDamping=true
    bUseGravityOverride=true
    simulationSpace=BoneRelative
    alphaBoolBlend {blendOption=Linear}
    boundBone {boneName=hand_r_Sreeves_03}
    componentAppliedLinearAccClamp=<3000 3000 3000>
    componentLinearAccScale=<0.2 0.2 0.25>
    componentLinearVelScale=<0.45 0.45 0.45>
    gravityOverride {y=-100, z=-800}
    physicsBodyDefinitions [{
      sphereCollisionRadius=0
      boundBone {boneName=hand_r_Sreeves_03}
      boxExtents=<20 20 15>
      constraintSetup {angularTargetAxis=AxisY, twistAxis=AxisZ, angularLimitsMax=<30 30 30>, angularLimitsMin=<-30 -30 -30>, angularTarget=<0 0 0>}
      localJointOffset=<0.01 -0.01 2>
     }]
    relativeSpaceBone {boneName=hand_r_Sreeves_gravity}
   }
   n122 = "Constraint - flont_uppelalm_l (PP)"!(ComponentPose=@n121, Alpha=1, ConstraintWeights_0=0.5, ConstraintWeights_1:float=0.5) -> (Pose) {
    alphaBoolBlend {blendOption=Linear}
    boneToModify {boneName=flont_uppelalm_l}
    constraintSetup [{targetBone {boneName=upperarm_twist_01_l}}, {targetBone {boneName=clavicle_l}}]
    constraintWeights [1,1]
   }
   n123 = "Constraint - up_uppelalm_l (PP)"!(ComponentPose=@n122, Alpha=1, ConstraintWeights_0=0.5, ConstraintWeights_1:float=0.5) -> (Pose) {
    alphaBoolBlend {blendOption=Linear}
    boneToModify {boneName=up_uppelalm_l}
    constraintSetup [{targetBone {boneName=upperarm_twist_01_l}}, {targetBone {boneName=clavicle_l}}]
    constraintWeights [1,1]
   }
   n124 = "Constraint - front_upperarm_r (PP)"!(ComponentPose=@n123, Alpha=1, ConstraintWeights_0=0.5, ConstraintWeights_1:float=0.5) -> (Pose) {
    alphaBoolBlend {blendOption=Linear}
    boneToModify {boneName=front_upperarm_r}
    constraintSetup [{targetBone {boneName=upperarm_twist_01_r}}, {targetBone {boneName=clavicle_r}}]
    constraintWeights [1,1]
   }
   n125 = "Constraint - up_upperarm_r (PP)"!(ComponentPose=@n124, Alpha=1, ConstraintWeights_0=0.5, ConstraintWeights_1:float=0.5) -> (Pose) {
    alphaBoolBlend {blendOption=Linear}
    boneToModify {boneName=up_upperarm_r}
    constraintSetup [{targetBone {boneName=upperarm_twist_01_r}}, {targetBone {boneName=clavicle_r}}]
    constraintWeights [1,1]
   }
   n126 = Get Shoulder_rot_f_l() -> (Shoulder_rot_f_l:double)
   n127 = "Transform (Modify) Bone - Bone: flont_uppelalm_l"(ComponentPose=@n125, Alpha=@n126, Translation="-5,5,0", Rotation="0,0,0", Scale="1,1,1") {
    rotationMode=BMM_Additive
    rotationSpace=BCS_BoneSpace
    scaleMode=BMM_Additive
    scaleSpace=BCS_BoneSpace
    translationMode=BMM_Additive
    translationSpace=BCS_BoneSpace
    alphaBoolBlend {blendOption=Linear}
    boneToModify {boneName=flont_uppelalm_l}
    translation=<-5 5>
   }
   n128 = Get Shoulder_rot_f_r()
   n129 = "Transform (Modify) Bone - Bone: front_upperarm_r"(ComponentPose=@n127, Alpha=@n128, Translation="5,-5,0", Rotation="0,0,0", Scale="1,1,1") {
    rotationMode=BMM_Additive
    rotationSpace=BCS_BoneSpace
    scaleMode=BMM_Additive
    scaleSpace=BCS_BoneSpace
    translationMode=BMM_Additive
    translationSpace=BCS_BoneSpace
    alphaBoolBlend {blendOption=Linear}
    boneToModify {boneName=front_upperarm_r}
    translation=<5 -5>
   }
   n130 = "Transform (Modify) Bone - Bone: up_uppelalm_l"(ComponentPose=@n129, Alpha=@n126, Translation="0,2,3", Rotation="0,0,0", Scale="1,1,1") {
    rotationMode=BMM_Additive
    rotationSpace=BCS_BoneSpace
    scaleMode=BMM_Additive
    scaleSpace=BCS_BoneSpace
    translationMode=BMM_Additive
    translationSpace=BCS_BoneSpace
    alphaBoolBlend {blendOption=Linear}
    boneToModify {boneName=up_uppelalm_l}
    translation {y=2, z=3}
   }
   n131 = "Transform (Modify) Bone - Bone: up_upperarm_r"(ComponentPose=@n130, Alpha=@n128, Translation="0,-7,-3", Rotation="0,0,0", Scale="1,1,1") {
    rotationMode=BMM_Additive
    rotationSpace=BCS_BoneSpace
    scaleMode=BMM_Additive
    scaleSpace=BCS_BoneSpace
    translationMode=BMM_Additive
    translationSpace=BCS_BoneSpace
    alphaBoolBlend {blendOption=Linear}
    boneToModify {boneName=up_upperarm_r}
    translation {y=-7, z=-3}
   }
   n132 = "Constraint - lowerarm_l_front (RR)"!(ComponentPose=@n131, Alpha=1, ConstraintWeights_0=0.5, ConstraintWeights_1:float=0.5) -> (Pose) {
    alphaBoolBlend {blendOption=Linear}
    boneToModify {boneName=lowerarm_l_front}
    constraintSetup [{transformType=Rotation, targetBone {boneName=lowerarm_l}}, {transformType=Rotation, targetBone {boneName=upperarm_l}}]
    constraintWeights [1,1]
   }
   n133 = "Constraint - lowerarm_r_front (RR)"!(ComponentPose=@n132, Alpha=1, ConstraintWeights_0=0.5, ConstraintWeights_1:float=0.5) -> (Pose) {
    alphaBoolBlend {blendOption=Linear}
    boneToModify {boneName=lowerarm_r_front}
    constraintSetup [{transformType=Rotation, targetBone {boneName=lowerarm_r}}, {transformType=Rotation, targetBone {boneName=upperarm_r}}]
    constraintWeights [1,1]
   }
   n134 = "Constraint - lowerarm_l_back (RR)"!(ComponentPose=@n133, Alpha=1, ConstraintWeights_0=0.5, ConstraintWeights_1:float=0.5) -> (Pose) {
    alphaBoolBlend {blendOption=Linear}
    boneToModify {boneName=lowerarm_l_back}
    constraintSetup [{transformType=Rotation, targetBone {boneName=lowerarm_l}}, {transformType=Rotation, targetBone {boneName=upperarm_l}}]
    constraintWeights [1,1]
   }
   n135 = "Constraint - lowerarm_r_back (RR)"!(ComponentPose=@n134, Alpha=1, ConstraintWeights_0=0.5, ConstraintWeights_1:float=0.5) -> (Pose) {
    alphaBoolBlend {blendOption=Linear}
    boneToModify {boneName=lowerarm_r_back}
    constraintSetup [{transformType=Rotation, targetBone {boneName=lowerarm_r}}, {transformType=Rotation, targetBone {boneName=upperarm_r}}]
    constraintWeights [1,1]
   }
   n136 = Get lowerarm_l_front() -> (lowerarm_l_front:double)
   n137 = "Transform (Modify) Bone - Bone: lowerarm_l_front"(ComponentPose=@n135, Alpha=@n136, Translation="0,10,0", Rotation="0,0,0", Scale="1,1,1") {
    rotationMode=BMM_Additive
    rotationSpace=BCS_BoneSpace
    scaleMode=BMM_Additive
    scaleSpace=BCS_BoneSpace
    translationMode=BMM_Additive
    translationSpace=BCS_BoneSpace
    alphaBoolBlend {blendOption=Linear}
    boneToModify {boneName=lowerarm_l_front}
    translation {y=10}
   }
   n138 = Get lowerarm_r_front() -> (lowerarm_r_front:double)
   n139 = "Transform (Modify) Bone - Bone: lowerarm_r_front"(ComponentPose=@n137, Alpha=@n138, Translation="0,-10,0", Rotation="0,0,0", Scale="1,1,1") {
    rotationMode=BMM_Additive
    rotationSpace=BCS_BoneSpace
    scaleMode=BMM_Additive
    scaleSpace=BCS_BoneSpace
    translationMode=BMM_Additive
    translationSpace=BCS_BoneSpace
    alphaBoolBlend {blendOption=Linear}
    boneToModify {boneName=lowerarm_r_front}
    translation {y=-10}
   }
   n140 = "Transform (Modify) Bone - Bone: lowerarm_l_back"(ComponentPose=@n139, Alpha=@n136, Translation="0,-50,0", Rotation="0,0,0", Scale="1,1,1") {
    rotationMode=BMM_Additive
    rotationSpace=BCS_BoneSpace
    scaleMode=BMM_Additive
    scaleSpace=BCS_BoneSpace
    translationMode=BMM_Additive
    translationSpace=BCS_BoneSpace
    alphaBoolBlend {blendOption=Linear}
    boneToModify {boneName=lowerarm_l_back}
    translation {y=-50}
   }
   n141 = "Transform (Modify) Bone - Bone: lowerarm_r_back"(ComponentPose=@n140, Alpha=@n138, Translation="0,2,-3", Rotation="0,0,0", Scale="1,1,1") {
    rotationMode=BMM_Additive
    rotationSpace=BCS_BoneSpace
    scaleMode=BMM_Additive
    scaleSpace=BCS_BoneSpace
    translationMode=BMM_Additive
    translationSpace=BCS_BoneSpace
    alphaBoolBlend {blendOption=Linear}
    boneToModify {boneName=lowerarm_r_back}
    translation {y=2, z=-3}
   }
   n142 = "Constraint - upperarm_twist_02_l (PP)"!(ComponentPose=@n141, Alpha=1, ConstraintWeights_0=0, ConstraintWeights_1:float=1) -> (Pose) {
    alphaBoolBlend {blendOption=Linear}
    boneToModify {boneName=upperarm_twist_02_l}
    constraintSetup [{targetBone {boneName=lowerarm_l}}, {targetBone {boneName=upperarm_twist_02_l}}]
    constraintWeights [1,1]
   }
   n143 = "Constraint - upperarm_twist_01_l (PP)"!(ComponentPose=@n142, Alpha=1, ConstraintWeights_0=0, ConstraintWeights_1:float=1) -> (Pose) {
    alphaBoolBlend {blendOption=Linear}
    boneToModify {boneName=upperarm_twist_01_l}
    constraintSetup [{targetBone {boneName=lowerarm_l}}, {targetBone {boneName=upperarm_twist_02_l}}]
    constraintWeights [1,1]
   }
   n144 = "Constraint - hand_l_Sleeves2 (PP)"(ComponentPose=@n143, Alpha=1, ConstraintWeights_0=1, ConstraintWeights_1=0) {
    alphaBoolBlend {blendOption=Linear}
    boneToModify {boneName=hand_l_Sleeves2}
    constraintSetup [{targetBone {boneName=lowerarm_l}}, {targetBone {boneName=upperarm_twist_02_l}}]
    constraintWeights [1,1]
   }
   n145 = "Constraint - hand_l_Sleeves1 (PP)"(ComponentPose=@n144, Alpha=1, ConstraintWeights_0=1, ConstraintWeights_1=0) {
    alphaBoolBlend {blendOption=Linear}
    boneToModify {boneName=hand_l_Sleeves1}
    constraintSetup [{targetBone {boneName=lowerarm_l}}, {targetBone {boneName=upperarm_twist_02_l}}]
    constraintWeights [1,1]
   }
   n146 = "Constraint - upperarm_twist_02_r (PP)"!(ComponentPose=@n145, Alpha=1, ConstraintWeights_0=0, ConstraintWeights_1:float=1) -> (Pose) {
    alphaBoolBlend {blendOption=Linear}
    boneToModify {boneName=upperarm_twist_02_r}
    constraintSetup [{targetBone {boneName=lowerarm_r}}, {targetBone {boneName=upperarm_twist_02_r}}]
    constraintWeights [1,1]
   }
   n147 = "Constraint - upperarm_twist_01_r (PP)"!(ComponentPose=@n146, Alpha=1, ConstraintWeights_0=0, ConstraintWeights_1:float=1) -> (Pose) {
    alphaBoolBlend {blendOption=Linear}
    boneToModify {boneName=upperarm_twist_01_r}
    constraintSetup [{targetBone {boneName=lowerarm_r}}, {targetBone {boneName=upperarm_twist_02_r}}]
    constraintWeights [1,1]
   }
   n148 = "Constraint - hand_r_Sleeves2 (PP)"(ComponentPose=@n147, Alpha=1, ConstraintWeights_0=1, ConstraintWeights_1=0) {
    alphaBoolBlend {blendOption=Linear}
    boneToModify {boneName=hand_r_Sleeves2}
    constraintSetup [{targetBone {boneName=lowerarm_r}}, {targetBone {boneName=upperarm_twist_02_r}}]
    constraintWeights [1,1]
   }
   n149 = "Constraint - hand_r_Sleeves1 (PP)"(ComponentPose=@n148, Alpha=1, ConstraintWeights_0=1, ConstraintWeights_1=0) {
    alphaBoolBlend {blendOption=Linear}
    boneToModify {boneName=hand_r_Sleeves1}
    constraintSetup [{targetBone {boneName=lowerarm_r}}, {targetBone {boneName=upperarm_twist_02_r}}]
    constraintWeights [1,1]
   }
   n150 = "Constraint - thigh_volume_fn_l (PP)"!(ComponentPose=@n149, Alpha=1, ConstraintWeights_0=1, ConstraintWeights_1:float=0) -> (Pose) {
    alphaBoolBlend {blendOption=Linear}
    boneToModify {boneName=thigh_volume_fn_l}
    constraintSetup [{targetBone {boneName=pelvis}}, {targetBone {boneName=thigh_twist_01_l}}]
    constraintWeights [1,1]
   }
   n151 = "Constraint - thigh_volume_fn_r (PP)"!(ComponentPose=@n150, Alpha=1, ConstraintWeights_0=1, ConstraintWeights_1:float=0) -> (Pose) {
    alphaBoolBlend {blendOption=Linear}
    boneToModify {boneName=thigh_volume_fn_r}
    constraintSetup [{targetBone {boneName=pelvis}}, {targetBone {boneName=thigh_twist_01_l}}]
    constraintWeights [1,1]
   }
   n152 = "Transform (Modify) Bone - Bone: thigh_volume_fn_l"(ComponentPose=@n151, Alpha=Get thigh_shape_fn_l() -> (thigh_shape_fn_l:double), Translation="14,-2,0", Rotation="0,0,0", Scale="1,1,1") {
    rotationMode=BMM_Additive
    rotationSpace=BCS_BoneSpace
    scaleMode=BMM_Additive
    scaleSpace=BCS_BoneSpace
    translationMode=BMM_Additive
    translationSpace=BCS_BoneSpace
    alphaBoolBlend {blendOption=Linear}
    boneToModify {boneName=thigh_volume_fn_l}
    translation=<14 -2>
   }
   n153 = "Transform (Modify) Bone - Bone: thigh_volume_fn_l"(ComponentPose=@n152, Alpha=Get thigh_shape_fn_r() -> (thigh_shape_fn_r:double), Translation="14,-2,0", Rotation="0,0,0", Scale="1,1,1") {
    rotationMode=BMM_Additive
    rotationSpace=BCS_BoneSpace
    scaleMode=BMM_Additive
    scaleSpace=BCS_BoneSpace
    translationMode=BMM_Additive
    translationSpace=BCS_BoneSpace
    alphaBoolBlend {blendOption=Linear}
    boneToModify {boneName=thigh_volume_fn_l}
    translation=<14 -2>
   }
   n154 = "Constraint - dyn_SkirtB01 (PP)"!(ComponentPose=@n153, Alpha=1, ConstraintWeights_0=1, ConstraintWeights_1:float=0) -> (Pose) {
    alphaBoolBlend {blendOption=Linear}
    boneToModify {boneName=dyn_SkirtB01}
    constraintSetup [{targetBone {boneName=spine_02}}, {targetBone {boneName=spine_03}}]
    constraintWeights [1,1]
   }
   n155 = "Constraint - dyn_SkirtD01 (PP)"!(ComponentPose=@n154, Alpha=1, ConstraintWeights_0=1, ConstraintWeights_1:float=0) -> (Pose) {
    alphaBoolBlend {blendOption=Linear}
    boneToModify {boneName=dyn_SkirtD01}
    constraintSetup [{targetBone {boneName=spine_02}}, {targetBone {boneName=spine_03}}]
    constraintWeights [1,1]
   }
   n156 = "Constraint - dyn_SkirtE01 (PP)"!(ComponentPose=@n155, Alpha=1, ConstraintWeights_0=1, ConstraintWeights_1:float=0) -> (Pose) {
    alphaBoolBlend {blendOption=Linear}
    boneToModify {boneName=dyn_SkirtE01}
    constraintSetup [{targetBone {boneName=spine_02}}, {targetBone {boneName=spine_03}}]
    constraintWeights [1,1]
   }
   n157 = "Constraint - dyn_SkirtF01 (PP)"!(ComponentPose=@n156, Alpha=1, ConstraintWeights_0=1, ConstraintWeights_1:float=0) -> (Pose) {
    alphaBoolBlend {blendOption=Linear}
    boneToModify {boneName=dyn_SkirtF01}
    constraintSetup [{targetBone {boneName=spine_02}}, {targetBone {boneName=spine_03}}]
    constraintWeights [1,1]
   }
   n158 = "Constraint - dyn_SkirtG01 (PP)"!(ComponentPose=@n157, Alpha=1, ConstraintWeights_0=1, ConstraintWeights_1:float=0) -> (Pose) {
    alphaBoolBlend {blendOption=Linear}
    boneToModify {boneName=dyn_SkirtG01}
    constraintSetup [{targetBone {boneName=spine_02}}, {targetBone {boneName=spine_03}}]
    constraintWeights [1,1]
   }
   n159 = "Constraint - dyn_SkirtH01 (PP)"!(ComponentPose=@n158, Alpha=1, ConstraintWeights_0=1, ConstraintWeights_1:float=0) -> (Pose) {
    alphaBoolBlend {blendOption=Linear}
    boneToModify {boneName=dyn_SkirtH01}
    constraintSetup [{targetBone {boneName=spine_02}}, {targetBone {boneName=spine_03}}]
    constraintWeights [1,1]
   }
   n160 = "Constraint - dyn_SkirtJ01 (PP)"!(ComponentPose=@n159, Alpha=1, ConstraintWeights_0=1, ConstraintWeights_1:float=0) -> (Pose) {
    alphaBoolBlend {blendOption=Linear}
    boneToModify {boneName=dyn_SkirtJ01}
    constraintSetup [{targetBone {boneName=spine_02}}, {targetBone {boneName=spine_03}}]
    constraintWeights [1,1]
   }
   n161 = "Constraint - dyn_SkirtB02 (PP)"!(ComponentPose=@n160, Alpha=1, ConstraintWeights_0=1, ConstraintWeights_1:float=0) -> (Pose) {
    alphaBoolBlend {blendOption=Linear}
    boneToModify {boneName=dyn_SkirtB02}
    constraintSetup [{targetBone {boneName=pelvis}}, {targetBone {boneName=spine_03}}]
    constraintWeights [1,1]
   }
   n162 = "Constraint - dyn_SkirtD02 (PP)"!(ComponentPose=@n161, Alpha=1, ConstraintWeights_0=1, ConstraintWeights_1:float=0) -> (Pose) {
    alphaBoolBlend {blendOption=Linear}
    boneToModify {boneName=dyn_SkirtD02}
    constraintSetup [{targetBone {boneName=pelvis}}, {targetBone {boneName=spine_03}}]
    constraintWeights [1,1]
   }
   n163 = "Constraint - dyn_SkirtE02 (PP)"!(ComponentPose=@n162, Alpha=1, ConstraintWeights_0=1, ConstraintWeights_1:float=0) -> (Pose) {
    alphaBoolBlend {blendOption=Linear}
    boneToModify {boneName=dyn_SkirtE02}
    constraintSetup [{targetBone {boneName=pelvis}}, {targetBone {boneName=spine_03}}]
    constraintWeights [1,1]
   }
   n164 = "Constraint - dyn_SkirtF02 (PP)"!(ComponentPose=@n163, Alpha=1, ConstraintWeights_0=1, ConstraintWeights_1:float=0) -> (Pose) {
    alphaBoolBlend {blendOption=Linear}
    boneToModify {boneName=dyn_SkirtF02}
    constraintSetup [{targetBone {boneName=pelvis}}, {targetBone {boneName=spine_03}}]
    constraintWeights [1,1]
   }
   n165 = "Constraint - dyn_SkirtG02 (PP)"!(ComponentPose=@n164, Alpha=1, ConstraintWeights_0=1, ConstraintWeights_1:float=0) -> (Pose) {
    alphaBoolBlend {blendOption=Linear}
    boneToModify {boneName=dyn_SkirtG02}
    constraintSetup [{targetBone {boneName=pelvis}}, {targetBone {boneName=spine_03}}]
    constraintWeights [1,1]
   }
   n166 = "Constraint - dyn_SkirtH02 (PP)"!(ComponentPose=@n165, Alpha=1, ConstraintWeights_0=1, ConstraintWeights_1:float=0) -> (Pose) {
    alphaBoolBlend {blendOption=Linear}
    boneToModify {boneName=dyn_SkirtH02}
    constraintSetup [{targetBone {boneName=pelvis}}, {targetBone {boneName=spine_03}}]
    constraintWeights [1,1]
   }
   n167 = "Constraint - dyn_SkirtJ02 (PP)"!(ComponentPose=@n166, Alpha=1, ConstraintWeights_0=1, ConstraintWeights_1:float=0) -> (Pose) {
    alphaBoolBlend {blendOption=Linear}
    boneToModify {boneName=dyn_SkirtJ02}
    constraintSetup [{targetBone {boneName=pelvis}}, {targetBone {boneName=spine_03}}]
    constraintWeights [1,1]
   }
   n168 = "Constraint - dyn_SkirtK01 (PP)"!(ComponentPose=@n167, Alpha=1, ConstraintWeights_0=1, ConstraintWeights_1:float=0) -> (Pose) {
    alphaBoolBlend {blendOption=Linear}
    boneToModify {boneName=dyn_SkirtK01}
    constraintSetup [{targetBone {boneName=spine_02}}, {targetBone {boneName=spine_03}}]
    constraintWeights [1,1]
   }
   n169 = "Constraint - dyn_SkirtK02 (PP)"!(ComponentPose=@n168, Alpha=1, ConstraintWeights_0=1, ConstraintWeights_1:float=0) -> (Pose) {
    alphaBoolBlend {blendOption=Linear}
    boneToModify {boneName=dyn_SkirtK02}
    constraintSetup [{targetBone {boneName=pelvis}}, {targetBone {boneName=spine_03}}]
    constraintWeights [1,1]
   }
   n170 = "Constraint - dyn_SkirtB03 (RR)"!(ComponentPose=@n169, Alpha=Get thigh_shape_fn_l_0() -> (thigh_shape_fn_l_0:double), ConstraintWeights_0=1, ConstraintWeights_1:float=0) -> (Pose) {
    alphaBoolBlend {blendOption=Linear}
    boneToModify {boneName=dyn_SkirtB03}
    constraintSetup [{transformType=Rotation, targetBone {boneName=thigh_twist_01_l}}, {transformType=Rotation, targetBone {boneName=thigh_twist_01_r}}]
    constraintWeights [1,1]
   }
   n171 = Get calf_shape_fn_l() -> (calf_shape_fn_l:double)
   n172 = "Constraint - dyn_SkirtB06 (PP)"(ComponentPose=@n170, Alpha=@n171, ConstraintWeights_0=0.5, ConstraintWeights_1=0.5) {
    alphaBoolBlend {blendOption=Linear}
    boneToModify {boneName=dyn_SkirtB06}
    constraintSetup [{targetBone {boneName=calf_twist_02_l}}, {targetBone {boneName=dyn_SkirtB05}}]
    constraintWeights [1,1]
   }
   n173 = "Transform (Modify) Bone - Bone: dyn_SkirtB03"(ComponentPose=@n172, Alpha=Get thigh_shape_fn_l(), Translation="-4,13,-2", Rotation="65,10,-10", Scale="1,1,1") {
    rotationMode=BMM_Additive
    rotationSpace=BCS_BoneSpace
    scaleMode=BMM_Additive
    scaleSpace=BCS_BoneSpace
    translationMode=BMM_Additive
    translationSpace=BCS_BoneSpace
    alphaBoolBlend {blendOption=Linear}
    boneToModify {boneName=dyn_SkirtB03}
    rotation=<rot 65 10 -10>
    translation=<-4 13 -2>
   }
   n174 = "Transform (Modify) Bone - Bone: dyn_SkirtB04"(ComponentPose=@n173, Alpha=Get thigh_shape_fn_l(), Translation="0,0,5", Rotation="-50,0,10", Scale="1,1,1") {
    rotationMode=BMM_Additive
    rotationSpace=BCS_BoneSpace
    scaleMode=BMM_Additive
    scaleSpace=BCS_BoneSpace
    translationMode=BMM_Additive
    translationSpace=BCS_BoneSpace
    alphaBoolBlend {blendOption=Linear}
    boneToModify {boneName=dyn_SkirtB04}
    rotation {pitch=-50, roll=10}
    translation {z=5}
   }
   n175 = "Transform (Modify) Bone - Bone: dyn_SkirtB03"(ComponentPose=@n174, Alpha=Get thigh_shape_lt_l() -> (thigh_shape_lt_l:double), Translation="0,3,0", Rotation="0,10,-10", Scale="1,1,1") {
    rotationMode=BMM_Additive
    rotationSpace=BCS_BoneSpace
    scaleMode=BMM_Additive
    scaleSpace=BCS_BoneSpace
    translationMode=BMM_Additive
    translationSpace=BCS_BoneSpace
    alphaBoolBlend {blendOption=Linear}
    boneToModify {boneName=dyn_SkirtB03}
    rotation {roll=-10, yaw=10}
    translation {y=3}
   }
   n176 = "Constraint - dyn_SkirtE03 (PP)"!(ComponentPose=@n175, Alpha=1, ConstraintWeights_0=0, ConstraintWeights_1:float=1) -> (Pose) {
    alphaBoolBlend {blendOption=Linear}
    boneToModify {boneName=dyn_SkirtE03}
    constraintSetup [{targetBone {boneName=thigh_twist_01_l}}, {targetBone {boneName=pelvis}}]
    constraintWeights [1,1]
   }
   n177 = "Constraint - dyn_SkirtE03 (PP)"(ComponentPose=@n176, Alpha=1, ConstraintWeights_0=0.5, ConstraintWeights_1=0.5) {
    alphaBoolBlend {blendOption=Linear}
    boneToModify {boneName=dyn_SkirtE03}
    constraintSetup [{targetBone {boneName=thigh_twist_01_l}}, {targetBone {boneName=pelvis}}]
    constraintWeights [1,1]
   }
   n178 = Get calf_shape_back_l() -> (calf_shape_back_l:double)
   n179 = "Constraint - dyn_SkirtE05 (PP)"(ComponentPose=@n177, Alpha=@n178, ConstraintWeights_0=0.7, ConstraintWeights_1=0.3) {
    alphaBoolBlend {blendOption=Linear}
    boneToModify {boneName=dyn_SkirtE05}
    constraintSetup [{targetBone {boneName=calf_twist_02_l}}, {targetBone {boneName=dyn_SkirtE03}}]
    constraintWeights [1,1]
   }
   n180 = "Transform (Modify) Bone - Bone: dyn_SkirtE04"(ComponentPose=@n179, Alpha=@n178, Translation="0,0,0", Rotation="0,0,-70", Scale="1,1,1") {
    rotationMode=BMM_Additive
    rotationSpace=BCS_BoneSpace
    scaleMode=BMM_Additive
    scaleSpace=BCS_BoneSpace
    translationMode=BMM_Additive
    translationSpace=BCS_BoneSpace
    alphaBoolBlend {blendOption=Linear}
    boneToModify {boneName=dyn_SkirtE04}
    rotation {roll=-70}
   }
   n181 = "Transform (Modify) Bone - Bone: dyn_SkirtF03"(ComponentPose=@n180, Alpha=@n178, Translation="0,0,0", Rotation="0,0,-70", Scale="1,1,1") {
    rotationMode=BMM_Additive
    rotationSpace=BCS_BoneSpace
    scaleMode=BMM_Additive
    scaleSpace=BCS_BoneSpace
    translationMode=BMM_Additive
    translationSpace=BCS_BoneSpace
    alphaBoolBlend {blendOption=Linear}
    boneToModify {boneName=dyn_SkirtF03}
    rotation {roll=-70}
   }
   n182 = "Constraint - dyn_SkirtD03 (PP)"!(ComponentPose=@n181, Alpha=1, ConstraintWeights_0=0.8, ConstraintWeights_1:float=0.2) -> (Pose) {
    alphaBoolBlend {blendOption=Linear}
    boneToModify {boneName=dyn_SkirtD03}
    constraintSetup [{targetBone {boneName=thigh_twist_01_l}}, {targetBone {boneName=dyn_SkirtE03}}]
    constraintWeights [1,1]
   }
   n183 = "Constraint - dyn_SkirtD05 (PP)"(ComponentPose=@n182, Alpha=@n171, ConstraintWeights_0=0.7, ConstraintWeights_1=0.3) {
    alphaBoolBlend {blendOption=Linear}
    boneToModify {boneName=dyn_SkirtD05}
    constraintSetup [{targetBone {boneName=calf_twist_02_l}}, {targetBone {boneName=dyn_SkirtD04}}]
    constraintWeights [1,1]
   }
   n184 = Get thigh_shape_lt_l()
   n185 = "Transform (Modify) Bone - Bone: dyn_SkirtD03"(ComponentPose=@n183, Alpha=@n184, Translation="0,0,0", Rotation="20,0,-15", Scale="1,1,1") {
    rotationMode=BMM_Additive
    rotationSpace=BCS_BoneSpace
    scaleMode=BMM_Additive
    scaleSpace=BCS_BoneSpace
    translationMode=BMM_Additive
    translationSpace=BCS_BoneSpace
    alphaBoolBlend {blendOption=Linear}
    boneToModify {boneName=dyn_SkirtD03}
    rotation {pitch=20, roll=-15}
   }
   n186 = "Transform (Modify) Bone - Bone: dyn_SkirtD05"(ComponentPose=@n185, Alpha=@n184, Translation="0,0,0", Rotation="0,0,-90", Scale="1,1,1") {
    rotationMode=BMM_Additive
    rotationSpace=BCS_BoneSpace
    scaleMode=BMM_Additive
    scaleSpace=BCS_BoneSpace
    translationMode=BMM_Additive
    translationSpace=BCS_BoneSpace
    alphaBoolBlend {blendOption=Linear}
    boneToModify {boneName=dyn_SkirtD05}
    rotation {roll=-90}
   }
   n187 = "Constraint - dyn_SkirtJ03 (RR)"!(ComponentPose=@n186, Alpha=Get thigh_shape_fn_r_0() -> (thigh_shape_fn_r_0:double), ConstraintWeights_0=1, ConstraintWeights_1:float=0) -> (Pose) {
    alphaBoolBlend {blendOption=Linear}
    boneToModify {boneName=dyn_SkirtJ03}
    constraintSetup [{transformType=Rotation, targetBone {boneName=thigh_twist_01_r}}, {transformType=Rotation, targetBone {boneName=pelvis}}]
    constraintWeights [1,1]
   }
   n188 = Get calf_shape_fn_r() -> (calf_shape_fn_r:double)
   n189 = "Constraint - dyn_SkirtJ06 (PP)"(ComponentPose=@n187, Alpha=@n188, ConstraintWeights_0=0.5, ConstraintWeights_1=0.5) {
    alphaBoolBlend {blendOption=Linear}
    boneToModify {boneName=dyn_SkirtJ06}
    constraintSetup [{targetBone {boneName=calf_twist_02_r}}, {targetBone {boneName=dyn_SkirtJ05}}]
    constraintWeights [1,1]
   }
   n190 = "Transform (Modify) Bone - Bone: dyn_SkirtJ03"(ComponentPose=@n189, Alpha=Get thigh_shape_fn_r(), Translation="5,10,-4", Rotation="-65,10,10", Scale="1,1,1") {
    rotationMode=BMM_Additive
    rotationSpace=BCS_BoneSpace
    scaleMode=BMM_Additive
    scaleSpace=BCS_BoneSpace
    translationMode=BMM_Additive
    translationSpace=BCS_BoneSpace
    alphaBoolBlend {blendOption=Linear}
    boneToModify {boneName=dyn_SkirtJ03}
    rotation=<rot -65 10 10>
    translation=<5 10 -4>
   }
   n191 = "Transform (Modify) Bone - Bone: dyn_SkirtJ04"(ComponentPose=@n190, Alpha=Get thigh_shape_fn_r(), Translation="0,0,5", Rotation="60,0,-10", Scale="1,1,1") {
    rotationMode=BMM_Additive
    rotationSpace=BCS_BoneSpace
    scaleMode=BMM_Additive
    scaleSpace=BCS_BoneSpace
    translationMode=BMM_Additive
    translationSpace=BCS_BoneSpace
    alphaBoolBlend {blendOption=Linear}
    boneToModify {boneName=dyn_SkirtJ04}
    rotation {pitch=60, roll=-10}
    translation {z=5}
   }
   n192 = "Transform (Modify) Bone - Bone: dyn_SkirtJ03"(ComponentPose=@n191, Alpha=Get thigh_shape_rt_r() -> (thigh_shape_rt_r:double), Translation="0,12,0", Rotation="0,0,10", Scale="1,1,1") {
    rotationMode=BMM_Additive
    rotationSpace=BCS_BoneSpace
    scaleMode=BMM_Additive
    scaleSpace=BCS_BoneSpace
    translationMode=BMM_Additive
    translationSpace=BCS_BoneSpace
    alphaBoolBlend {blendOption=Linear}
    boneToModify {boneName=dyn_SkirtJ03}
    rotation {roll=10}
    translation {y=12}
   }
   n193 = "Constraint - dyn_SkirtH03 (PP)"!(ComponentPose=@n192, Alpha=1, ConstraintWeights_0=0.8, ConstraintWeights_1:float=0.2) -> (Pose) {
    alphaBoolBlend {blendOption=Linear}
    boneToModify {boneName=dyn_SkirtH03}
    constraintSetup [{targetBone {boneName=thigh_twist_01_r}}, {targetBone {boneName=dyn_SkirtG03}}]
    constraintWeights [1,1]
   }
   n194 = Get thigh_shape_rt_r()
   n195 = "Transform (Modify) Bone - Bone: dyn_SkirtH03"(ComponentPose=@n193, Alpha=@n194, Translation="-5,5,0", Rotation="0,0,-30", Scale="1,1,1") {
    rotationMode=BMM_Additive
    rotationSpace=BCS_BoneSpace
    scaleMode=BMM_Additive
    scaleSpace=BCS_BoneSpace
    translationMode=BMM_Additive
    translationSpace=BCS_BoneSpace
    alphaBoolBlend {blendOption=Linear}
    boneToModify {boneName=dyn_SkirtH03}
    rotation {roll=-30}
    translation=<-5 5>
   }
   n196 = "Constraint - dyn_SkirtH05 (PP)"(ComponentPose=@n195, Alpha=@n188, ConstraintWeights_0=0.7, ConstraintWeights_1=0.3) {
    alphaBoolBlend {blendOption=Linear}
    boneToModify {boneName=dyn_SkirtH05}
    constraintSetup [{targetBone {boneName=calf_twist_02_r}}, {targetBone {boneName=dyn_SkirtH04}}]
    constraintWeights [1,1]
   }
   n197 = "Transform (Modify) Bone - Bone: dyn_SkirtH05"(ComponentPose=@n196, Alpha=@n194, Translation="0,10,0", Rotation="-20,0,-10", Scale="1,1,1") {
    rotationMode=BMM_Additive
    rotationSpace=BCS_BoneSpace
    scaleMode=BMM_Additive
    scaleSpace=BCS_BoneSpace
    translationMode=BMM_Additive
    translationSpace=BCS_BoneSpace
    alphaBoolBlend {blendOption=Linear}
    boneToModify {boneName=dyn_SkirtH05}
    rotation {pitch=-20, roll=-10}
    translation {y=10}
   }
   n198 = "Transform (Modify) Bone - Bone: dyn_SkirtH03"(ComponentPose=@n197, Alpha=1, Translation="0,0,0", Rotation="0,0,10", Scale="1,1,1") {
    rotationMode=BMM_Additive
    rotationSpace=BCS_BoneSpace
    scaleMode=BMM_Additive
    scaleSpace=BCS_BoneSpace
    translationMode=BMM_Additive
    translationSpace=BCS_BoneSpace
    alphaBoolBlend {blendOption=Linear}
    boneToModify {boneName=dyn_SkirtH03}
    rotation {roll=10}
   }
   n199 = "Constraint - dyn_SkirtG03 (PP)"!(ComponentPose=@n198, Alpha=1, ConstraintWeights_0=0.1, ConstraintWeights_1:float=0.9) -> (Pose) {
    alphaBoolBlend {blendOption=Linear}
    boneToModify {boneName=dyn_SkirtG03}
    constraintSetup [{targetBone {boneName=thigh_twist_01_r}}, {targetBone {boneName=pelvis}}]
    constraintWeights [1,1]
   }
   n200 = "Constraint - dyn_SkirtG03 (PP)"(ComponentPose=@n199, Alpha=1, ConstraintWeights_0=0.4, ConstraintWeights_1=0.6) {
    alphaBoolBlend {blendOption=Linear}
    boneToModify {boneName=dyn_SkirtG03}
    constraintSetup [{targetBone {boneName=thigh_twist_01_r}}, {targetBone {boneName=pelvis}}]
    constraintWeights [1,1]
   }
   n201 = Get calf_shape_back_r() -> (calf_shape_back_r:double)
   n202 = "Constraint - dyn_SkirtG05 (PP)"(ComponentPose=@n200, Alpha=@n201, ConstraintWeights_0=0.2, ConstraintWeights_1=0.8) {
    alphaBoolBlend {blendOption=Linear}
    boneToModify {boneName=dyn_SkirtG05}
    constraintSetup [{targetBone {boneName=calf_twist_02_r}}, {targetBone {boneName=dyn_SkirtG03}}]
    constraintWeights [1,1]
   }
   n203 = "Transform (Modify) Bone - Bone: dyn_SkirtF03"(ComponentPose=@n202, Alpha=@n201, Translation="0,0,0", Rotation="0,0,-70", Scale="1,1,1") {
    rotationMode=BMM_Additive
    rotationSpace=BCS_BoneSpace
    scaleMode=BMM_Additive
    scaleSpace=BCS_BoneSpace
    translationMode=BMM_Additive
    translationSpace=BCS_BoneSpace
    alphaBoolBlend {blendOption=Linear}
    boneToModify {boneName=dyn_SkirtF03}
    rotation {roll=-70}
   }
   n204 = Get calf_shape_back_r_0() -> (calf_shape_back_r_0:double)
   n205 = "Transform (Modify) Bone - Bone: dyn_SkirtG03"(ComponentPose=@n203, Alpha=@n204, Translation="0,0,5", Rotation="0,0,-30", Scale="1,1,1") {
    rotationMode=BMM_Additive
    rotationSpace=BCS_BoneSpace
    scaleMode=BMM_Additive
    scaleSpace=BCS_BoneSpace
    translationMode=BMM_Additive
    translationSpace=BCS_BoneSpace
    alphaBoolBlend {blendOption=Linear}
    boneToModify {boneName=dyn_SkirtG03}
    rotation {roll=-30}
    translation {z=5}
   }
   n206 = "Transform (Modify) Bone - Bone: dyn_SkirtG04"(ComponentPose=@n205, Alpha=@n204, Translation="0,0,7.5", Rotation="0,0,-40", Scale="1,1,1") {
    rotationMode=BMM_Additive
    rotationSpace=BCS_BoneSpace
    scaleMode=BMM_Additive
    scaleSpace=BCS_BoneSpace
    translationMode=BMM_Additive
    translationSpace=BCS_BoneSpace
    alphaBoolBlend {blendOption=Linear}
    boneToModify {boneName=dyn_SkirtG04}
    rotation {roll=-40}
    translation {z=7.5}
   }
   n207 = Get pelvis_rot() -> (pelvis_rot:double)
   n208 = "Constraint - dyn_SkirtE04 (PP)"!(ComponentPose=@n206, Alpha=@n207, ConstraintWeights_0=0.6, ConstraintWeights_1:float=0.4) -> (Pose) {
    alphaBoolBlend {blendOption=Linear}
    boneToModify {boneName=dyn_SkirtE04}
    constraintSetup [{targetBone {boneName=thigh_twist_01_l}}, {targetBone {boneName=pelvis}}]
    constraintWeights [1,1]
   }
   n209 = "Constraint - dyn_SkirtG04 (PP)"!(ComponentPose=@n208, Alpha=@n207, ConstraintWeights_0=0.6, ConstraintWeights_1:float=0.4) -> (Pose) {
    alphaBoolBlend {blendOption=Linear}
    boneToModify {boneName=dyn_SkirtG04}
    constraintSetup [{targetBone {boneName=thigh_twist_01_r}}, {targetBone {boneName=pelvis}}]
    constraintWeights [1,1]
   }
   n210 = Get IsFrontEnd_0() -> (IsFrontEnd_0:double)
   n211 = Anim Dynamics - Bone: dyn_SkirtE02(ComponentPose=@n209, Alpha=@n210, ExternalForce="0,0,0") {
    angularSpringConstant=800
    bAngularSpring=true
    bGravityOverrideInSimSpace=true
    bOverrideAngularDamping=true
    bUseGravityOverride=true
    gravityScale=0.3
    simulationSpace=RootRelative
    alphaBoolBlend {blendOption=Linear}
    boundBone {boneName=dyn_SkirtE02}
    componentAppliedLinearAccClamp=<1200 1200 1200>
    componentLinearAccScale=<0.2 0.2 0.25>
    componentLinearVelScale=<0.25 0.25 0.25>
    gravityOverride {z=-980}
    physicsBodyDefinitions [{
      sphereCollisionRadius=0
      boundBone {boneName=dyn_SkirtE02}
      boxExtents=<25 25 25>
      constraintSetup {twistAxis=AxisY, angularLimitsMax=<30 0.1 1>, angularLimitsMin=<1 -0.1 -1>, angularTarget=<0 0 0>}
      localJointOffset=<0.01 -0.1 -4>
     }]
    relativeSpaceBone {boneName=pelvis}
   }
   n212 = Anim Dynamics - Bone: dyn_SkirtF02(ComponentPose=@n211, Alpha=@n210, ExternalForce="0,0,0") {
    angularSpringConstant=800
    bAngularSpring=true
    bGravityOverrideInSimSpace=true
    bOverrideAngularDamping=true
    bUseGravityOverride=true
    gravityScale=0.3
    simulationSpace=RootRelative
    alphaBoolBlend {blendOption=Linear}
    boundBone {boneName=dyn_SkirtF02}
    componentAppliedLinearAccClamp=<1200 1200 1200>
    componentLinearAccScale=<0.2 0.2 0.25>
    componentLinearVelScale=<0.25 0.25 0.25>
    gravityOverride {z=-980}
    physicsBodyDefinitions [{
      sphereCollisionRadius=0
      boundBone {boneName=dyn_SkirtF02}
      boxExtents=<25 25 25>
      constraintSetup {twistAxis=AxisY, angularLimitsMax=<30 0.1 1>, angularLimitsMin=<1 -0.1 -1>, angularTarget=<0 0 0>}
      localJointOffset=<0.01 -0.1 -4>
     }]
    relativeSpaceBone {boneName=pelvis}
   }
   n213 = Anim Dynamics - Bone: dyn_SkirtG02(ComponentPose=@n212, Alpha=@n210, ExternalForce="0,0,0") {
    angularSpringConstant=800
    bAngularSpring=true
    bGravityOverrideInSimSpace=true
    bOverrideAngularDamping=true
    bUseGravityOverride=true
    gravityScale=0.3
    simulationSpace=RootRelative
    alphaBoolBlend {blendOption=Linear}
    boundBone {boneName=dyn_SkirtG02}
    componentAppliedLinearAccClamp=<1200 1200 1200>
    componentLinearAccScale=<0.2 0.2 0.25>
    componentLinearVelScale=<0.25 0.25 0.25>
    gravityOverride {z=-980}
    physicsBodyDefinitions [{
      sphereCollisionRadius=0
      boundBone {boneName=dyn_SkirtG02}
      boxExtents=<25 25 25>
      constraintSetup {twistAxis=AxisY, angularLimitsMax=<30 0.1 1>, angularLimitsMin=<1 -0.1 -1>, angularTarget=<0 0 0>}
      localJointOffset=<0.01 -0.1 -4>
     }]
    relativeSpaceBone {boneName=pelvis}
   }
   n214 = Get IsSliding() -> (IsSliding:double)
   n215 = "Transform (Modify) Bone - Bone: dyn_SkirtE02"(ComponentPose=@n213, Alpha=@n214, Translation="0,5,10", Rotation="0,0,-30", Scale="1,1,1") {
    rotationMode=BMM_Additive
    rotationSpace=BCS_BoneSpace
    scaleMode=BMM_Additive
    scaleSpace=BCS_BoneSpace
    translationMode=BMM_Additive
    translationSpace=BCS_BoneSpace
    alphaBoolBlend {blendOption=Linear}
    boneToModify {boneName=dyn_SkirtE02}
    rotation {roll=-30}
    translation {y=5, z=10}
   }
   n216 = "Transform (Modify) Bone - Bone: dyn_SkirtF02"(ComponentPose=@n215, Alpha=@n214, Translation="0,5,10", Rotation="0,0,-30", Scale="1,1,1") {
    rotationMode=BMM_Additive
    rotationSpace=BCS_BoneSpace
    scaleMode=BMM_Additive
    scaleSpace=BCS_BoneSpace
    translationMode=BMM_Additive
    translationSpace=BCS_BoneSpace
    alphaBoolBlend {blendOption=Linear}
    boneToModify {boneName=dyn_SkirtF02}
    rotation {roll=-30}
    translation {y=5, z=10}
   }
   n217 = "Transform (Modify) Bone - Bone: dyn_SkirtG02"(ComponentPose=@n216, Alpha=@n214, Translation="0,5,10", Rotation="0,0,-30", Scale="1,1,1") {
    rotationMode=BMM_Additive
    rotationSpace=BCS_BoneSpace
    scaleMode=BMM_Additive
    scaleSpace=BCS_BoneSpace
    translationMode=BMM_Additive
    translationSpace=BCS_BoneSpace
    alphaBoolBlend {blendOption=Linear}
    boneToModify {boneName=dyn_SkirtG02}
    rotation {roll=-30}
    translation {y=5, z=10}
   }
   n218 = "Transform (Modify) Bone - Bone: collision_01"(ComponentPose=@n217, Alpha=Get collision_0() -> (collision_0:double), Translation="0,0,-50", Rotation="0,0,0", Scale="1,1,1") {
    rotationMode=BMM_Additive
    rotationSpace=BCS_BoneSpace
    scaleMode=BMM_Additive
    scaleSpace=BCS_BoneSpace
    translationMode=BMM_Additive
    translationSpace=BCS_BoneSpace
    alphaBoolBlend {blendOption=Linear}
    alphaScaleBiasClamp {bInterpResult=true}
    boneToModify {boneName=collision_01}
    translation {z=-50}
   }
   Save cached pose 'RbanCache'(Pose=Component To Local(ComponentPose=@n218)) {cacheName=RbanCache}
   Get offset_weapon()
   Get Shoulder_rot_up_l()
   Get Shoulder_rot_up_r()
  ]}
 EventGraph {nodes [
   Event BlueprintInitializeAnimation()
   > Set Skel_Mesh(Skel_Mesh=GetOwnerSkeletalMeshComponent())
   > n1 = Cast To PlayerPawn_Athena(Object=TryGetPawnOwner())
   > n2 = Set Pawn Owner(Pawn Owner:Fort Player Pawn Athena Object Reference=@n1) -> (Output_Get:Fort Player Pawn Athena Object Reference)
   > Set Shoulder_Mesh(Shoulder_Mesh:Skeletal Mesh Component Object Reference=GetSkeletalMeshForPartType(self=@n2)) -> (Output_Get:Skeletal Mesh Component Object Reference)
   Event BlueprintUpdateAnimation()
   > n3 = Sequence!(execute) -> (then_0, then_1, then_2:Exec, then_3:Exec, then_4:Exec, then_5:Exec, then_6:Exec, then_7:Exec)
   n4 = Get Skel_Mesh()
   Set thigh_shape_fn_l(
    execute=@n3.then_1
    thigh_shape_fn_l:double=GetAlphaFromReferenceTransformDelta(SkeletalMeshComponent=@n4, BoneName=thigh_twist_01_l, BaseName=pelvis, InputRange=(X=-70,Y=-110), TwistAxis="0,0,1")
   ) -> (Output_Get:double)
   > Set thigh_shape_fn_r(thigh_shape_fn_r:double=GetAlphaFromReferenceTransformDelta(SkeletalMeshComponent=@n4, BoneName=thigh_twist_01_r, BaseName=pelvis, InputRange=(X=-40,Y=-110), TwistAxis="0,0,1")) -> (Output_Get:double)
   n5 = Get Skel_Mesh()
   > Set thigh_shape_bk_l(thigh_shape_bk_l:double=GetAlphaFromReferenceTransformDelta(SkeletalMeshComponent=@n5, BoneName=thigh_l, BaseName=pelvis, InputRange=(X=0,Y=-110), TwistAxis="0,0,1")) -> (Output_Get:double)
   > Set thigh_shape_bk_r(thigh_shape_bk_r:double=GetAlphaFromReferenceTransformDelta(SkeletalMeshComponent=@n5, BoneName=thigh_r, BaseName=pelvis, InputRange=(X=0,Y=-110), TwistAxis="0,0,1")) -> (Output_Get:double)
   n6 = Get Skel_Mesh()
   > Set thigh_shape_lt_l(thigh_shape_lt_l:double=GetAlphaFromReferenceTransformDelta(
     SkeletalMeshComponent=@n6
     BoneName=thigh_l
     BaseName=pelvis
     DeltaComponentType=EulerY
     InputRange=(X=90,Y=0)
     OutputRange=(X=1,Y=0)
     TwistAxis="0,0,1"
    )) -> (Output_Get:double)
   > Set thigh_shape_rt_r(thigh_shape_rt_r:double=GetAlphaFromReferenceTransformDelta(SkeletalMeshComponent=@n6, BoneName=thigh_twist_01_r, BaseName=pelvis, DeltaComponentType=EulerY, InputRange=(X=-10,Y=60), TwistAxis="0,0,1")) -> (Output_Get:double)
   n7 = Get Skel_Mesh()
   > Set thigh_shape_rt_l(thigh_shape_rt_l:double=GetAlphaFromReferenceTransformDelta(
     SkeletalMeshComponent=@n7
     BoneName=thigh_twist_01_l
     BaseName=pelvis
     DeltaComponentType=EulerY
     InputRange=(X=-50,Y=0)
     OutputRange=(X=1,Y=0)
     TwistAxis="0,0,1"
    )) -> (Output_Get:double)
   > Set thigh_shape_lt_r(thigh_shape_lt_r:double=GetAlphaFromReferenceTransformDelta(
     SkeletalMeshComponent=@n7
     BoneName=thigh_twist_01_r
     BaseName=pelvis
     DeltaComponentType=EulerY
     InputRange=(X=-50,Y=0)
     OutputRange=(X=1,Y=0)
     TwistAxis="0,0,1"
    )) -> (Output_Get:double)
   n8 = Get Skel_Mesh()
   > Set thigh_shape_fn_l_0(thigh_shape_fn_l_0:double=GetAlphaFromReferenceTransformDelta(SkeletalMeshComponent=@n8, BoneName=thigh_twist_01_l, BaseName=pelvis, InputRange=(X=60,Y=-60), TwistAxis="0,0,1")) -> (Output_Get:double)
   > Set thigh_shape_fn_r_0(thigh_shape_fn_r_0:double=GetAlphaFromReferenceTransformDelta(SkeletalMeshComponent=@n8, BoneName=thigh_twist_01_r, BaseName=pelvis, InputRange=(X=60,Y=-60), TwistAxis="0,0,1")) -> (Output_Get:double)
   n9 = Get Skel_Mesh()
   > Set calf_shape_fn_l(calf_shape_fn_l:double=GetAlphaFromReferenceTransformDelta(SkeletalMeshComponent=@n9, BoneName=calf_l, BaseName=thigh_l, InputRange=(X=0,Y=110), TwistAxis="0,0,1")) -> (Output_Get:double)
   > Set calf_shape_fn_r(calf_shape_fn_r:double=GetAlphaFromReferenceTransformDelta(SkeletalMeshComponent=@n9, BoneName=calf_r, BaseName=thigh_r, InputRange=(X=0,Y=110), TwistAxis="0,0,1")) -> (Output_Get:double)
   n10 = Get thigh_shape_bk_l() -> (thigh_shape_bk_l:double)
   > Set calf_shape_back_l(calf_shape_back_l:double=float + float(A=Get calf_shape_fn_l(), B=float * float(A=@n10, B=-0.7) {is=title K2Node_PromotableOperator})) -> (Output_Get:double)
   n11 = float + float(A=Get calf_shape_fn_r(), B=float * float(A=Get thigh_shape_bk_r() -> (thigh_shape_bk_r:double), B=-0.7))
   > Set calf_shape_back_r(calf_shape_back_r:double=@n11) -> (Output_Get:double)
   > n12 = Branch(Condition=Get bIsCrouching())
   > Set calf_shape_back_r_0(calf_shape_back_r_0:double=@n11) -> (Output_Get:double)
   Set calf_shape_back_r_0(execute=@n12.else, calf_shape_back_r_0=float * float(A=@n11, B=0.5))
   Set hand_shape_up_l(
    execute=@n3.then_2
    hand_shape_up_l:double=GetAlphaFromReferenceTransformDelta(
     SkeletalMeshComponent=Get Skel_Mesh()
     BoneName=hand_l
     BaseName=lowerarm_twist_01_l
     DeltaComponentType=EulerY
     InputRange=(X=0,Y=-60)
     TwistAxis="0,0,1"
    )
   ) -> (Output_Get:double)
   > Set hand_shape_up_r(hand_shape_up_r:double=GetAlphaFromReferenceTransformDelta(
     SkeletalMeshComponent=Get Skel_Mesh()
     BoneName=hand_r
     BaseName=lowerarm_twist_01_r
     DeltaComponentType=EulerY
     InputRange=(X=0,Y=-60)
     TwistAxis="0,0,1"
    )) -> (Output_Get:double)
   > Set hand_shape_dn_l(hand_shape_dn_l:double=GetAlphaFromReferenceTransformDelta(
     SkeletalMeshComponent=Get Skel_Mesh()
     BoneName=hand_l
     BaseName=lowerarm_twist_01_l
     DeltaComponentType=EulerY
     InputRange=(X=0,Y=60)
     TwistAxis="0,0,1"
    )) -> (Output_Get:double)
   > Set hand_shape_dn_r(hand_shape_dn_r:double=GetAlphaFromReferenceTransformDelta(
     SkeletalMeshComponent=Get Skel_Mesh()
     BoneName=hand_r
     BaseName=lowerarm_twist_01_r
     DeltaComponentType=EulerY
     InputRange=(X=0,Y=60)
     TwistAxis="0,0,1"
    )) -> (Output_Get:double)
   > Set hand_shape_lt_l(hand_shape_lt_l:double=GetAlphaFromReferenceTransformDelta(
     SkeletalMeshComponent=Get Skel_Mesh()
     BoneName=hand_l
     BaseName=lowerarm_twist_01_l
     DeltaComponentType=EulerX
     InputRange=(X=0,Y=60)
     TwistAxis="0,0,1"
    )) -> (Output_Get:double)
   > Set hand_shape_lt_r(hand_shape_lt_r:double=GetAlphaFromReferenceTransformDelta(
     SkeletalMeshComponent=Get Skel_Mesh()
     BoneName=hand_r
     BaseName=lowerarm_twist_01_r
     DeltaComponentType=EulerX
     InputRange=(X=0,Y=60)
     TwistAxis="0,0,1"
    )) -> (Output_Get:double)
   > Set hand_shape_rt_l(hand_shape_rt_l:double=GetAlphaFromReferenceTransformDelta(
     SkeletalMeshComponent=Get Skel_Mesh()
     BoneName=hand_l
     BaseName=lowerarm_twist_01_l
     DeltaComponentType=EulerX
     InputRange=(X=0,Y=-60)
     TwistAxis="0,0,1"
    )) -> (Output_Get:double)
   > Set hand_shape_rt_r(hand_shape_rt_r:double=GetAlphaFromReferenceTransformDelta(
     SkeletalMeshComponent=Get Skel_Mesh()
     BoneName=hand_r
     BaseName=lowerarm_twist_01_r
     DeltaComponentType=EulerX
     InputRange=(X=0,Y=-60)
     TwistAxis="0,0,1"
    )) -> (Output_Get:double)
   n13 = Branch(execute=@n3.then_3, Condition=OR Boolean!(A=Get bIsSliding(), B=Get bIsSlopeSliding() -> (bIsSlopeSliding:bool), C:bool) -> (ReturnValue))
   > Set IsSliding(IsSliding:double=1) -> (Output_Get:double)
   Set IsSliding(execute=@n13.else, IsSliding=0)
   n14 = AND Boolean(A=NOT Boolean(A=Get bIsPlayingEmote()), B=Get bIsInFrontEnd())
   n15 = Branch(execute=@n3.then_4, Condition=@n14)
   > Set IsFrontEnd(IsFrontEnd:double=SelectFloat(A=1, B=0.7, bPickA=@n14)) -> (Output_Get:double)
   Set IsFrontEnd(execute=@n15.else, IsFrontEnd=0)
   n16 = Branch(execute=@n3.then_7, Condition=Get bIsSkyDiving())
   > Set IsDiving(IsDiving:double=0) -> (Output_Get:double)
   Set IsDiving(execute=@n16.else, IsDiving=1)
   n17 = Get Shoulder_Mesh() -> (Shoulder_Mesh:Skeletal Mesh Component Object Reference)
   Set Shoulder_rot_f_r(
    execute=@n3.then_0
    Shoulder_rot_f_r:double=GetAlphaFromReferenceTransformDelta(SkeletalMeshComponent=@n17, BoneName=upperarm_r, BaseName=clavicle_r, InputRange=(X=0,Y=110), TwistAxis="0,0,1")
   ) -> (Output_Get:double)
   > Set Shoulder_rot_f_l(Shoulder_rot_f_l:double=GetAlphaFromReferenceTransformDelta(SkeletalMeshComponent=@n17, BoneName=upperarm_l, BaseName=clavicle_l, InputRange=(X=0,Y=110), TwistAxis="0,0,1")) -> (Output_Get:double)
   n18 = Get Shoulder_Mesh()
   > Set Shoulder_rot_up_r(Shoulder_rot_up_r:double=GetAlphaFromReferenceTransformDelta(SkeletalMeshComponent=@n18, BoneName=upperarm_r, BaseName=clavicle_r, DeltaComponentType=EulerY, InputRange=(X=0,Y=110), TwistAxis="0,0,1")) -> (Output_Get:double)
   > Set Shoulder_rot_up_l(Shoulder_rot_up_l:double=GetAlphaFromReferenceTransformDelta(SkeletalMeshComponent=@n18, BoneName=upperarm_l, BaseName=clavicle_l, DeltaComponentType=EulerY, InputRange=(X=0,Y=110), TwistAxis="0,0,1")) -> (Output_Get:double)
   n19 = Get Shoulder_Mesh()
   > Set Shoulder_rot_dn_r(Shoulder_rot_dn_r:double=GetAlphaFromReferenceTransformDelta(SkeletalMeshComponent=@n19, BoneName=upperarm_r, BaseName=clavicle_r, DeltaComponentType=EulerY, InputRange=(X=0,Y=-90), TwistAxis="0,0,1")) -> (Output_Get:double)
   > Set Shoulder_rot_dn_l(Shoulder_rot_dn_l:double=GetAlphaFromReferenceTransformDelta(SkeletalMeshComponent=@n19, BoneName=upperarm_l, BaseName=clavicle_l, DeltaComponentType=EulerY, InputRange=(X=0,Y=-90), TwistAxis="0,0,1")) -> (Output_Get:double)
   > n20 = Sequence!(execute) -> (then_0, then_1, then_2, then_3, then_4)
   > Set lowerarm_l_front(lowerarm_l_front:double=GetAlphaFromReferenceTransformDelta(SkeletalMeshComponent=Get Shoulder_Mesh(), BoneName=upperarm_l, BaseName=lowerarm_l, InputRange=(X=0,Y=-90), TwistAxis="0,0,1")) -> (Output_Get:double)
   > Set calf_l_front(calf_l_front:double=GetAlphaFromReferenceTransformDelta(SkeletalMeshComponent=Get Skel_Mesh(), BoneName=thigh_l, BaseName=calf_l, InputRange=(X=0,Y=-90), TwistAxis="0,0,1")) -> (Output_Get:double)
   > n21 = Branch(Condition=OR Boolean(A=Get bIsSprinting(), B=Get bIsCrouchSprinting(), C=Get bIsMoving2D()))
   > Set character direction(character direction:double=1) -> (Output_Get:double)
   Set character direction(execute=@n21.else, character direction=0)
   Set lowerarm_r_front(
    execute=@n20.then_1
    lowerarm_r_front:double=GetAlphaFromReferenceTransformDelta(SkeletalMeshComponent=Get Shoulder_Mesh(), BoneName=upperarm_r, BaseName=lowerarm_r, InputRange=(X=0,Y=-90), TwistAxis="0,0,1")
   ) -> (Output_Get:double)
   > Set calf_r_front(calf_r_front:double=GetAlphaFromReferenceTransformDelta(SkeletalMeshComponent=Get Skel_Mesh(), BoneName=thigh_r, BaseName=calf_r, InputRange=(X=0,Y=-90), TwistAxis="0,0,1")) -> (Output_Get:double)
   > Set pelvis_rot(pelvis_rot:double=GetAlphaFromReferenceTransformDelta(SkeletalMeshComponent=Get Skel_Mesh(), BoneName=pelvis, BaseName=root, InputRange=(X=0,Y=140), TwistAxis="0,0,1")) -> (Output_Get:double)
   > n22 = Branch(Condition=OR Boolean!(
     A=false
     B=false
     C
     D:bool=Get bIsSwimSprinting() -> (bIsSwimSprinting:bool)
     E:bool=Get bIsParachuteOpened()
     F:bool=Get bIsWaterJump() -> (bIsWaterJump:bool)
     G:bool
    ) -> (ReturnValue))
   > Set collision(collision:bool=true) -> (Output_Get:bool)
   Set collision(execute=@n22.else, collision=false)
   n23 = Get Skel_Mesh()
   Set neck_rot_up(
    execute=@n20.then_2
    neck_rot_up:double=GetAlphaFromReferenceTransformDelta(SkeletalMeshComponent=@n23, BoneName=C_jaw, BaseName=spine_05, DeltaComponentType=EulerX, InputRange=(X=30,Y=-30), TwistAxis="0,0,1")
   ) -> (Output_Get:double)
   > Set neck_rot_dn(neck_rot_dn:double=GetAlphaFromReferenceTransformDelta(SkeletalMeshComponent=@n23, BoneName=C_jaw, BaseName=spine_05, DeltaComponentType=EulerX, InputRange=(X=0,Y=25), TwistAxis="0,0,1")) -> (Output_Get:double)
   > n24 = Branch(Condition=OR Boolean!(A=Get bIsPlayingEmote(), B=Get bIsParachuteOpened(), C) -> (ReturnValue))
   > Set collision_0(collision_0:double=1) -> (Output_Get:double)
   Set collision_0(execute=@n24.else, collision_0=0)
   n25 = Gravity Override State Function DEPRECATED(
    execute=@n20.then_3
    CharacterPartAnimInstance=Self-Reference()
    GameTimeInSeconds=GetGameTimeInSeconds()
    GravityOverrideParams=Get Gravity Override Params()
   )
   > Set GravityOverride(GravityOverride=@n25)
   > n26 = Branch(Condition=Get bIsCrouching())
   > Set crouch(crouch:double=1) -> (Output_Get:double)
   Set crouch(execute=@n26.else, crouch=0)
   n27 = Cast To FortPlayerAnimInstance(execute=@n20.then_4, Object=GetAnimInstance(self=GetOwnerSkeletalMeshComponent()))
   > Set offset_weapon(offset_weapon:double=0) -> (Output_Get:double)
   > Set offset_weapon_0(offset_weapon_0:double=0) -> (Output_Get:double)
   > n28 = Switch on EFortWeaponCoreAnimation!(execute, Selection:EFortWeaponCoreAnimation Enum=Get WeaponCoreAnim(self=@n27)) -> (
    Melee:Exec
    Pistol:Exec
    Shotgun:Exec
    PaperBlueprint:Exec
    Rifle:Exec
    MeleeOneHand:Exec
    MachinePistol:Exec
    RocketLauncher:Exec
    GrenadeLauncher:Exec
    GoingCommando:Exec
    AssaultRifle:Exec
    TacticalShotgun:Exec
    SniperRifle:Exec
    TrapPlacement:Exec
    ShoulderLauncher:Exec
    AbilityDecoTool:Exec
    Crossbow:Exec
    C4:Exec
    RemoteControl:Exec
    DualWield:Exec
    AR_BullPup:Exec
    AR_ForwardGrip:Exec
    MedPackPaddles:Exec
    SMG_P90:Exec
    AR_DrumGun:Exec
    Consumable_Small:Exec
    Consumable_Large:Exec
    Balloon:Exec
    MountedTurret:Exec
    CreativeTool:Exec
    ExplosiveBow:Exec
    AshtonIndigo:Exec
    AshtonChicago:Exec
    MeleeDualWield:Exec
    Unarmed:Exec
   )
   n29 = SelectFloat(A=0.35, B=1, bPickA=Get bIsPlayingEmote())
   Set offset_weapon(execute=[@n32.then_1, @n31.else, @n28.AssaultRifle, @n28.Shotgun], offset_weapon=@n29)
   n30 = Sequence!(execute=@n28.Pistol) -> (then_0, then_1)
   > Set IsTargeting(IsTargeting:double="To Float (Boolean)"(InBool=Get bIsTargeting())) -> (Output_Get:double)
   n31 = Branch(execute=@n30.then_1, Condition=Get bIsPlayingEmote())
   n32 = Sequence(execute=@n28.TacticalShotgun)
   > Set offset_weapon_0(offset_weapon_0=@n29)
   Set IsFrontEnd_0(execute=@n3.then_5, IsFrontEnd_0:double=SelectFloat(A=0.15, B=0.5, bPickA=@n14)) -> (Output_Get:double)
   Set IsFrontEnd_1(execute=@n3.then_6, IsFrontEnd_1:double=SelectFloat(A=0, B=1, bPickA=@n14)) -> (Output_Get:double)
   n33 = Get GravityOverride()
   n34 = Get bIsSlopeSliding()
   n35 = OR Boolean!(A=Get bIsSkyDiving(), B=Get bIsSkydiveDiveMode()) -> (ReturnValue)
   > Set GravityOverride01(GravityOverride01:Vector=SelectVector(A="0,0,-50", B=SelectVector(A="0,0,-100", B=@n33, bPickA=@n34), bPickA=@n35)) -> (Output_Get:Vector)
   > Set Component_Scale(Component_Scale:Vector=SelectVector(A="0.01,0.01,0.01", B="0.15,0.35,0.35", bPickA=@n35)) -> (Output_Get:Vector)
   > Set Is Diving or Basball Sliding(Is Diving or Basball Sliding:bool=OR Boolean(A=Get bIsSkyDiving(), B=Get bIsBaseballSliding())) -> (Output_Get:bool)
   Get bIsSkydiveDiveMode()
   Get bIsSkydiveFloating()
   Set offset_weapon(offset_weapon=0)
  ]}
]
variables [
 calf_l_front {type=number}
 calf_r_front {type=number}
 calf_shape_back_l {type=number}
 calf_shape_back_r {type=number}
 calf_shape_back_r_0 {type=number}
 calf_shape_fn_l {type=number}
 calf_shape_fn_r {type=number}
 character direction {type=number}
 collision {type=boolean}
 collision_0 {type=number}
 Component_Scale {type=Vector}
 crouch {type=number}
 Gravity Override Params {
  type=GravityOverrideParamsStruct
  default {
   onGround_Standing {windAmplitude=<-10 -150 70>, windFrequency=<-7 -1 -1>, windOffset=<-10 -100 70>}
   skydive_FloatingDown {gravityOverride=<0 -700 400>, windAmplitude=<200 -500 200>, windOffset=<0 -600 200>}
   skydive_GoingUp {gravityOverride=<0 0 -980>, windAmplitude=<0 0 0>, windFrequency=<0 0 0>}
  }
 }
 GravityOverride {type=Vector}
 GravityOverride01 {type=Vector}
 hand_shape_dn_l {type=number}
 hand_shape_dn_r {type=number}
 hand_shape_lt_l {type=number}
 hand_shape_lt_r {type=number}
 hand_shape_rt_l {type=number}
 hand_shape_rt_r {type=number}
 hand_shape_up_l {type=number}
 hand_shape_up_r {type=number}
 Is Diving or Basball Sliding {type=boolean}
 Is Sprinting {type=boolean}
 IsDiving {type=number}
 IsFrontEnd {type=number}
 IsFrontEnd_0 {type=number}
 IsFrontEnd_1 {type=number}
 IsSliding {type=number}
 IsTargeting {type=number}
 lowerarm_l_back {type=number}
 lowerarm_l_front {type=number}
 lowerarm_r_back {type=number}
 lowerarm_r_front {type=number}
 neck_rot_dn {type=number}
 neck_rot_up {type=number}
 offset_weapon {type=number}
 offset_weapon_0 {type=number}
 Pawn Owner {type=FortPlayerPawnAthena}
 pelvis_rot {type=number}
 Shoulder_Mesh {type=SkeletalMeshComponent}
 Shoulder_rot_dn_l {type=number}
 Shoulder_rot_dn_r {type=number}
 Shoulder_rot_f_l {type=number}
 Shoulder_rot_f_r {type=number}
 Shoulder_rot_up_l {type=number}
 Shoulder_rot_up_r {type=number}
 Skel_Mesh {type=SkeletalMeshComponent}
 sleeve_weapon {type=number}
 Switch {type=boolean}
 thigh_shape_bk_l {type=number}
 thigh_shape_bk_r {type=number}
 thigh_shape_fn_l {type=number}
 thigh_shape_fn_l_0 {type=number}
 thigh_shape_fn_r {type=number}
 thigh_shape_fn_r_0 {type=number}
 thigh_shape_lt_l {type=number}
 thigh_shape_lt_r {type=number}
 thigh_shape_rt_l {type=number}
 thigh_shape_rt_r {type=number}
]
