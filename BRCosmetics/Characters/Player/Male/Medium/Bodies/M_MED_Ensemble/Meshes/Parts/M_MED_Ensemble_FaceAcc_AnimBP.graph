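# Animation Blueprint for the male Ensemble face accessory, adding secondary motion to its left and right hair bones.
# It copies the parent mesh pose, then runs chained spring-driven AnimDynamics nodes, simulated root-relative with tight angular limits.
# A final additive bone rotation nudges dyn_hairC_r_01, weighted by IsTargeting.
# The update event sets IsTargeting from FortPlayerAnimInstance only for Pistol weapon animation; other types leave it unchanged.
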AnimBlueprint /BRCosmetics/Characters/Player/Male/Medium/Bodies/M_MED_Ensemble/Meshes/Parts/M_MED_Ensemble_FaceAcc_AnimBP
target_skeleton=&/M_MED_Ensemble_FaceAcc_Skeleton
graphs [
 AnimGraph {nodes [
   n1 = Anim Dynamics - Bone: dyn_hairA_l_01(ComponentPose=Local To Component(LocalPose=Copy Pose From Mesh() {bUseAttachedParent=true}), Alpha=0.7, ExternalForce="0,0,0") {
    angularDampingOverride=0.999
    angularSpringConstant=800
    bAngularSpring=true
    bGravityOverrideInSimSpace=true
    bOverrideAngularDamping=true
    gravityScale=0.2
    simulationSpace=RootRelative
    alphaBoolBlend {blendOption=Linear}
    boundBone {boneName=dyn_hairA_l_01}
    componentAppliedLinearAccClamp=<10000 10000 10000>
    componentLinearAccScale=<0.1 0.1 0.1>
    componentLinearVelScale=<0.1 0.1 0.1>
    physicsBodyDefinitions [{
      sphereCollisionRadius=0
      boundBone {boneName=dyn_hairA_l_01}
      boxExtents=<5 8 8>
      constraintSetup {angularTargetAxis=AxisY, angularLimitsMax=<1 10 2>, angularLimitsMin=<-1 -10 -2>, angularTarget=<0 0 0>}
      localJointOffset=<-5 0.1 0.5>
     }]
    relativeSpaceBone {boneName=pelvis}
   }
   n2 = Anim Dynamics - Bone: dyn_hairB_l_01(ComponentPose=@n1, Alpha=0.7, ExternalForce="0,0,0") {
    angularDampingOverride=0.999
    angularSpringConstant=800
    bAngularSpring=true
    bGravityOverrideInSimSpace=true
    bOverrideAngularDamping=true
    gravityScale=0.2
    simulationSpace=RootRelative
    alphaBoolBlend {blendOption=Linear}
    boundBone {boneName=dyn_hairB_l_01}
    componentAppliedLinearAccClamp=<10000 10000 10000>
    componentLinearAccScale=<0.1 0.1 0.1>
    componentLinearVelScale=<0.1 0.1 0.1>
    physicsBodyDefinitions [{
      sphereCollisionRadius=0
      boundBone {boneName=dyn_hairB_l_01}
      boxExtents=<10 15 15>
      constraintSetup {angularTargetAxis=AxisY, angularLimitsMax=<1 0 2>, angularLimitsMin=<-1 -10 -2>, angularTarget=<0 0 0>}
      localJointOffset=<-5 0.1 0.5>
     }]
    relativeSpaceBone {boneName=pelvis}
   }
   n3 = Anim Dynamics - Bone: dyn_hairB_l_02(ComponentPose=@n2, Alpha=0.7, ExternalForce="0,0,0") {
    angularDampingOverride=0.999
    angularSpringConstant=800
    bAngularSpring=true
    bGravityOverrideInSimSpace=true
    bOverrideAngularDamping=true
    gravityScale=0.2
    simulationSpace=RootRelative
    alphaBoolBlend {blendOption=Linear}
    boundBone {boneName=dyn_hairB_l_02}
    componentAppliedLinearAccClamp=<10000 10000 10000>
    componentLinearAccScale=<0.1 0.1 0.1>
    componentLinearVelScale=<0.1 0.1 0.1>
    physicsBodyDefinitions [{
      sphereCollisionRadius=0
      boundBone {boneName=dyn_hairB_l_02}
      boxExtents=<10 15 15>
      constraintSetup {angularTargetAxis=AxisY, angularLimitsMax=<1 0 2>, angularLimitsMin=<-1 -20 -2>, angularTarget=<0 0 0>}
      localJointOffset=<-5 0.1 0.5>
     }]
    relativeSpaceBone {boneName=pelvis}
   }
   n4 = Anim Dynamics - Bone: dyn_hairC_l_01(ComponentPose=@n3, Alpha=0.7, ExternalForce="0,0,0") {
    angularDampingOverride=0.999
    angularSpringConstant=800
    bAngularSpring=true
    bGravityOverrideInSimSpace=true
    bOverrideAngularDamping=true
    gravityScale=0.2
    simulationSpace=RootRelative
    alphaBoolBlend {blendOption=Linear}
    boundBone {boneName=dyn_hairC_l_01}
    componentAppliedLinearAccClamp=<10000 10000 10000>
    componentLinearAccScale=<0.1 0.1 0.1>
    componentLinearVelScale=<0.1 0.1 0.1>
    physicsBodyDefinitions [{
      sphereCollisionRadius=0
      boundBone {boneName=dyn_hairC_l_01}
      boxExtents=<10 15 15>
      constraintSetup {angularTargetAxis=AxisY, angularLimitsMax=<2 5 5>, angularLimitsMin=<-2 -2 -5>, angularTarget=<0 0 0>}
      localJointOffset=<-5 0.1 0.5>
     }]
    relativeSpaceBone {boneName=pelvis}
   }
   n5 = Anim Dynamics - Bone: dyn_hairC_l_a01(ComponentPose=@n4, Alpha=0.7, ExternalForce="0,0,0") {
    angularDampingOverride=0.999
    angularSpringConstant=800
    bAngularSpring=true
    bGravityOverrideInSimSpace=true
    bOverrideAngularDamping=true
    gravityScale=0.2
    simulationSpace=RootRelative
    alphaBoolBlend {blendOption=Linear}
    boundBone {boneName=dyn_hairC_l_a01}
    componentAppliedLinearAccClamp=<10000 10000 10000>
    componentLinearAccScale=<0.1 0.1 0.1>
    componentLinearVelScale=<0.1 0.1 0.1>
    physicsBodyDefinitions [{
      sphereCollisionRadius=0
      boundBone {boneName=dyn_hairC_l_a01}
      boxExtents=<10 15 15>
      constraintSetup {angularTargetAxis=AxisY, angularLimitsMax=<2 5 5>, angularLimitsMin=<-2 -2 -5>, angularTarget=<0 0 0>}
      localJointOffset=<-5 0.1 0.5>
     }]
    relativeSpaceBone {boneName=pelvis}
   }
   n6 = Anim Dynamics - Bone: dyn_hairC_l_a02(ComponentPose=@n5, Alpha=0.7, ExternalForce="0,0,0") {
    angularDampingOverride=0.999
    angularSpringConstant=800
    bAngularSpring=true
    bGravityOverrideInSimSpace=true
    bOverrideAngularDamping=true
    gravityScale=0.2
    simulationSpace=RootRelative
    alphaBoolBlend {blendOption=Linear}
    boundBone {boneName=dyn_hairC_l_a02}
    componentAppliedLinearAccClamp=<10000 10000 10000>
    componentLinearAccScale=<0.1 0.1 0.1>
    componentLinearVelScale=<0.1 0.1 0.1>
    physicsBodyDefinitions [{
      sphereCollisionRadius=0
      boundBone {boneName=dyn_hairC_l_a02}
      boxExtents=<10 15 15>
      constraintSetup {angularTargetAxis=AxisY, angularLimitsMax=<2 5 5>, angularLimitsMin=<-2 -2 -5>, angularTarget=<0 0 0>}
      localJointOffset=<-5 0.1 0.5>
     }]
    relativeSpaceBone {boneName=pelvis}
   }
   n7 = Anim Dynamics - Bone: dyn_hairC_l_b01(ComponentPose=@n6, Alpha=0.7, ExternalForce="0,0,0") {
    angularDampingOverride=0.999
    angularSpringConstant=800
    bAngularSpring=true
    bGravityOverrideInSimSpace=true
    bOverrideAngularDamping=true
    gravityScale=0.2
    simulationSpace=RootRelative
    alphaBoolBlend {blendOption=Linear}
    boundBone {boneName=dyn_hairC_l_b01}
    componentAppliedLinearAccClamp=<10000 10000 10000>
    componentLinearAccScale=<0.1 0.1 0.1>
    componentLinearVelScale=<0.1 0.1 0.1>
    physicsBodyDefinitions [{
      sphereCollisionRadius=0
      boundBone {boneName=dyn_hairC_l_b01}
      boxExtents=<10 15 15>
      constraintSetup {angularTargetAxis=AxisY, angularLimitsMax=<2 5 5>, angularLimitsMin=<-2 -2 -5>, angularTarget=<0 0 0>}
      localJointOffset=<-5 0.1 0.5>
     }]
    relativeSpaceBone {boneName=pelvis}
   }
   n8 = Anim Dynamics - Bone: dyn_hairC_l_b02(ComponentPose=@n7, Alpha=0.7, ExternalForce="0,0,0") {
    angularDampingOverride=0.999
    angularSpringConstant=800
    bAngularSpring=true
    bGravityOverrideInSimSpace=true
    bOverrideAngularDamping=true
    gravityScale=0.2
    simulationSpace=RootRelative
    alphaBoolBlend {blendOption=Linear}
    boundBone {boneName=dyn_hairC_l_b02}
    componentAppliedLinearAccClamp=<10000 10000 10000>
    componentLinearAccScale=<0.1 0.1 0.1>
    componentLinearVelScale=<0.1 0.1 0.1>
    physicsBodyDefinitions [{
      sphereCollisionRadius=0
      boundBone {boneName=dyn_hairC_l_b02}
      boxExtents=<10 15 15>
      constraintSetup {angularTargetAxis=AxisY, angularLimitsMax=<2 5 5>, angularLimitsMin=<-2 -2 -5>, angularTarget=<0 0 0>}
      localJointOffset=<-5 0.1 0.5>
     }]
    relativeSpaceBone {boneName=pelvis}
   }
   n9 = Anim Dynamics - Bone: dyn_hairC_l_c01(ComponentPose=@n8, Alpha=0.7, ExternalForce="0,0,0") {
    angularDampingOverride=0.999
    angularSpringConstant=800
    bAngularSpring=true
    bGravityOverrideInSimSpace=true
    bOverrideAngularDamping=true
    gravityScale=0.2
    simulationSpace=RootRelative
    alphaBoolBlend {blendOption=Linear}
    boundBone {boneName=dyn_hairC_l_c01}
    componentAppliedLinearAccClamp=<10000 10000 10000>
    componentLinearAccScale=<0.1 0.1 0.1>
    componentLinearVelScale=<0.1 0.1 0.1>
    physicsBodyDefinitions [{
      sphereCollisionRadius=0
      boundBone {boneName=dyn_hairC_l_c01}
      boxExtents=<10 15 15>
      constraintSetup {angularTargetAxis=AxisY, angularLimitsMax=<2 5 5>, angularLimitsMin=<-2 -2 -5>, angularTarget=<0 0 0>}
      localJointOffset=<-5 0.1 0.5>
     }]
    relativeSpaceBone {boneName=pelvis}
   }
   n10 = Anim Dynamics - Bone: dyn_hairC_l_c02(ComponentPose=@n9, Alpha=0.7, ExternalForce="0,0,0") {
    angularDampingOverride=0.999
    angularSpringConstant=800
    bAngularSpring=true
    bGravityOverrideInSimSpace=true
    bOverrideAngularDamping=true
    gravityScale=0.2
    simulationSpace=RootRelative
    alphaBoolBlend {blendOption=Linear}
    boundBone {boneName=dyn_hairC_l_c02}
    componentAppliedLinearAccClamp=<10000 10000 10000>
    componentLinearAccScale=<0.1 0.1 0.1>
    componentLinearVelScale=<0.1 0.1 0.1>
    physicsBodyDefinitions [{
      sphereCollisionRadius=0
      boundBone {boneName=dyn_hairC_l_c02}
      boxExtents=<10 15 15>
      constraintSetup {angularTargetAxis=AxisY, angularLimitsMax=<2 5 5>, angularLimitsMin=<-2 -2 -5>, angularTarget=<0 0 0>}
      localJointOffset=<-5 0.1 0.5>
     }]
    relativeSpaceBone {boneName=pelvis}
   }
   n11 = Anim Dynamics - Bone: dyn_hairA_r_01(ComponentPose=@n10, Alpha=0.7, ExternalForce="0,0,0") {
    angularDampingOverride=0.999
    angularSpringConstant=800
    bAngularSpring=true
    bGravityOverrideInSimSpace=true
    bOverrideAngularDamping=true
    gravityScale=0.2
    simulationSpace=RootRelative
    alphaBoolBlend {blendOption=Linear}
    boundBone {boneName=dyn_hairA_r_01}
    componentAppliedLinearAccClamp=<10000 10000 10000>
    componentLinearAccScale=<0.1 0.1 0.1>
    componentLinearVelScale=<0.1 0.1 0.1>
    physicsBodyDefinitions [{
      sphereCollisionRadius=0
      boundBone {boneName=dyn_hairA_r_01}
      boxExtents=<5 8 8>
      constraintSetup {angularTargetAxis=AxisY, angularLimitsMax=<1 10 2>, angularLimitsMin=<-1 -10 -2>, angularTarget=<0 0 0>}
      localJointOffset=<-5 0.1 0.5>
     }]
    relativeSpaceBone {boneName=pelvis}
   }
   n12 = Anim Dynamics - Bone: dyn_hairB_r_01(ComponentPose=@n11, Alpha=0.7, ExternalForce="0,0,0") {
    angularDampingOverride=0.999
    angularSpringConstant=800
    bAngularSpring=true
    bGravityOverrideInSimSpace=true
    bOverrideAngularDamping=true
    gravityScale=0.2
    simulationSpace=RootRelative
    alphaBoolBlend {blendOption=Linear}
    boundBone {boneName=dyn_hairB_r_01}
    componentAppliedLinearAccClamp=<10000 10000 10000>
    componentLinearAccScale=<0.1 0.1 0.1>
    componentLinearVelScale=<0.1 0.1 0.1>
    physicsBodyDefinitions [{
      sphereCollisionRadius=0
      boundBone {boneName=dyn_hairB_r_01}
      boxExtents=<10 15 15>
      constraintSetup {angularTargetAxis=AxisY, angularLimitsMax=<1 0 2>, angularLimitsMin=<-1 -10 -2>, angularTarget=<0 0 0>}
      localJointOffset=<-5 0.1 0.5>
     }]
    relativeSpaceBone {boneName=pelvis}
   }
   n13 = Anim Dynamics - Bone: dyn_hairB_r_02(ComponentPose=@n12, Alpha=0.7, ExternalForce="0,0,0") {
    angularDampingOverride=0.999
    angularSpringConstant=800
    bAngularSpring=true
    bGravityOverrideInSimSpace=true
    bOverrideAngularDamping=true
    gravityScale=0.2
    simulationSpace=RootRelative
    alphaBoolBlend {blendOption=Linear}
    boundBone {boneName=dyn_hairB_r_02}
    componentAppliedLinearAccClamp=<10000 10000 10000>
    componentLinearAccScale=<0.1 0.1 0.1>
    componentLinearVelScale=<0.1 0.1 0.1>
    physicsBodyDefinitions [{
      sphereCollisionRadius=0
      boundBone {boneName=dyn_hairB_r_02}
      boxExtents=<10 15 15>
      constraintSetup {angularTargetAxis=AxisY, angularLimitsMax=<1 0 2>, angularLimitsMin=<-1 -20 -2>, angularTarget=<0 0 0>}
      localJointOffset=<-5 0.1 0.5>
     }]
    relativeSpaceBone {boneName=pelvis}
   }
   n14 = Anim Dynamics - Bone: dyn_hairC_r_01(ComponentPose=@n13, Alpha=0.7, ExternalForce="0,0,0") {
    angularDampingOverride=0.999
    angularSpringConstant=800
    bAngularSpring=true
    bGravityOverrideInSimSpace=true
    bOverrideAngularDamping=true
    gravityScale=0.2
    simulationSpace=RootRelative
    alphaBoolBlend {blendOption=Linear}
    boundBone {boneName=dyn_hairC_r_01}
    componentAppliedLinearAccClamp=<10000 10000 10000>
    componentLinearAccScale=<0.1 0.1 0.1>
    componentLinearVelScale=<0.1 0.1 0.1>
    physicsBodyDefinitions [{
      sphereCollisionRadius=0
      boundBone {boneName=dyn_hairC_r_01}
      boxExtents=<10 15 15>
      constraintSetup {angularTargetAxis=AxisY, angularLimitsMax=<2 5 5>, angularLimitsMin=<-2 -2 -5>, angularTarget=<0 0 0>}
      localJointOffset=<-5 0.1 0.5>
     }]
    relativeSpaceBone {boneName=pelvis}
   }
   n15 = Anim Dynamics - Bone: dyn_hairC_r_a01(ComponentPose=@n14, Alpha=0.7, ExternalForce="0,0,0") {
    angularDampingOverride=0.999
    angularSpringConstant=800
    bAngularSpring=true
    bGravityOverrideInSimSpace=true
    bOverrideAngularDamping=true
    gravityScale=0.2
    simulationSpace=RootRelative
    alphaBoolBlend {blendOption=Linear}
    boundBone {boneName=dyn_hairC_r_a01}
    componentAppliedLinearAccClamp=<10000 10000 10000>
    componentLinearAccScale=<0.1 0.1 0.1>
    componentLinearVelScale=<0.1 0.1 0.1>
    physicsBodyDefinitions [{
      sphereCollisionRadius=0
      boundBone {boneName=dyn_hairC_r_a01}
      boxExtents=<10 15 15>
      constraintSetup {angularTargetAxis=AxisY, angularLimitsMax=<2 5 5>, angularLimitsMin=<-2 -2 -5>, angularTarget=<0 0 0>}
      localJointOffset=<-5 0.1 0.5>
     }]
    relativeSpaceBone {boneName=pelvis}
   }
   n16 = Anim Dynamics - Bone: dyn_hairC_r_a02(ComponentPose=@n15, Alpha=0.7, ExternalForce="0,0,0") {
    angularDampingOverride=0.999
    angularSpringConstant=800
    bAngularSpring=true
    bGravityOverrideInSimSpace=true
    bOverrideAngularDamping=true
    gravityScale=0.2
    simulationSpace=RootRelative
    alphaBoolBlend {blendOption=Linear}
    boundBone {boneName=dyn_hairC_r_a02}
    componentAppliedLinearAccClamp=<10000 10000 10000>
    componentLinearAccScale=<0.1 0.1 0.1>
    componentLinearVelScale=<0.1 0.1 0.1>
    physicsBodyDefinitions [{
      sphereCollisionRadius=0
      boundBone {boneName=dyn_hairC_r_a02}
      boxExtents=<10 15 15>
      constraintSetup {angularTargetAxis=AxisY, angularLimitsMax=<2 5 5>, angularLimitsMin=<-2 -2 -5>, angularTarget=<0 0 0>}
      localJointOffset=<-5 0.1 0.5>
     }]
    relativeSpaceBone {boneName=pelvis}
   }
   n17 = Anim Dynamics - Bone: dyn_hairC_r_b01(ComponentPose=@n16, Alpha=0.7, ExternalForce="0,0,0") {
    angularDampingOverride=0.999
    angularSpringConstant=800
    bAngularSpring=true
    bGravityOverrideInSimSpace=true
    bOverrideAngularDamping=true
    gravityScale=0.2
    simulationSpace=RootRelative
    alphaBoolBlend {blendOption=Linear}
    boundBone {boneName=dyn_hairC_r_b01}
    componentAppliedLinearAccClamp=<10000 10000 10000>
    componentLinearAccScale=<0.1 0.1 0.1>
    componentLinearVelScale=<0.1 0.1 0.1>
    physicsBodyDefinitions [{
      sphereCollisionRadius=0
      boundBone {boneName=dyn_hairC_r_b01}
      boxExtents=<10 15 15>
      constraintSetup {angularTargetAxis=AxisY, angularLimitsMax=<2 5 5>, angularLimitsMin=<-2 -2 -5>, angularTarget=<0 0 0>}
      localJointOffset=<-5 0.1 0.5>
     }]
    relativeSpaceBone {boneName=pelvis}
   }
   n18 = Anim Dynamics - Bone: dyn_hairC_r_b02(ComponentPose=@n17, Alpha=0.7, ExternalForce="0,0,0") {
    angularDampingOverride=0.999
    angularSpringConstant=800
    bAngularSpring=true
    bGravityOverrideInSimSpace=true
    bOverrideAngularDamping=true
    gravityScale=0.2
    simulationSpace=RootRelative
    alphaBoolBlend {blendOption=Linear}
    boundBone {boneName=dyn_hairC_r_b02}
    componentAppliedLinearAccClamp=<10000 10000 10000>
    componentLinearAccScale=<0.1 0.1 0.1>
    componentLinearVelScale=<0.1 0.1 0.1>
    physicsBodyDefinitions [{
      sphereCollisionRadius=0
      boundBone {boneName=dyn_hairC_r_b02}
      boxExtents=<10 15 15>
      constraintSetup {angularTargetAxis=AxisY, angularLimitsMax=<2 5 5>, angularLimitsMin=<-2 -2 -5>, angularTarget=<0 0 0>}
      localJointOffset=<-5 0.1 0.5>
     }]
    relativeSpaceBone {boneName=pelvis}
   }
   n19 = Anim Dynamics - Bone: dyn_hairC_r_c01(ComponentPose=@n18, Alpha=0.7, ExternalForce="0,0,0") {
    angularDampingOverride=0.999
    angularSpringConstant=800
    bAngularSpring=true
    bGravityOverrideInSimSpace=true
    bOverrideAngularDamping=true
    gravityScale=0.2
    simulationSpace=RootRelative
    alphaBoolBlend {blendOption=Linear}
    boundBone {boneName=dyn_hairC_r_c01}
    componentAppliedLinearAccClamp=<10000 10000 10000>
    componentLinearAccScale=<0.1 0.1 0.1>
    componentLinearVelScale=<0.1 0.1 0.1>
    physicsBodyDefinitions [{
      sphereCollisionRadius=0
      boundBone {boneName=dyn_hairC_r_c01}
      boxExtents=<10 15 15>
      constraintSetup {angularTargetAxis=AxisY, angularLimitsMax=<2 5 5>, angularLimitsMin=<-2 -2 -5>, angularTarget=<0 0 0>}
      localJointOffset=<-5 0.1 0.5>
     }]
    relativeSpaceBone {boneName=pelvis}
   }
   n20 = Anim Dynamics - Bone: dyn_hairC_r_c02(ComponentPose=@n19, Alpha=0.7, ExternalForce="0,0,0") {
    angularDampingOverride=0.999
    angularSpringConstant=800
    bAngularSpring=true
    bGravityOverrideInSimSpace=true
    bOverrideAngularDamping=true
    gravityScale=0.2
    simulationSpace=RootRelative
    alphaBoolBlend {blendOption=Linear}
    boundBone {boneName=dyn_hairC_r_c02}
    componentAppliedLinearAccClamp=<10000 10000 10000>
    componentLinearAccScale=<0.1 0.1 0.1>
    componentLinearVelScale=<0.1 0.1 0.1>
    physicsBodyDefinitions [{
      sphereCollisionRadius=0
      boundBone {boneName=dyn_hairC_r_c02}
      boxExtents=<10 15 15>
      constraintSetup {angularTargetAxis=AxisY, angularLimitsMax=<2 5 5>, angularLimitsMin=<-2 -2 -5>, angularTarget=<0 0 0>}
      localJointOffset=<-5 0.1 0.5>
     }]
    relativeSpaceBone {boneName=pelvis}
   }
   n21 = "Transform (Modify) Bone - Bone: dyn_hairC_r_01"(ComponentPose=@n20, Alpha=Get IsTargeting() -> (IsTargeting:double), Translation="0,0,0", Rotation="5,0,0", Scale="1,1,1") {
    rotationMode=BMM_Additive
    rotationSpace=BCS_BoneSpace
    scaleMode=BMM_Additive
    scaleSpace=BCS_BoneSpace
    translationMode=BMM_Additive
    translationSpace=BCS_BoneSpace
    alphaBoolBlend {blendOption=Linear}
    boneToModify {boneName=dyn_hairC_r_01}
    rotation {pitch=5}
   }
   Output Pose(Result=Component To Local(ComponentPose=@n21)) {layerGroup=None}
  ]}
 EventGraph {nodes [
   Event BlueprintUpdateAnimation()
   > n1 = Cast To FortPlayerAnimInstance(Object=GetAnimInstance(self=GetOwnerSkeletalMeshComponent()))
   > n2 = Switch on EFortWeaponCoreAnimation!(execute, Selection:EFortWeaponCoreAnimation Enum=Get WeaponCoreAnim(self=@n1)) -> (
    Melee:Exec
    Pistol:Exec
    Shotgun:Exec
    PaperBlueprint:Exec
    Rifle:Exec
    MeleeOneHand:Exec
    MachinePistol:Exec
    RocketLauncher:Exec
    GrenadeLauncher:Exec
    GoingCommando:Exec
    AssaultRifle:Exec
    TacticalShotgun:Exec
    SniperRifle:Exec
    TrapPlacement:Exec
    ShoulderLauncher:Exec
    AbilityDecoTool:Exec
    Crossbow:Exec
    C4:Exec
    RemoteControl:Exec
    DualWield:Exec
    AR_BullPup:Exec
    AR_ForwardGrip:Exec
    MedPackPaddles:Exec
    SMG_P90:Exec
    AR_DrumGun:Exec
    Consumable_Small:Exec
    Consumable_Large:Exec
    Balloon:Exec
    MountedTurret:Exec
    CreativeTool:Exec
    ExplosiveBow:Exec
    AshtonIndigo:Exec
    AshtonChicago:Exec
    MeleeDualWield:Exec
    Unarmed:Exec
   )
   Set IsTargeting(execute=@n2.Pistol, IsTargeting:double="To Float (Boolean)"(InBool=Get bIsTargeting())) -> (Output_Get:double)
  ]}
]
variables [IsTargeting {type=number}]
